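# Physics asset for the belt mesh of the male medium Weeping Woods Fall player cosmetic, covering a right belt piece and three grenades.
# Capsules on the pelvis-to-spine_05 chain form the torso bodies, with tiny MSTR capsules anchoring the belt piece to spine_03 and each grenade to spine_04.
# Only the belt piece's joint to its anchor has tight angular limits; every other constraint, grenades included, leaves swing and twist free.
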PhysicsAsset /BRCosmetics/Characters/Player/Male/Medium/Bodies/M_MED_WeepingWoods_Fall/Meshes/M_MED_Weeping_Woods_Fall_belt_Physics
bodies {
 dyn_belt_NEW_01_r [{shapeType=Capsule, length=12.9029, radius=4.61, center=<10.8215 -0.0001 0>, rotation=<rot 90 0 -89.9998>}]
 dyn_belt_NEW_01_r_MSTR [{shapeType=Capsule, length=0.01, radius=0.01}]
 dyn_grenade1_01 [{shapeType=Capsule, length=6.3269, radius=3.3296, center=<6.75 0 -0.0002>, rotation=<rot -90 0.0197 -0.0195>}]
 dyn_grenade1_01_MSTR [{shapeType=Capsule, length=0.01, radius=0.01}]
 dyn_grenade2_01 [{shapeType=Capsule, length=5.5769, radius=4.0796, center=<6.75 0 -0.0002>, rotation=<rot -90 0.0197 -0.0195>}]
 dyn_grenade2_01_MSTR [{shapeType=Capsule, length=0.01, radius=0.01}]
 dyn_grenade3_01 [{shapeType=Capsule, length=5.5769, radius=4.0796, center=<6.75 0 -0.0002>, rotation=<rot -90 0.0197 -0.0195>}]
 dyn_grenade3_01_MSTR [{shapeType=Capsule, length=0.01, radius=0.01}]
 pelvis [{shapeType=Capsule, length=11.6988, radius=13.0546}]
 spine_01 [{shapeType=Capsule, length=6.4488, radius=8.8046}]
 spine_02 [{shapeType=Capsule, length=6.75, radius=13.0546, center=<-3.9428 -1.4523 0>}]
 spine_03 [{shapeType=Capsule, length=8.1988, radius=14.3046, center=<0 -0.7561 0>}]
 spine_04 [{shapeType=Capsule, length=8, radius=14.0546}]
 spine_05 [{shapeType=Capsule, length=11, radius=12.5546}]
}
constraints [
 {
  bone1Name=spine_01
  bone2Name=pelvis
  swing1LimitDegrees=45
  swing1Motion=Free
  swing2LimitDegrees=45
  swing2Motion=Free
  twistLimitDegrees=45
  twistMotion=Free
 }
 {
  bone1Name=spine_02
  bone2Name=spine_01
  swing1LimitDegrees=45
  swing1Motion=Free
  swing2LimitDegrees=45
  swing2Motion=Free
  twistLimitDegrees=45
  twistMotion=Free
 }
 {
  bone1Name=spine_03
  bone2Name=spine_02
  swing1LimitDegrees=45
  swing1Motion=Free
  swing2LimitDegrees=45
  swing2Motion=Free
  twistLimitDegrees=45
  twistMotion=Free
 }
 {
  bone1Name=spine_04
  bone2Name=spine_03
  swing1LimitDegrees=45
  swing1Motion=Free
  swing2LimitDegrees=45
  swing2Motion=Free
  twistLimitDegrees=45
  twistMotion=Free
 }
 {
  bone1Name=spine_05
  bone2Name=spine_04
  swing1LimitDegrees=45
  swing1Motion=Free
  swing2LimitDegrees=45
  swing2Motion=Free
  twistLimitDegrees=45
  twistMotion=Free
 }
 {
  bone1Name=dyn_belt_NEW_01_r_MSTR
  bone2Name=spine_03
  swing1LimitDegrees=45
  swing1Motion=Free
  swing2LimitDegrees=45
  swing2Motion=Free
  twistLimitDegrees=45
  twistMotion=Free
 }
 {bone1Name=dyn_belt_NEW_01_r, bone2Name=dyn_belt_NEW_01_r_MSTR, swing1LimitDegrees=5, swing2LimitDegrees=25, twistLimitDegrees=5}
 {
  bone1Name=dyn_grenade3_01
  bone2Name=dyn_grenade3_01_MSTR
  swing1LimitDegrees=45
  swing1Motion=Free
  swing2LimitDegrees=45
  swing2Motion=Free
  twistLimitDegrees=45
  twistMotion=Free
 }
 {
  bone1Name=dyn_grenade3_01_MSTR
  bone2Name=spine_04
  swing1LimitDegrees=45
  swing1Motion=Free
  swing2LimitDegrees=45
  swing2Motion=Free
  twistLimitDegrees=45
  twistMotion=Free
 }
 {
  bone1Name=dyn_grenade2_01
  bone2Name=dyn_grenade2_01_MSTR
  swing1LimitDegrees=45
  swing1Motion=Free
  swing2LimitDegrees=45
  swing2Motion=Free
  twistLimitDegrees=45
  twistMotion=Free
 }
 {
  bone1Name=dyn_grenade2_01_MSTR
  bone2Name=spine_04
  swing1LimitDegrees=45
  swing1Motion=Free
  swing2LimitDegrees=45
  swing2Motion=Free
  twistLimitDegrees=45
  twistMotion=Free
 }
 {
  bone1Name=dyn_grenade1_01
  bone2Name=dyn_grenade1_01_MSTR
  swing1LimitDegrees=45
  swing1Motion=Free
  swing2LimitDegrees=45
  swing2Motion=Free
  twistLimitDegrees=45
  twistMotion=Free
 }
 {
  bone1Name=dyn_grenade1_01_MSTR
  bone2Name=spine_04
  swing1LimitDegrees=45
  swing1Motion=Free
  swing2LimitDegrees=45
  swing2Motion=Free
  twistLimitDegrees=45
  twistMotion=Free
 }
]
