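# Physics asset for the Weeping Woods Fall male medium body cosmetic, covering the torso and its dangling leaf bones.
# Capsules on pelvis and spine form a torso chain with free angular joints.
# Tiny master capsules, one per leaf, attach to spine_04 with free rotation and act as anchors.
# Each leaf box hinges on its master with limited swing and tight twist, allowing secondary motion without much spin.
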PhysicsAsset /BRCosmetics/Characters/Player/Male/Medium/Bodies/M_MED_WeepingWoods_Fall/Meshes/M_MED_Weeping_Woods_Fall_body_Physics
bodies {
 dyn_leafbody1_01_l [{shapeType=Box, extentX=9.1, extentY=9.1, extentZ=3.1, center=<4.6773 0 -0.0001>}]
 dyn_leafbody1_01_l_MSTR [{shapeType=Capsule, length=0.01, radius=0.01}]
 dyn_leafbody1_01_r [{shapeType=Box, extentX=9.1, extentY=9.1, extentZ=3.1, center=<4.6773 0 -0.0001>}]
 dyn_leafbody1_01_r_MSTR [{shapeType=Capsule, length=0.01, radius=0.01}]
 dyn_leafbody2_01_l [{shapeType=Box, extentX=9.1, extentY=9.1, extentZ=3.1, center=<4.6773 0 -0.0001>}]
 dyn_leafbody2_01_l_MSTR [{shapeType=Capsule, length=0.01, radius=0.01}]
 dyn_leafbody2_01_r [{shapeType=Box, extentX=9.1, extentY=9.1, extentZ=3.1, center=<4.6773 0 -0.0001>}]
 dyn_leafbody2_01_r_MSTR [{shapeType=Capsule, length=0.01, radius=0.01}]
 dyn_leafbody3_01_l [{shapeType=Box, extentX=9.1, extentY=9.1, extentZ=3.1, center=<4.6773 0 -0.0001>}]
 dyn_leafbody3_01_l_MSTR [{shapeType=Capsule, length=0.01, radius=0.01}]
 dyn_leafbody3_01_r [{shapeType=Box, extentX=9.1, extentY=9.1, extentZ=3.1, center=<4.6773 0 -0.0001>}]
 dyn_leafbody3_01_r_MSTR [{shapeType=Capsule, length=0.01, radius=0.01}]
 dyn_leafbody4_01_l [{shapeType=Box, extentX=9.1, extentY=9.1, extentZ=3.1, center=<5.9681 0 0>}]
 dyn_leafbody4_01_l_MSTR [{shapeType=Capsule, length=0.01, radius=0.01}]
 dyn_leafbody4_01_r [{shapeType=Box, extentX=9.1, extentY=9.1, extentZ=3.1, center=<4.6773 0 -0.0001>}]
 dyn_leafbody4_01_r_MSTR [{shapeType=Capsule, length=0.01, radius=0.01}]
 dyn_leafbody5_01_l [{shapeType=Box, extentX=9.1, extentY=9.1, extentZ=3.1, center=<4.6773 0 -0.0001>}]
 dyn_leafbody5_01_l_MSTR [{shapeType=Capsule, length=0.01, radius=0.01}]
 dyn_leafbody7_01_r [{shapeType=Box, extentX=9.1, extentY=9.1, extentZ=3.1, center=<4.6773 0 -0.0001>}]
 dyn_leafbody7_01_r_MSTR [{shapeType=Capsule, length=0.01, radius=0.01}]
 dyn_leafbody8_01_r [{shapeType=Box, extentX=9.1, extentY=9.1, extentZ=3.1, center=<4.6773 0 -0.0001>}]
 dyn_leafbody8_01_r_MSTR [{shapeType=Capsule, length=0.01, radius=0.01}]
 pelvis [{shapeType=Capsule, length=11.6988, radius=13.0546}]
 spine_01 [{shapeType=Capsule, length=6.4488, radius=8.8046}]
 spine_02 [{shapeType=Capsule, length=6.75, radius=13.0546, center=<-3.9428 -1.4523 0>}]
 spine_03 [{shapeType=Capsule, length=8.1988, radius=14.3046, center=<0 -0.7561 0>}]
 spine_04 [{shapeType=Capsule, length=8, radius=14.0546}]
 spine_05 [{shapeType=Capsule, length=11, radius=12.5546}]
}
constraints [
 {
  bone1Name=spine_01
  bone2Name=pelvis
  swing1LimitDegrees=45
  swing1Motion=Free
  swing2LimitDegrees=45
  swing2Motion=Free
  twistLimitDegrees=45
  twistMotion=Free
 }
 {
  bone1Name=spine_02
  bone2Name=spine_01
  swing1LimitDegrees=45
  swing1Motion=Free
  swing2LimitDegrees=45
  swing2Motion=Free
  twistLimitDegrees=45
  twistMotion=Free
 }
 {
  bone1Name=spine_03
  bone2Name=spine_02
  swing1LimitDegrees=45
  swing1Motion=Free
  swing2LimitDegrees=45
  swing2Motion=Free
  twistLimitDegrees=45
  twistMotion=Free
 }
 {
  bone1Name=spine_04
  bone2Name=spine_03
  swing1LimitDegrees=45
  swing1Motion=Free
  swing2LimitDegrees=45
  swing2Motion=Free
  twistLimitDegrees=45
  twistMotion=Free
 }
 {
  bone1Name=spine_05
  bone2Name=spine_04
  swing1LimitDegrees=45
  swing1Motion=Free
  swing2LimitDegrees=45
  swing2Motion=Free
  twistLimitDegrees=45
  twistMotion=Free
 }
 {
  bone1Name=dyn_leafbody1_01_l_MSTR
  bone2Name=spine_04
  swing1LimitDegrees=45
  swing1Motion=Free
  swing2LimitDegrees=45
  swing2Motion=Free
  twistLimitDegrees=45
  twistMotion=Free
 }
 {
  bone1Name=dyn_leafbody2_01_l_MSTR
  bone2Name=spine_04
  swing1LimitDegrees=45
  swing1Motion=Free
  swing2LimitDegrees=45
  swing2Motion=Free
  twistLimitDegrees=45
  twistMotion=Free
 }
 {
  bone1Name=dyn_leafbody3_01_l_MSTR
  bone2Name=spine_04
  swing1LimitDegrees=45
  swing1Motion=Free
  swing2LimitDegrees=45
  swing2Motion=Free
  twistLimitDegrees=45
  twistMotion=Free
 }
 {
  bone1Name=dyn_leafbody4_01_l_MSTR
  bone2Name=spine_04
  swing1LimitDegrees=45
  swing1Motion=Free
  swing2LimitDegrees=45
  swing2Motion=Free
  twistLimitDegrees=45
  twistMotion=Free
 }
 {
  bone1Name=dyn_leafbody5_01_l_MSTR
  bone2Name=spine_04
  swing1LimitDegrees=45
  swing1Motion=Free
  swing2LimitDegrees=45
  swing2Motion=Free
  twistLimitDegrees=45
  twistMotion=Free
 }
 {
  bone1Name=dyn_leafbody7_01_r_MSTR
  bone2Name=spine_04
  swing1LimitDegrees=45
  swing1Motion=Free
  swing2LimitDegrees=45
  swing2Motion=Free
  twistLimitDegrees=45
  twistMotion=Free
 }
 {
  bone1Name=dyn_leafbody8_01_r_MSTR
  bone2Name=spine_04
  swing1LimitDegrees=45
  swing1Motion=Free
  swing2LimitDegrees=45
  swing2Motion=Free
  twistLimitDegrees=45
  twistMotion=Free
 }
 {
  bone1Name=dyn_leafbody1_01_r_MSTR
  bone2Name=spine_04
  swing1LimitDegrees=45
  swing1Motion=Free
  swing2LimitDegrees=45
  swing2Motion=Free
  twistLimitDegrees=45
  twistMotion=Free
 }
 {
  bone1Name=dyn_leafbody2_01_r_MSTR
  bone2Name=spine_04
  swing1LimitDegrees=45
  swing1Motion=Free
  swing2LimitDegrees=45
  swing2Motion=Free
  twistLimitDegrees=45
  twistMotion=Free
 }
 {
  bone1Name=dyn_leafbody3_01_r_MSTR
  bone2Name=spine_04
  swing1LimitDegrees=45
  swing1Motion=Free
  swing2LimitDegrees=45
  swing2Motion=Free
  twistLimitDegrees=45
  twistMotion=Free
 }
 {
  bone1Name=dyn_leafbody4_01_r_MSTR
  bone2Name=spine_04
  swing1LimitDegrees=45
  swing1Motion=Free
  swing2LimitDegrees=45
  swing2Motion=Free
  twistLimitDegrees=45
  twistMotion=Free
 }
 {bone1Name=dyn_leafbody1_01_l, bone2Name=dyn_leafbody1_01_l_MSTR, swing1LimitDegrees=25, swing2LimitDegrees=25, twistLimitDegrees=5}
 {bone1Name=dyn_leafbody2_01_l, bone2Name=dyn_leafbody2_01_l_MSTR, swing1LimitDegrees=25, swing2LimitDegrees=25, twistLimitDegrees=5}
 {bone1Name=dyn_leafbody3_01_l, bone2Name=dyn_leafbody3_01_l_MSTR, swing1LimitDegrees=45, swing2LimitDegrees=45, twistLimitDegrees=5}
 {bone1Name=dyn_leafbody4_01_l, bone2Name=dyn_leafbody4_01_l_MSTR, swing1LimitDegrees=25, swing2LimitDegrees=25, twistLimitDegrees=5}
 {bone1Name=dyn_leafbody5_01_l, bone2Name=dyn_leafbody5_01_l_MSTR, swing1LimitDegrees=45, swing2LimitDegrees=45, twistLimitDegrees=5}
 {bone1Name=dyn_leafbody7_01_r, bone2Name=dyn_leafbody7_01_r_MSTR, swing1LimitDegrees=25, swing2LimitDegrees=25, twistLimitDegrees=5}
 {bone1Name=dyn_leafbody8_01_r, bone2Name=dyn_leafbody8_01_r_MSTR, swing1LimitDegrees=25, swing2LimitDegrees=25, twistLimitDegrees=5}
 {bone1Name=dyn_leafbody1_01_r, bone2Name=dyn_leafbody1_01_r_MSTR, swing1LimitDegrees=25, swing2LimitDegrees=25, twistLimitDegrees=5}
 {bone1Name=dyn_leafbody2_01_r, bone2Name=dyn_leafbody2_01_r_MSTR, swing1LimitDegrees=45, swing2LimitDegrees=45, twistLimitDegrees=5}
 {bone1Name=dyn_leafbody3_01_r, bone2Name=dyn_leafbody3_01_r_MSTR, swing1LimitDegrees=45, swing2LimitDegrees=45, twistLimitDegrees=5}
 {bone1Name=dyn_leafbody4_01_r, bone2Name=dyn_leafbody4_01_r_MSTR, swing1LimitDegrees=45, swing2LimitDegrees=45, twistLimitDegrees=5}
]
