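# Post-simulation AnimBP for the Weeping Woods Fall medium male cosmetic body, adding secondary motion to its leaves, skirt, belt and grenades.
# AnimDynamics swings grenades, while RigidBody nodes simulate skirt and body using gravity and wind vectors.
# Constraints and bone offsets correct the pose, some blended by crouch, cape and head-swing alphas.
# Many-target weighted position constraints attach leaf master bones to body and skirt bones, much like skin weights.
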AnimBlueprint /BRCosmetics/Characters/Player/Male/Medium/Bodies/M_MED_WeepingWoods_Fall/Meshes/M_MED_Weeping_Woods_Fall_AnimBP2
target_skeleton=&/M_MED_Weeping_Woods_Fall_Skeleton
graphs [
 AnimGraph {nodes [
   n1 = Anim Dynamics - Bone: dyn_grenade1_01(Alpha=0.75, ExternalForce="0,0,0") {
    angularDampingOverride=1
    angularSpringConstant=1500
    bAngularSpring=true
    bGravityOverrideInSimSpace=true
    bOverrideAngularDamping=true
    gravityScale=0.35
    simulationSpace=RootRelative
    alphaBoolBlend {blendOption=Linear}
    boundBone {boneName=dyn_grenade1_01}
    physicsBodyDefinitions [{
      sphereCollisionRadius=0
      boundBone {boneName=dyn_grenade1_01}
      boxExtents=<10 20 20>
      constraintSetup {angularLimitsMax=<5 0 10>, angularLimitsMin=<-5 -15 -10>, angularTarget=<0 0 0>}
      localJointOffset=<10 0 0>
     }]
   }
   n2 = Anim Dynamics - Bone: dyn_grenade2_01(ComponentPose=@n1, Alpha=0.6, ExternalForce="0,0,0") {
    angularDampingOverride=1
    angularSpringConstant=1500
    bAngularSpring=true
    bGravityOverrideInSimSpace=true
    bOverrideAngularDamping=true
    gravityScale=0.35
    simulationSpace=RootRelative
    alphaBoolBlend {blendOption=Linear}
    boundBone {boneName=dyn_grenade2_01}
    physicsBodyDefinitions [{
      sphereCollisionRadius=0
      boundBone {boneName=dyn_grenade2_01}
      boxExtents=<10 20 20>
      constraintSetup {angularLimitsMax=<5 0 10>, angularLimitsMin=<-5 -15 -10>, angularTarget=<0 0 0>}
      localJointOffset=<10 0 0>
     }]
   }
   Anim Dynamics - Bone: dyn_grenade3_01(ComponentPose=@n2, Alpha=0.5, ExternalForce="0,0,0") {
    angularDampingOverride=1
    angularSpringConstant=1500
    bAngularSpring=true
    bGravityOverrideInSimSpace=true
    bOverrideAngularDamping=true
    gravityScale=0.35
    simulationSpace=RootRelative
    alphaBoolBlend {blendOption=Linear}
    boundBone {boneName=dyn_grenade3_01}
    physicsBodyDefinitions [{
      sphereCollisionRadius=0
      boundBone {boneName=dyn_grenade3_01}
      boxExtents=<10 20 20>
      constraintSetup {angularLimitsMax=<5 0 10>, angularLimitsMin=<-5 -15 -10>, angularTarget=<0 0 0>}
      localJointOffset=<10 0 0>
     }]
   }
   n3 = "Transform (Modify) Bone - Bone: dyn_belt_mid_03_r"(Alpha=0, Translation="0,0,0", Rotation="10.000,0.000,0.000", Scale="1,1,1") {
    rotationMode=BMM_Additive
    rotationSpace=BCS_BoneSpace
    translationMode=BMM_Additive
    translationSpace=BCS_BoneSpace
    alphaBoolBlend {blendOption=Linear}
    boneToModify {boneName=dyn_belt_mid_03_r}
    rotation {pitch=10}
   }
   n4 = "Constraint - dyn_belt_fr_04_r (P)"(ComponentPose=@n3, Alpha=1, ConstraintWeights_0=1) {
    alphaBoolBlend {blendOption=Linear}
    boneToModify {boneName=dyn_belt_fr_04_r}
    constraintSetup [{targetBone {boneName=dyn_belt_mid_03_r}}]
    constraintWeights [1]
   }
   n5 = "Constraint - dyn_belt_bk_04_r (P)"(ComponentPose=@n4, Alpha=1, ConstraintWeights_0=1) {
    alphaBoolBlend {blendOption=Linear}
    boneToModify {boneName=dyn_belt_bk_04_r}
    constraintSetup [{targetBone {boneName=dyn_belt_mid_03_r}}]
    constraintWeights [1]
   }
   n6 = "Constraint - dyn_belt_fr_03_r (P)"(ComponentPose=@n5, Alpha=1, ConstraintWeights_0=1) {
    alphaBoolBlend {blendOption=Linear}
    boneToModify {boneName=dyn_belt_fr_03_r}
    constraintSetup [{targetBone {boneName=dyn_belt_mid_02_r}}]
    constraintWeights [1]
   }
   n7 = "Constraint - dyn_belt_bk_03_r (P)"(ComponentPose=@n6, Alpha=1, ConstraintWeights_0=1) {
    alphaBoolBlend {blendOption=Linear}
    boneToModify {boneName=dyn_belt_bk_03_r}
    constraintSetup [{targetBone {boneName=dyn_belt_mid_02_r}}]
    constraintWeights [1]
   }
   n8 = "Transform (Modify) Bone - Bone: dyn_belt_fr_04_r"(ComponentPose=@n7, Alpha=1, Translation="0.000,-1.480,3.679", Rotation="-0.001,-0.004,10.000", Scale="1,1,1") {
    rotationMode=BMM_Additive
    rotationSpace=BCS_BoneSpace
    translationMode=BMM_Additive
    translationSpace=BCS_BoneSpace
    alphaBoolBlend {blendOption=Linear}
    boneToModify {boneName=dyn_belt_fr_04_r}
    rotation=<rot -0.001 -0.004 10.0001>
    translation {y=-1.48, z=3.679}
   }
   n9 = Look At - Bone: dyn_belt_bk_02_r(ComponentPose=@n8, Alpha=1) {
    alphaBoolBlend {blendOption=Linear}
    boneToModify {boneName=dyn_belt_bk_02_r}
    lookAtLocation {x=0}
    lookAtTarget {boneReference {boneName=dyn_belt_bk_04_r}}
    lookAt_Axis {axis=<1 0>}
    lookUp_Axis {axis {x=0, z=1}}
   }
   Look At - Bone: dyn_belt_bk_03_r(ComponentPose=@n9, Alpha=1) {
    alphaBoolBlend {blendOption=Linear}
    boneToModify {boneName=dyn_belt_bk_03_r}
    lookAtLocation {x=0}
    lookAtTarget {boneReference {boneName=dyn_belt_bk_04_r}}
    lookAt_Axis {axis=<1 0>}
    lookUp_Axis {axis {x=0, z=1}}
   }
   "Transform (Modify) Bone - Bone: dyn_belt_bk_04_r"(Alpha=1, Translation="-0.000,1.420,0.003", Rotation="5.000,0.012,-0.006", Scale="1,1,1") {
    rotationMode=BMM_Additive
    rotationSpace=BCS_BoneSpace
    translationMode=BMM_Additive
    translationSpace=BCS_BoneSpace
    alphaBoolBlend {blendOption=Linear}
    boneToModify {boneName=dyn_belt_bk_04_r}
    rotation=<rot 5 0.012 -0.006>
    translation {y=1.42, z=0.003}
   }
   n10 = "Constraint - dyn_body_leaf_017_MSTR (R)"(Alpha=1, ConstraintWeights_0=1) {
    alphaBoolBlend {blendOption=Linear}
    boneToModify {boneName=dyn_body_leaf_017_MSTR}
    constraintSetup [{transformType=Rotation, targetBone {boneName=thigh_l}}]
    constraintWeights [1]
   }
   n11 = "Constraint - dyn_body_leaf_035_MSTR (R)"(ComponentPose=@n10, Alpha=1, ConstraintWeights_0=1) {
    alphaBoolBlend {blendOption=Linear}
    boneToModify {boneName=dyn_body_leaf_035_MSTR}
    constraintSetup [{transformType=Rotation, targetBone {boneName=thigh_l}}]
    constraintWeights [1]
   }
   n12 = "Constraint - dyn_body_leaf_034_MSTR (R)"(ComponentPose=@n11, Alpha=1, ConstraintWeights_0=1) {
    alphaBoolBlend {blendOption=Linear}
    boneToModify {boneName=dyn_body_leaf_034_MSTR}
    constraintSetup [{transformType=Rotation, targetBone {boneName=thigh_l}}]
    constraintWeights [1]
   }
   n13 = "Constraint - dyn_body_leaf_031_MSTR (R)"(ComponentPose=@n12, Alpha=1, ConstraintWeights_0=1) {
    alphaBoolBlend {blendOption=Linear}
    boneToModify {boneName=dyn_body_leaf_031_MSTR}
    constraintSetup [{transformType=Rotation, targetBone {boneName=thigh_l}}]
    constraintWeights [1]
   }
   n14 = "Constraint - dyn_body_leaf_027_MSTR (R)"(ComponentPose=@n13, Alpha=1, ConstraintWeights_0=1) {
    alphaBoolBlend {blendOption=Linear}
    boneToModify {boneName=dyn_body_leaf_027_MSTR}
    constraintSetup [{transformType=Rotation, targetBone {boneName=thigh_r}}]
    constraintWeights [1]
   }
   n15 = "Constraint - dyn_body_leaf_012_MSTR (R)"(ComponentPose=@n14, Alpha=1, ConstraintWeights_0=1) {
    alphaBoolBlend {blendOption=Linear}
    boneToModify {boneName=dyn_body_leaf_012_MSTR}
    constraintSetup [{transformType=Rotation, targetBone {boneName=thigh_r}}]
    constraintWeights [1]
   }
   n16 = "Constraint - dyn_body_leaf_03_MSTR (R)"(ComponentPose=@n15, Alpha=1, ConstraintWeights_0=1) {
    alphaBoolBlend {blendOption=Linear}
    boneToModify {boneName=dyn_body_leaf_03_MSTR}
    constraintSetup [{transformType=Rotation, targetBone {boneName=thigh_r}}]
    constraintWeights [1]
   }
   n17 = "Transform (Modify) Bone - Bone: dyn_body_leaf_017_MSTR"(ComponentPose=@n16, Alpha=0.5, Translation="-0.135,-0.000,1.553", Rotation="10,0.000,0.000", Scale="1,1,1") {
    rotationMode=BMM_Additive
    rotationSpace=BCS_BoneSpace
    translationMode=BMM_Additive
    translationSpace=BCS_BoneSpace
    alphaBoolBlend {blendOption=Linear}
    boneToModify {boneName=dyn_body_leaf_017_MSTR}
    rotation {pitch=10}
    translation {x=-0.135, z=1.553}
   }
   n18 = "Transform (Modify) Bone - Bone: dyn_body_leaf_035_MSTR"(ComponentPose=@n17, Alpha=0.5, Translation="0,0,0", Rotation="4.694,-40.295,-1.723", Scale="1,1,1") {
    rotationMode=BMM_Additive
    rotationSpace=BCS_BoneSpace
    translationMode=BMM_Additive
    translationSpace=BCS_BoneSpace
    alphaBoolBlend {blendOption=Linear}
    boneToModify {boneName=dyn_body_leaf_035_MSTR}
    rotation=<rot 4.694 -40.295 -1.723>
   }
   n19 = "Transform (Modify) Bone - Bone: dyn_body_leaf_032_MSTR"(ComponentPose=@n18, Alpha=1, Translation="0,0,0", Rotation="25.000,0.000,0.000", Scale="1,1,1") {
    rotationMode=BMM_Additive
    rotationSpace=BCS_BoneSpace
    translationMode=BMM_Additive
    translationSpace=BCS_BoneSpace
    alphaBoolBlend {blendOption=Linear}
    boneToModify {boneName=dyn_body_leaf_032_MSTR}
    rotation {pitch=25}
   }
   n20 = "Transform (Modify) Bone - Bone: dyn_body_leaf_034_MSTR"(ComponentPose=@n19, Alpha=0.5, Translation="0,0,0", Rotation="20.000,0.000,0.000", Scale="1,1,1") {
    rotationMode=BMM_Additive
    rotationSpace=BCS_BoneSpace
    translationMode=BMM_Additive
    translationSpace=BCS_BoneSpace
    alphaBoolBlend {blendOption=Linear}
    boneToModify {boneName=dyn_body_leaf_034_MSTR}
    rotation {pitch=20}
   }
   n21 = "Transform (Modify) Bone - Bone: dyn_body_leaf_031_MSTR"(ComponentPose=@n20, Alpha=0.5, Translation="0,0,0", Rotation="15.000,-0.001,-0.001", Scale="1,1,1") {
    rotationMode=BMM_Additive
    rotationSpace=BCS_BoneSpace
    translationMode=BMM_Additive
    translationSpace=BCS_BoneSpace
    alphaBoolBlend {blendOption=Linear}
    boneToModify {boneName=dyn_body_leaf_031_MSTR}
    rotation=<rot 15 -0.001 -0.001>
   }
   n22 = "Transform (Modify) Bone - Bone: dyn_body_leaf_027_MSTR"(ComponentPose=@n21, Alpha=0.5, Translation="0,0,0", Rotation="10,0.000,0.000", Scale="1,1,1") {
    rotationMode=BMM_Additive
    rotationSpace=BCS_BoneSpace
    translationMode=BMM_Additive
    translationSpace=BCS_BoneSpace
    alphaBoolBlend {blendOption=Linear}
    boneToModify {boneName=dyn_body_leaf_027_MSTR}
    rotation {pitch=10}
   }
   n23 = "Transform (Modify) Bone - Bone: dyn_body_leaf_012_MSTR"(ComponentPose=@n22, Alpha=0.5, Translation="-0.907,0.000,3.390", Rotation="20,-0.021,-0.016", Scale="1,1,1") {
    rotationMode=BMM_Additive
    rotationSpace=BCS_BoneSpace
    translationMode=BMM_Additive
    translationSpace=BCS_BoneSpace
    alphaBoolBlend {blendOption=Linear}
    boneToModify {boneName=dyn_body_leaf_012_MSTR}
    rotation=<rot 20 -0.021 -0.016>
    translation {x=-0.907, z=3.39}
   }
   n24 = "Transform (Modify) Bone - Bone: dyn_body_leaf_03_MSTR"(ComponentPose=@n23, Alpha=0.5, Translation="0,0,0", Rotation="10.000,0.000,0.000", Scale="1,1,1") {
    rotationMode=BMM_Additive
    rotationSpace=BCS_BoneSpace
    translationMode=BMM_Additive
    translationSpace=BCS_BoneSpace
    alphaBoolBlend {blendOption=Linear}
    boneToModify {boneName=dyn_body_leaf_03_MSTR}
    rotation {pitch=10}
   }
   n25 = Get isWearingCapeAlpha() -> (isWearingCapeAlpha:double)
   n26 = "Constraint - dyn_leafbody7_01_r (P)"(ComponentPose=@n24, Alpha=@n25, ConstraintWeights_0=1) {
    alphaBoolBlend {blendOption=Linear}
    boneToModify {boneName=dyn_leafbody7_01_r}
    constraintSetup [{targetBone {boneName=spine_04}}]
    constraintWeights [1]
   }
   n27 = "Constraint - dyn_leafbody4_01_l (P)"(ComponentPose=@n26, Alpha=@n25, ConstraintWeights_0=1) {
    alphaBoolBlend {blendOption=Linear}
    boneToModify {boneName=dyn_leafbody4_01_l}
    constraintSetup [{targetBone {boneName=spine_04}}]
    constraintWeights [1]
   }
   n28 = "Constraint - dyn_leafbody5_01_l (P)"(ComponentPose=@n27, Alpha=@n25, ConstraintWeights_0=1) {
    alphaBoolBlend {blendOption=Linear}
    boneToModify {boneName=dyn_leafbody5_01_l}
    constraintSetup [{targetBone {boneName=spine_04}}]
    constraintWeights [1]
   }
   n29 = "Constraint - dyn_body_leaf_039 (P)"(ComponentPose=@n28, Alpha=@n25, ConstraintWeights_0=1) {
    alphaBoolBlend {blendOption=Linear}
    boneToModify {boneName=dyn_body_leaf_039}
    constraintSetup [{targetBone {boneName=spine_04}}]
    constraintWeights [1]
   }
   n30 = "Constraint - dyn_body_leaf_037 (P)"(ComponentPose=@n29, Alpha=@n25, ConstraintWeights_0=1) {
    alphaBoolBlend {blendOption=Linear}
    boneToModify {boneName=dyn_body_leaf_037}
    constraintSetup [{targetBone {boneName=spine_04}}]
    constraintWeights [1]
   }
   n31 = "Constraint - dyn_body_leaf_055 (R)"(ComponentPose=@n30, Alpha=@n25, ConstraintWeights_0=1) {
    alphaBoolBlend {blendOption=Linear}
    boneToModify {boneName=dyn_body_leaf_055}
    constraintSetup [{transformType=Rotation, targetBone {boneName=spine_04}}]
    constraintWeights [1]
   }
   n32 = "Transform (Modify) Bone - Bone: dyn_body_leaf_055"(ComponentPose=@n31, Alpha=@n25, Translation="0.464,0.000,1.273", Rotation="-20,0.000,0.000", Scale="1,1,1") {
    rotationMode=BMM_Additive
    rotationSpace=BCS_BoneSpace
    translationMode=BMM_Additive
    translationSpace=BCS_BoneSpace
    alphaBoolBlend {blendOption=Linear}
    boneToModify {boneName=dyn_body_leaf_055}
    rotation {pitch=-20}
    translation=<0.4639 0 1.2727>
   }
   "Transform (Modify) Bone - Bone: dyn_body_leaf_039"(ComponentPose=@n32, Alpha=@n25, Translation="0.464,0.000,1.273", Rotation="-20,0.000,0.000", Scale="1,1,1") {
    rotationMode=BMM_Additive
    rotationSpace=BCS_BoneSpace
    translationMode=BMM_Additive
    translationSpace=BCS_BoneSpace
    alphaBoolBlend {blendOption=Linear}
    boneToModify {boneName=dyn_body_leaf_039}
    rotation {pitch=-20}
    translation=<0.4639 0 1.2727>
   }
   n33 = Get finalGravity() -> (finalGravity:Vector)
   n34 = RigidBody!(ComponentPose, Alpha, OverrideWorldGravity:Vector=@n33, ExternalForce) -> (Pose) {
    bClampLinearTranslationLimitToRefPose=true
    bOverrideWorldGravity=true
    overridePhysicsAsset=/BRCosmetics/Characters/Player/Male/Medium/Bodies/M_MED_WeepingWoods/Meshes/M_MED_Weeping_Woods_skirt_l_Physics
    simulationSpace=BaseBoneSpace
    alphaBoolBlend {blendOption=Linear}
    baseBoneRef {boneName=thigh_l}
    componentAppliedLinearAccClamp=<3000 3000 3000>
    componentLinearAccScale=<0.5 0.5 0.5>
    componentLinearVelScale=<0.5 0.5 0.5>
    overrideWorldGravity {z=-980}
   }
   RigidBody(ComponentPose=@n34, OverrideWorldGravity=@n33) {
    bClampLinearTranslationLimitToRefPose=true
    bOverrideWorldGravity=true
    overridePhysicsAsset=/BRCosmetics/Characters/Player/Male/Medium/Bodies/M_MED_WeepingWoods/Meshes/M_MED_Weeping_Woods_skirt_r_Physics
    simulationSpace=BaseBoneSpace
    alphaBoolBlend {blendOption=Linear}
    baseBoneRef {boneName=thigh_r}
    componentAppliedLinearAccClamp=<3000 3000 3000>
    componentLinearAccScale=<0.5 0.5 0.5>
    componentLinearVelScale=<0.5 0.5 0.5>
    overrideWorldGravity {z=-980}
   }
   n35 = Get windVector() -> (windVector:Vector)
   RigidBody(OverrideWorldGravity=@n35) {
    bClampLinearTranslationLimitToRefPose=true
    bOverrideWorldGravity=true
    overridePhysicsAsset=&/M_MED_Weeping_Woods_Fall_body_Physics
    simulationSpace=BaseBoneSpace
    alphaBoolBlend {blendOption=Linear}
    baseBoneRef {boneName=root}
    componentAppliedLinearAccClamp=<3000 3000 3000>
    componentLinearAccScale=<0.5 0.5 0.5>
    componentLinearVelScale=<0.7 0.7 0.7>
    overrideWorldGravity {z=-980}
   }
   n36 = Get isCrouchingAlpha() -> (isCrouchingAlpha:double)
   n37 = "Constraint - dyn_skirt_fr_03_l (P)"(ComponentPose=Local To Component(LocalPose=Use cached pose 'simulationPose'()), Alpha=@n36, ConstraintWeights_0=1) {
    alphaBoolBlend {blendOption=Linear}
    boneToModify {boneName=dyn_skirt_fr_03_l}
    constraintSetup [{targetBone {boneName=dyn_skirt_fr_02_l}}]
    constraintWeights [1]
   }
   n38 = "Constraint - dyn_skirt_fr_02_l (P)"(ComponentPose=@n37, Alpha=@n36, ConstraintWeights_0=1) {
    alphaBoolBlend {blendOption=Linear}
    boneToModify {boneName=dyn_skirt_fr_02_l}
    constraintSetup [{targetBone {boneName=dyn_skirt_fr_01_l}}]
    constraintWeights [1]
   }
   n39 = "Constraint - dyn_skirt_fr_01_l (P)"(ComponentPose=@n38, Alpha=@n36, ConstraintWeights_0=1) {
    alphaBoolBlend {blendOption=Linear}
    boneToModify {boneName=dyn_skirt_fr_01_l}
    constraintSetup [{targetBone {boneName=dyn_skirt_fr_01_l_MSTR}}]
    constraintWeights [1]
   }
   n40 = "Transform (Modify) Bone - Bone: dyn_skirt_fr_02_l"(ComponentPose=@n39, Alpha=@n36, Translation="0,0,0", Rotation="49.995,0.456,0.606", Scale="1,1,1") {
    rotationMode=BMM_Additive
    rotationSpace=BCS_BoneSpace
    alphaBoolBlend {blendOption=Linear}
    boneToModify {boneName=dyn_skirt_fr_02_l}
    rotation=<rot 49.9949 0.456 0.606>
   }
   n41 = "Transform (Modify) Bone - Bone: dyn_skirt_fr_03_l"(ComponentPose=@n40, Alpha=@n36, Translation="0,0,0", Rotation="-30.000,0.031,0.218", Scale="1,1,1") {
    rotationMode=BMM_Additive
    rotationSpace=BCS_BoneSpace
    translationMode=BMM_Additive
    translationSpace=BCS_BoneSpace
    alphaBoolBlend {blendOption=Linear}
    boneToModify {boneName=dyn_skirt_fr_03_l}
    rotation=<rot -29.9999 0.031 0.218>
   }
   n42 = "Transform (Modify) Bone - Bone: dyn_skirt_fr_01_l"(ComponentPose=@n41, Alpha=@n36, Translation="0,0,0", Rotation="-10.000,0.002,0.000", Scale="1,1,1") {
    rotationMode=BMM_Additive
    rotationSpace=BCS_BoneSpace
    alphaBoolBlend {blendOption=Linear}
    boneToModify {boneName=dyn_skirt_fr_01_l}
    rotation {pitch=-10, yaw=0.002}
   }
   n43 = "Transform (Modify) Bone - Bone: dyn_body_leaf_011"(ComponentPose=@n42, Alpha=@n36, Translation="0,0,0", Rotation="-20.000,-0.089,0.116", Scale="1,1,1") {
    rotationMode=BMM_Additive
    rotationSpace=BCS_BoneSpace
    alphaBoolBlend {blendOption=Linear}
    boneToModify {boneName=dyn_body_leaf_011}
    rotation=<rot -20 -0.089 0.116>
   }
   n44 = "Constraint - dyn_skirt_fr_01 (PP)"!(ComponentPose=@n43, Alpha=1, ConstraintWeights_0=0.5, ConstraintWeights_1:float=0.5) -> (Pose) {
    alphaBoolBlend {blendOption=Linear}
    boneToModify {boneName=dyn_skirt_fr_01}
    constraintSetup [{targetBone {boneName=dyn_skirt_fr_01_l}}, {targetBone {boneName=dyn_skirt_fr_01_r}}]
    constraintWeights [1,1]
   }
   n45 = "Constraint - dyn_skirt_fr_02 (RR)"!(ComponentPose=@n44, Alpha=1, ConstraintWeights_0=0.5, ConstraintWeights_1:float=0.5) -> (Pose) {
    alphaBoolBlend {blendOption=Linear}
    boneToModify {boneName=dyn_skirt_fr_02}
    constraintSetup [{transformType=Rotation, targetBone {boneName=dyn_skirt_fr_02_l}}, {transformType=Rotation, targetBone {boneName=dyn_skirt_fr_02_r}}]
    constraintWeights [1,1]
   }
   n46 = "Constraint - dyn_skirt_fr_03 (RR)"!(ComponentPose=@n45, Alpha=1, ConstraintWeights_0=0.5, ConstraintWeights_1:float=0.5) -> (Pose) {
    alphaBoolBlend {blendOption=Linear}
    boneToModify {boneName=dyn_skirt_fr_03}
    constraintSetup [{transformType=Rotation, targetBone {boneName=dyn_skirt_fr_03_l}}, {transformType=Rotation, targetBone {boneName=dyn_skirt_fr_03_r}}]
    constraintWeights [1,1]
   }
   n47 = "Constraint - dyn_skirt_bk_01 (RR)"!(ComponentPose=@n46, Alpha=1, ConstraintWeights_0=0.5, ConstraintWeights_1:float=0.5) -> (Pose) {
    alphaBoolBlend {blendOption=Linear}
    boneToModify {boneName=dyn_skirt_bk_01}
    constraintSetup [{transformType=Rotation, targetBone {boneName=dyn_skirt_bk_01_l}}, {transformType=Rotation, targetBone {boneName=dyn_skirt_bk_01_r}}]
    constraintWeights [1,1]
   }
   n48 = "Constraint - dyn_skirt_bk_02 (RR)"!(ComponentPose=@n47, Alpha=1, ConstraintWeights_0=0.5, ConstraintWeights_1:float=0.5) -> (Pose) {
    alphaBoolBlend {blendOption=Linear}
    boneToModify {boneName=dyn_skirt_bk_02}
    constraintSetup [{transformType=Rotation, targetBone {boneName=dyn_skirt_bk_02_l}}, {transformType=Rotation, targetBone {boneName=dyn_skirt_bk_02_r}}]
    constraintWeights [1,1]
   }
   n49 = "Constraint - dyn_skirt_bk_03 (RR)"!(ComponentPose=@n48, Alpha=1, ConstraintWeights_0=0.5, ConstraintWeights_1:float=0.5) -> (Pose) {
    alphaBoolBlend {blendOption=Linear}
    boneToModify {boneName=dyn_skirt_bk_03}
    constraintSetup [{transformType=Rotation, targetBone {boneName=dyn_skirt_bk_03_l}}, {transformType=Rotation, targetBone {boneName=dyn_skirt_bk_03_r}}]
    constraintWeights [1,1]
   }
   n50 = "Transform (Modify) Bone - Bone: dyn_leafbody5_01_r"(
    ComponentPose=@n49
    Alpha=Get headSwingAlpha() -> (headSwingAlpha:double)
    Translation="0,0,0"
    Rotation="-35.000,-0.001,0.000"
    Scale="1,1,1"
   ) {
    rotationMode=BMM_Additive
    rotationSpace=BCS_BoneSpace
    translationMode=BMM_Additive
    translationSpace=BCS_BoneSpace
    alphaBoolBlend {blendOption=Linear}
    boneToModify {boneName=dyn_leafbody5_01_r}
    rotation {pitch=-35, yaw=-0.001}
   }
   n51 = "Transform (Modify) Bone - Bone: dyn_belt_NEW_01_r"(
    ComponentPose=Local To Component(LocalPose=Component To Local(ComponentPose=@n50))
    Alpha=1
    Translation="0,0,0"
    Rotation="15.000,0.000,0.000"
    Scale="1,1,1"
   ) {
    rotationMode=BMM_Additive
    rotationSpace=BCS_BoneSpace
    translationMode=BMM_Additive
    translationSpace=BCS_BoneSpace
    alphaBoolBlend {blendOption=Linear}
    boneToModify {boneName=dyn_belt_NEW_01_r}
    rotation {pitch=15}
   }
   n52 = "Transform (Modify) Bone - Bone: dyn_skirt_fr_03_l"(ComponentPose=@n51, Alpha=1, Translation="0,0,0", Rotation="15.000,0.000,0.000", Scale="1,1,1") {
    rotationMode=BMM_Additive
    rotationSpace=BCS_BoneSpace
    translationMode=BMM_Additive
    translationSpace=BCS_BoneSpace
    alphaBoolBlend {blendOption=Linear}
    boneToModify {boneName=dyn_skirt_fr_03_l}
    rotation {pitch=15}
   }
   n53 = Get isCrouchingAlpha()
   n54 = "Transform (Modify) Bone - Bone: dyn_skirt_fr_02_r"(ComponentPose=@n52, Alpha=@n53, Translation="-0.443,0.135,3.584", Rotation="25.000,0.000,0.000", Scale="1,1,1") {
    rotationMode=BMM_Additive
    rotationSpace=BCS_BoneSpace
    translationMode=BMM_Additive
    translationSpace=BCS_BoneSpace
    alphaBoolBlend {blendOption=Linear}
    alphaScaleBiasClamp {bInterpResult=true}
    boneToModify {boneName=dyn_skirt_fr_02_r}
    rotation {pitch=25}
    translation=<-0.443 0.135 3.584>
   }
   n55 = "Transform (Modify) Bone - Bone: dyn_skirt_fr_02"(ComponentPose=@n54, Alpha=@n53, Translation="-1.206,0.004,2.587", Rotation="25.000,0.000,0.000", Scale="1,1,1") {
    rotationMode=BMM_Additive
    rotationSpace=BCS_BoneSpace
    translationMode=BMM_Additive
    translationSpace=BCS_BoneSpace
    alphaBoolBlend {blendOption=Linear}
    alphaScaleBiasClamp {bInterpResult=true}
    boneToModify {boneName=dyn_skirt_fr_02}
    rotation {pitch=25}
    translation=<-1.206 0.004 2.587>
   }
   Output Pose(Result=Component To Local(ComponentPose=@n55)) {layerGroup=None}
   Save cached pose 'postSimOutPose'() {cacheName=postSimOutPose}
   n56 = "Constraint - dyn_body_leaf_054_MSTR (PPPPPPPPP)"!(
    ComponentPose=Local To Component(LocalPose=Use cached pose 'postSimOutPose'())
    Alpha=1
    ConstraintWeights_0=0.2644
    ConstraintWeights_1:float=0.0139
    ConstraintWeights_2:float=0.0451
    ConstraintWeights_3:float=0.3051
    ConstraintWeights_4:float=0.2948
    ConstraintWeights_5:float=0.0445
    ConstraintWeights_6:float=0.0024
    ConstraintWeights_7:float=0.0286
    ConstraintWeights_8:float=0.0011
   ) -> (Pose) {
    alphaBoolBlend {blendOption=Linear}
    boneToModify {boneName=dyn_body_leaf_054_MSTR}
    constraintSetup [
     {targetBone {boneName=clavicle_l}}
     {targetBone {boneName=spine_01}}
     {targetBone {boneName=spine_02}}
     {targetBone {boneName=spine_03}}
     {targetBone {boneName=spine_04}}
     {targetBone {boneName=spine_05}}
     {targetBone {boneName=upperarm_twist_01_l}}
     {targetBone {boneName=clavicle_r}}
     {targetBone {boneName=pelvis}}
    ]
    constraintWeights [1,1,1,1,1,1,1,1,1]
   }
   n57 = "Constraint - dyn_body_leaf_037_MSTR (PPPPPPPPPP)"!(
    ComponentPose=@n56
    Alpha=1
    ConstraintWeights_0=0.1802
    ConstraintWeights_1:float=0.04
    ConstraintWeights_2:float=0.0743
    ConstraintWeights_3:float=0.3043
    ConstraintWeights_4:float=0.2554
    ConstraintWeights_5:float=0.051
    ConstraintWeights_6:float=0.002
    ConstraintWeights_7:float=0.0493
    ConstraintWeights_8:float=0.0424
    ConstraintWeights_9:float=0.0012
   ) -> (Pose) {
    alphaBoolBlend {blendOption=Linear}
    boneToModify {boneName=dyn_body_leaf_037_MSTR}
    constraintSetup [
     {targetBone {boneName=clavicle_l}}
     {targetBone {boneName=spine_01}}
     {targetBone {boneName=spine_02}}
     {targetBone {boneName=spine_03}}
     {targetBone {boneName=spine_04}}
     {targetBone {boneName=spine_05}}
     {targetBone {boneName=upperarm_twist_01_l}}
     {targetBone {boneName=clavicle_r}}
     {targetBone {boneName=pelvis}}
     {targetBone {boneName=thigh_twist_01_l}}
    ]
    constraintWeights [1,1,1,1,1,1,1,1,1,1]
   }
   n58 = "Constraint - dyn_body_leaf_021_MSTR (PPPPPPPPPPP)"!(
    ComponentPose=@n57
    Alpha=1
    ConstraintWeights_0=0.0015
    ConstraintWeights_1:float=0.2124
    ConstraintWeights_2:float=0.0551
    ConstraintWeights_3:float=0.0471
    ConstraintWeights_4:float=0.1404
    ConstraintWeights_5:float=0.24
    ConstraintWeights_6:float=0.184
    ConstraintWeights_7:float=0.0164
    ConstraintWeights_8:float=0.0878
    ConstraintWeights_9:float=0.0129
    ConstraintWeights_10:float=0.0026
   ) -> (Pose) {
    alphaBoolBlend {blendOption=Linear}
    boneToModify {boneName=dyn_body_leaf_021_MSTR}
    constraintSetup [
     {targetBone {boneName=head}}
     {targetBone {boneName=clavicle_l}}
     {targetBone {boneName=spine_01}}
     {targetBone {boneName=upperarm_twist_02_l}}
     {targetBone {boneName=spine_02}}
     {targetBone {boneName=spine_03}}
     {targetBone {boneName=spine_04}}
     {targetBone {boneName=spine_05}}
     {targetBone {boneName=upperarm_twist_01_l}}
     {targetBone {boneName=pelvis}}
     {targetBone {boneName=upperarm_l}}
    ]
    constraintWeights [1,1,1,1,1,1,1,1,1,1,1]
   }
   n59 = "Constraint - dyn_body_leaf_055_MSTR (PPPPPPPPPP)"!(
    ComponentPose=@n58
    Alpha=1
    ConstraintWeights_0=0.1391
    ConstraintWeights_1:float=0.0001
    ConstraintWeights_2:float=0.0442
    ConstraintWeights_3:float=0.0769
    ConstraintWeights_4:float=0.3019
    ConstraintWeights_5:float=0.2348
    ConstraintWeights_6:float=0.0609
    ConstraintWeights_7:float=0.0001
    ConstraintWeights_8:float=0.0903
    ConstraintWeights_9:float=0.0517
   ) -> (Pose) {
    alphaBoolBlend {blendOption=Linear}
    boneToModify {boneName=dyn_body_leaf_055_MSTR}
    constraintSetup [
     {targetBone {boneName=clavicle_l}}
     {targetBone {boneName=upperarm_twist_01_r}}
     {targetBone {boneName=spine_01}}
     {targetBone {boneName=spine_02}}
     {targetBone {boneName=spine_03}}
     {targetBone {boneName=spine_04}}
     {targetBone {boneName=spine_05}}
     {targetBone {boneName=upperarm_twist_01_l}}
     {targetBone {boneName=clavicle_r}}
     {targetBone {boneName=pelvis}}
    ]
    constraintWeights [1,1,1,1,1,1,1,1,1,1]
   }
   n60 = "Constraint - dyn_body_leaf_058_MSTR (PPPPPPPPPPPPPPPP)"!(
    ComponentPose=@n59
    Alpha=1
    ConstraintWeights_0=0.0004
    ConstraintWeights_1:float=0.0135
    ConstraintWeights_2:float=0.2093
    ConstraintWeights_3:float=0.2833
    ConstraintWeights_4:float=0.4387
    ConstraintWeights_5:float=0.0031
    ConstraintWeights_6:float=0.0007
    ConstraintWeights_7:float=0.0008
    ConstraintWeights_8:float=0.0005
    ConstraintWeights_9:float=0.0001
    ConstraintWeights_10:float=0.0001
    ConstraintWeights_11:float=0.0017
    ConstraintWeights_12:float=0.007
    ConstraintWeights_13:float=0.034
    ConstraintWeights_14:float=0.0066
    ConstraintWeights_15:float=0.0001
   ) -> (Pose) {
    alphaBoolBlend {blendOption=Linear}
    boneToModify {boneName=dyn_body_leaf_058_MSTR}
    constraintSetup [
     {targetBone {boneName=dyn_skirt_side_03_l}}
     {targetBone {boneName=dyn_skirt_fr_03_l}}
     {targetBone {boneName=dyn_skirt_fr_02_l}}
     {targetBone {boneName=dyn_skirt_fr_01}}
     {targetBone {boneName=dyn_skirt_fr_02}}
     {targetBone {boneName=dyn_skirt_fr_03}}
     {targetBone {boneName=thigh_twist_01_r}}
     {targetBone {boneName=dyn_skirt_side_01_l}}
     {targetBone {boneName=spine_01}}
     {targetBone {boneName=dyn_skirt_fr_02_r}}
     {targetBone {boneName=thigh_l}}
     {targetBone {boneName=dyn_skirt_side_02_l}}
     {targetBone {boneName=pelvis}}
     {targetBone {boneName=dyn_skirt_fr_01_l}}
     {targetBone {boneName=thigh_twist_01_l}}
     {targetBone {boneName=dyn_skirt_fr_03_r}}
    ]
    constraintWeights [1,1,1,1,1,1,1,1,1,1,1,1,1,1,1,1]
   }
   n61 = "Constraint - dyn_body_leaf_048_MSTR (PPPPPPPPPPPPPPPP)"!(
    ComponentPose=@n60
    Alpha=1
    ConstraintWeights_0=0.0073
    ConstraintWeights_1:float=0.0017
    ConstraintWeights_2:float=0.0001
    ConstraintWeights_3:float=0.0001
    ConstraintWeights_4:float=0.077
    ConstraintWeights_5:float=0.6951
    ConstraintWeights_6:float=0.0189
    ConstraintWeights_7:float=0.0019
    ConstraintWeights_8:float=0.0011
    ConstraintWeights_9:float=0.0004
    ConstraintWeights_10:float=0.1222
    ConstraintWeights_11:float=0.0002
    ConstraintWeights_12:float=0.0018
    ConstraintWeights_13:float=0.0695
    ConstraintWeights_14:float=0.0012
    ConstraintWeights_15:float=0.0013
   ) -> (Pose) {
    alphaBoolBlend {blendOption=Linear}
    boneToModify {boneName=dyn_body_leaf_048_MSTR}
    constraintSetup [
     {targetBone {boneName=dyn_skirt_fr_01_r}}
     {targetBone {boneName=dyn_skirt_side_02_r}}
     {targetBone {boneName=spine_01}}
     {targetBone {boneName=thigh_r}}
     {targetBone {boneName=dyn_skirt_fr_01}}
     {targetBone {boneName=dyn_skirt_fr_02}}
     {targetBone {boneName=dyn_skirt_fr_03}}
     {targetBone {boneName=thigh_twist_01_r}}
     {targetBone {boneName=dyn_skirt_fr_02_l}}
     {targetBone {boneName=dyn_skirt_side_01_r}}
     {targetBone {boneName=dyn_skirt_fr_02_r}}
     {targetBone {boneName=thigh_twist_01_l}}
     {targetBone {boneName=pelvis}}
     {targetBone {boneName=dyn_skirt_fr_03_r}}
     {targetBone {boneName=dyn_skirt_fr_03_l}}
     {targetBone {boneName=dyn_skirt_side_03_r}}
    ]
    constraintWeights [1,1,1,1,1,1,1,1,1,1,1,1,1,1,1,1]
   }
   n62 = "Constraint - dyn_body_leaf_013_MSTR (PPPPPPPPPP)"!(
    ComponentPose=@n61
    Alpha=1
    ConstraintWeights_0=0.0048
    ConstraintWeights_1:float=0.0003
    ConstraintWeights_2:float=0.0716
    ConstraintWeights_3:float=0.8939
    ConstraintWeights_4:float=0.0011
    ConstraintWeights_5:float=0.0012
    ConstraintWeights_6:float=0.0239
    ConstraintWeights_7:float=0.0013
    ConstraintWeights_8:float=0.0016
    ConstraintWeights_9:float=0.0001
   ) -> (Pose) {
    alphaBoolBlend {blendOption=Linear}
    boneToModify {boneName=dyn_body_leaf_013_MSTR}
    constraintSetup [
     {targetBone {boneName=dyn_skirt_fr_01_r}}
     {targetBone {boneName=dyn_skirt_side_02_r}}
     {targetBone {boneName=dyn_skirt_fr_01}}
     {targetBone {boneName=dyn_skirt_fr_02}}
     {targetBone {boneName=dyn_skirt_fr_03}}
     {targetBone {boneName=thigh_twist_01_r}}
     {targetBone {boneName=dyn_skirt_fr_02_r}}
     {targetBone {boneName=pelvis}}
     {targetBone {boneName=dyn_skirt_fr_03_r}}
     {targetBone {boneName=thigh_twist_01_l}}
    ]
    constraintWeights [1,1,1,1,1,1,1,1,1,1]
   }
   n63 = "Constraint - dyn_body_leaf_06_MSTR (PPPPPPPPPPPP)"!(
    ComponentPose=@n62
    Alpha=1
    ConstraintWeights_0=0.0005
    ConstraintWeights_1:float=0.0318
    ConstraintWeights_2:float=0.8327
    ConstraintWeights_3:float=0.1037
    ConstraintWeights_4:float=0.0003
    ConstraintWeights_5:float=0.0004
    ConstraintWeights_6:float=0.0001
    ConstraintWeights_7:float=0.0003
    ConstraintWeights_8:float=0.0037
    ConstraintWeights_9:float=0.0002
    ConstraintWeights_10:float=0.004
    ConstraintWeights_11:float=0.0223
   ) -> (Pose) {
    alphaBoolBlend {blendOption=Linear}
    boneToModify {boneName=dyn_body_leaf_06_MSTR}
    constraintSetup [
     {targetBone {boneName=dyn_skirt_fr_03_l}}
     {targetBone {boneName=dyn_skirt_fr_02_l}}
     {targetBone {boneName=dyn_skirt_fr_01}}
     {targetBone {boneName=dyn_skirt_fr_02}}
     {targetBone {boneName=dyn_skirt_fr_03}}
     {targetBone {boneName=thigh_twist_01_r}}
     {targetBone {boneName=dyn_skirt_side_01_l}}
     {targetBone {boneName=spine_01}}
     {targetBone {boneName=thigh_twist_01_l}}
     {targetBone {boneName=dyn_skirt_side_02_l}}
     {targetBone {boneName=pelvis}}
     {targetBone {boneName=dyn_skirt_fr_01_l}}
    ]
    constraintWeights [1,1,1,1,1,1,1,1,1,1,1,1]
   }
   n64 = "Constraint - dyn_body_leaf_059_MSTR (PPPPPPP)"!(
    ComponentPose=@n63
    Alpha=1
    ConstraintWeights_0=0.0006
    ConstraintWeights_1:float=0.1733
    ConstraintWeights_2:float=0.8249
    ConstraintWeights_3:float=0.0001
    ConstraintWeights_4:float=0.0008
    ConstraintWeights_5:float=0.0001
    ConstraintWeights_6:float=0.0001
   ) -> (Pose) {
    alphaBoolBlend {blendOption=Linear}
    boneToModify {boneName=dyn_body_leaf_059_MSTR}
    constraintSetup [
     {targetBone {boneName=dyn_skirt_fr_01_r}}
     {targetBone {boneName=dyn_skirt_fr_01}}
     {targetBone {boneName=dyn_skirt_fr_02}}
     {targetBone {boneName=thigh_twist_01_r}}
     {targetBone {boneName=dyn_skirt_fr_02_r}}
     {targetBone {boneName=pelvis}}
     {targetBone {boneName=dyn_skirt_fr_01_l}}
    ]
    constraintWeights [1,1,1,1,1,1,1]
   }
   n65 = "Constraint - dyn_body_leaf_065_MSTR (PPPPPPPPPPPPPPPP)"!(
    ComponentPose=@n64
    Alpha=1
    ConstraintWeights_0=0.0977
    ConstraintWeights_1:float=0.2237
    ConstraintWeights_2:float=0.0006
    ConstraintWeights_3:float=0.0009
    ConstraintWeights_4:float=0.6211
    ConstraintWeights_5:float=0.0332
    ConstraintWeights_6:float=0.0006
    ConstraintWeights_7:float=0.0078
    ConstraintWeights_8:float=0.0032
    ConstraintWeights_9:float=0.0004
    ConstraintWeights_10:float=0.0005
    ConstraintWeights_11:float=0.0001
    ConstraintWeights_12:float=0.0082
    ConstraintWeights_13:float=0.0003
    ConstraintWeights_14:float=0.0007
    ConstraintWeights_15:float=0.0009
   ) -> (Pose) {
    alphaBoolBlend {blendOption=Linear}
    boneToModify {boneName=dyn_body_leaf_065_MSTR}
    constraintSetup [
     {targetBone {boneName=dyn_skirt_fr_02_r}}
     {targetBone {boneName=dyn_skirt_fr_01_r}}
     {targetBone {boneName=dyn_skirt_side_02_r}}
     {targetBone {boneName=spine_01}}
     {targetBone {boneName=dyn_skirt_fr_01}}
     {targetBone {boneName=dyn_skirt_fr_02}}
     {targetBone {boneName=dyn_skirt_fr_03}}
     {targetBone {boneName=thigh_twist_01_r}}
     {targetBone {boneName=dyn_skirt_side_01_r}}
     {targetBone {boneName=spine_02}}
     {targetBone {boneName=spine_03}}
     {targetBone {boneName=spine_04}}
     {targetBone {boneName=pelvis}}
     {targetBone {boneName=dyn_skirt_fr_01_l}}
     {targetBone {boneName=thigh_twist_01_l}}
     {targetBone {boneName=dyn_skirt_fr_03_r}}
    ]
    constraintWeights [1,1,1,1,1,1,1,1,1,1,1,1,1,1,1,1]
   }
   n66 = "Constraint - dyn_body_leaf_038_MSTR (PPPPPPPPPP)"!(
    ComponentPose=@n65
    Alpha=1
    ConstraintWeights_0=0.3381
    ConstraintWeights_1:float=0.0163
    ConstraintWeights_2:float=0.0027
    ConstraintWeights_3:float=0.0579
    ConstraintWeights_4:float=0.269
    ConstraintWeights_5:float=0.2541
    ConstraintWeights_6:float=0.0179
    ConstraintWeights_7:float=0.0298
    ConstraintWeights_8:float=0.0102
    ConstraintWeights_9:float=0.0039
   ) -> (Pose) {
    alphaBoolBlend {blendOption=Linear}
    boneToModify {boneName=dyn_body_leaf_038_MSTR}
    constraintSetup [
     {targetBone {boneName=clavicle_l}}
     {targetBone {boneName=spine_01}}
     {targetBone {boneName=upperarm_twist_02_l}}
     {targetBone {boneName=spine_02}}
     {targetBone {boneName=spine_03}}
     {targetBone {boneName=spine_04}}
     {targetBone {boneName=spine_05}}
     {targetBone {boneName=upperarm_twist_01_l}}
     {targetBone {boneName=clavicle_r}}
     {targetBone {boneName=pelvis}}
    ]
    constraintWeights [1,1,1,1,1,1,1,1,1,1]
   }
   n67 = "Constraint - dyn_body_leaf_051_MSTR (PPPPPPPPPPPP)"!(
    ComponentPose=@n66
    Alpha=1
    ConstraintWeights_0=0.0118
    ConstraintWeights_1:float=0.3582
    ConstraintWeights_2:float=0.0106
    ConstraintWeights_3:float=0.1024
    ConstraintWeights_4:float=0.0425
    ConstraintWeights_5:float=0.1058
    ConstraintWeights_6:float=0.074
    ConstraintWeights_7:float=0.0547
    ConstraintWeights_8:float=0.2334
    ConstraintWeights_9:float=0.0012
    ConstraintWeights_10:float=0.0039
    ConstraintWeights_11:float=0.0014
   ) -> (Pose) {
    alphaBoolBlend {blendOption=Linear}
    boneToModify {boneName=dyn_body_leaf_051_MSTR}
    constraintSetup [
     {targetBone {boneName=head}}
     {targetBone {boneName=clavicle_l}}
     {targetBone {boneName=spine_01}}
     {targetBone {boneName=upperarm_twist_02_l}}
     {targetBone {boneName=spine_02}}
     {targetBone {boneName=spine_03}}
     {targetBone {boneName=spine_04}}
     {targetBone {boneName=spine_05}}
     {targetBone {boneName=upperarm_twist_01_l}}
     {targetBone {boneName=clavicle_r}}
     {targetBone {boneName=neck_02}}
     {targetBone {boneName=neck_01}}
    ]
    constraintWeights [1,1,1,1,1,1,1,1,1,1,1,1]
   }
   n68 = "Constraint - dyn_body_leaf_023_MSTR (PPPPPPPPPPP)"!(
    ComponentPose=@n67
    Alpha=1
    ConstraintWeights_0=0.0023
    ConstraintWeights_1:float=0.3544
    ConstraintWeights_2:float=0.0141
    ConstraintWeights_3:float=0.0999
    ConstraintWeights_4:float=0.0509
    ConstraintWeights_5:float=0.1288
    ConstraintWeights_6:float=0.0791
    ConstraintWeights_7:float=0.0329
    ConstraintWeights_8:float=0.2353
    ConstraintWeights_9:float=0.0011
    ConstraintWeights_10:float=0.0012
   ) -> (Pose) {
    alphaBoolBlend {blendOption=Linear}
    boneToModify {boneName=dyn_body_leaf_023_MSTR}
    constraintSetup [
     {targetBone {boneName=head}}
     {targetBone {boneName=clavicle_l}}
     {targetBone {boneName=spine_01}}
     {targetBone {boneName=upperarm_twist_02_l}}
     {targetBone {boneName=spine_02}}
     {targetBone {boneName=spine_03}}
     {targetBone {boneName=spine_04}}
     {targetBone {boneName=spine_05}}
     {targetBone {boneName=upperarm_twist_01_l}}
     {targetBone {boneName=neck_02}}
     {targetBone {boneName=pelvis}}
    ]
    constraintWeights [1,1,1,1,1,1,1,1,1,1,1]
   }
   n69 = "Constraint - dyn_body_leaf_01_MSTR (PPPPPPPPPPPPP)"!(
    ComponentPose=@n68
    Alpha=1
    ConstraintWeights_0=0.0194
    ConstraintWeights_1:float=0.1563
    ConstraintWeights_2:float=0.0516
    ConstraintWeights_3:float=0.0202
    ConstraintWeights_4:float=0.1478
    ConstraintWeights_5:float=0.3057
    ConstraintWeights_6:float=0.2006
    ConstraintWeights_7:float=0.037
    ConstraintWeights_8:float=0.039
    ConstraintWeights_9:float=0.0037
    ConstraintWeights_10:float=0.0075
    ConstraintWeights_11:float=0.0099
    ConstraintWeights_12:float=0.0013
   ) -> (Pose) {
    alphaBoolBlend {blendOption=Linear}
    boneToModify {boneName=dyn_body_leaf_01_MSTR}
    constraintSetup [
     {targetBone {boneName=head}}
     {targetBone {boneName=clavicle_l}}
     {targetBone {boneName=spine_01}}
     {targetBone {boneName=upperarm_twist_02_l}}
     {targetBone {boneName=spine_02}}
     {targetBone {boneName=spine_03}}
     {targetBone {boneName=spine_04}}
     {targetBone {boneName=spine_05}}
     {targetBone {boneName=upperarm_twist_01_l}}
     {targetBone {boneName=neck_02}}
     {targetBone {boneName=clavicle_r}}
     {targetBone {boneName=pelvis}}
     {targetBone {boneName=neck_01}}
    ]
    constraintWeights [1,1,1,1,1,1,1,1,1,1,1,1,1]
   }
   n70 = "Constraint - dyn_body_leaf_047_MSTR (PPPPPPPPPP)"!(
    ComponentPose=@n69
    Alpha=1
    ConstraintWeights_0=0.0012
    ConstraintWeights_1:float=0.3788
    ConstraintWeights_2:float=0.0122
    ConstraintWeights_3:float=0.0867
    ConstraintWeights_4:float=0.0454
    ConstraintWeights_5:float=0.1326
    ConstraintWeights_6:float=0.0741
    ConstraintWeights_7:float=0.0314
    ConstraintWeights_8:float=0.2364
    ConstraintWeights_9:float=0.0011
   ) -> (Pose) {
    alphaBoolBlend {blendOption=Linear}
    boneToModify {boneName=dyn_body_leaf_047_MSTR}
    constraintSetup [
     {targetBone {boneName=head}}
     {targetBone {boneName=clavicle_l}}
     {targetBone {boneName=spine_01}}
     {targetBone {boneName=upperarm_twist_02_l}}
     {targetBone {boneName=spine_02}}
     {targetBone {boneName=spine_03}}
     {targetBone {boneName=spine_04}}
     {targetBone {boneName=spine_05}}
     {targetBone {boneName=upperarm_twist_01_l}}
     {targetBone {boneName=pelvis}}
    ]
    constraintWeights [1,1,1,1,1,1,1,1,1,1]
   }
   n71 = "Constraint - dyn_body_leaf_024_MSTR (PPPPPPPPPPPP)"!(
    ComponentPose=@n70
    Alpha=1
    ConstraintWeights_0=0.0015
    ConstraintWeights_1:float=0.444
    ConstraintWeights_2:float=0.0052
    ConstraintWeights_3:float=0.0582
    ConstraintWeights_4:float=0.0245
    ConstraintWeights_5:float=0.1185
    ConstraintWeights_6:float=0.0548
    ConstraintWeights_7:float=0.0481
    ConstraintWeights_8:float=0.2381
    ConstraintWeights_9:float=0.0027
    ConstraintWeights_10:float=0.003
    ConstraintWeights_11:float=0.0014
   ) -> (Pose) {
    alphaBoolBlend {blendOption=Linear}
    boneToModify {boneName=dyn_body_leaf_024_MSTR}
    constraintSetup [
     {targetBone {boneName=head}}
     {targetBone {boneName=clavicle_l}}
     {targetBone {boneName=spine_01}}
     {targetBone {boneName=upperarm_twist_02_l}}
     {targetBone {boneName=spine_02}}
     {targetBone {boneName=spine_03}}
     {targetBone {boneName=spine_04}}
     {targetBone {boneName=spine_05}}
     {targetBone {boneName=upperarm_twist_01_l}}
     {targetBone {boneName=clavicle_r}}
     {targetBone {boneName=neck_02}}
     {targetBone {boneName=neck_01}}
    ]
    constraintWeights [1,1,1,1,1,1,1,1,1,1,1,1]
   }
   n72 = "Constraint - dyn_body_leaf_018_MSTR (PPPPPPPPPPPPPPP)"!(
    ComponentPose=@n71
    Alpha=1
    ConstraintWeights_0=0.0921
    ConstraintWeights_1:float=0.1674
    ConstraintWeights_2:float=0.0298
    ConstraintWeights_3:float=0.0035
    ConstraintWeights_4:float=0.0018
    ConstraintWeights_5:float=0.0246
    ConstraintWeights_6:float=0.0989
    ConstraintWeights_7:float=0.2105
    ConstraintWeights_8:float=0.1769
    ConstraintWeights_9:float=0.0897
    ConstraintWeights_10:float=0.0514
    ConstraintWeights_11:float=0.0155
    ConstraintWeights_12:float=0.0278
    ConstraintWeights_13:float=0.0046
    ConstraintWeights_14:float=0.0055
   ) -> (Pose) {
    alphaBoolBlend {blendOption=Linear}
    boneToModify {boneName=dyn_body_leaf_018_MSTR}
    constraintSetup [
     {targetBone {boneName=head}}
     {targetBone {boneName=clavicle_l}}
     {targetBone {boneName=spine_01}}
     {targetBone {boneName=upperarm_twist_01_r}}
     {targetBone {boneName=upperarm_twist_02_r}}
     {targetBone {boneName=upperarm_twist_02_l}}
     {targetBone {boneName=spine_02}}
     {targetBone {boneName=spine_03}}
     {targetBone {boneName=spine_04}}
     {targetBone {boneName=spine_05}}
     {targetBone {boneName=upperarm_twist_01_l}}
     {targetBone {boneName=neck_02}}
     {targetBone {boneName=clavicle_r}}
     {targetBone {boneName=pelvis}}
     {targetBone {boneName=neck_01}}
    ]
    constraintWeights [1,1,1,1,1,1,1,1,1,1,1,1,1,1,1]
   }
   n73 = "Constraint - dyn_body_leaf_070_MSTR (PPPPPPPPPPPPPPP)"!(
    ComponentPose=@n72
    Alpha=1
    ConstraintWeights_0=0.1994
    ConstraintWeights_1:float=0.055
    ConstraintWeights_2:float=0.0179
    ConstraintWeights_3:float=0.0355
    ConstraintWeights_4:float=0.016
    ConstraintWeights_5:float=0.0047
    ConstraintWeights_6:float=0.0656
    ConstraintWeights_7:float=0.1408
    ConstraintWeights_8:float=0.1485
    ConstraintWeights_9:float=0.1306
    ConstraintWeights_10:float=0.0097
    ConstraintWeights_11:float=0.0295
    ConstraintWeights_12:float=0.1333
    ConstraintWeights_13:float=0.003
    ConstraintWeights_14:float=0.0104
   ) -> (Pose) {
    alphaBoolBlend {blendOption=Linear}
    boneToModify {boneName=dyn_body_leaf_070_MSTR}
    constraintSetup [
     {targetBone {boneName=head}}
     {targetBone {boneName=clavicle_l}}
     {targetBone {boneName=spine_01}}
     {targetBone {boneName=upperarm_twist_01_r}}
     {targetBone {boneName=upperarm_twist_02_r}}
     {targetBone {boneName=upperarm_twist_02_l}}
     {targetBone {boneName=spine_02}}
     {targetBone {boneName=spine_03}}
     {targetBone {boneName=spine_04}}
     {targetBone {boneName=spine_05}}
     {targetBone {boneName=upperarm_twist_01_l}}
     {targetBone {boneName=neck_02}}
     {targetBone {boneName=clavicle_r}}
     {targetBone {boneName=pelvis}}
     {targetBone {boneName=neck_01}}
    ]
    constraintWeights [1,1,1,1,1,1,1,1,1,1,1,1,1,1,1]
   }
   n74 = "Constraint - dyn_body_leaf_05_MSTR (PPPPPPPPPPPP)"!(
    ComponentPose=@n73
    Alpha=1
    ConstraintWeights_0=0.0016
    ConstraintWeights_1:float=0.003
    ConstraintWeights_2:float=0.2333
    ConstraintWeights_3:float=0.0561
    ConstraintWeights_4:float=0.0052
    ConstraintWeights_5:float=0.0247
    ConstraintWeights_6:float=0.1211
    ConstraintWeights_7:float=0.0571
    ConstraintWeights_8:float=0.0486
    ConstraintWeights_9:float=0.4447
    ConstraintWeights_10:float=0.0032
    ConstraintWeights_11:float=0.0014
   ) -> (Pose) {
    alphaBoolBlend {blendOption=Linear}
    boneToModify {boneName=dyn_body_leaf_05_MSTR}
    constraintSetup [
     {targetBone {boneName=head}}
     {targetBone {boneName=clavicle_l}}
     {targetBone {boneName=upperarm_twist_01_r}}
     {targetBone {boneName=upperarm_twist_02_r}}
     {targetBone {boneName=spine_01}}
     {targetBone {boneName=spine_02}}
     {targetBone {boneName=spine_03}}
     {targetBone {boneName=spine_04}}
     {targetBone {boneName=spine_05}}
     {targetBone {boneName=clavicle_r}}
     {targetBone {boneName=neck_02}}
     {targetBone {boneName=neck_01}}
    ]
    constraintWeights [1,1,1,1,1,1,1,1,1,1,1,1]
   }
   n75 = "Constraint - dyn_body_leaf_04_MSTR (PPPPPPPPPPPP)"!(
    ComponentPose=@n74
    Alpha=1
    ConstraintWeights_0=0.0121
    ConstraintWeights_1:float=0.0012
    ConstraintWeights_2:float=0.2326
    ConstraintWeights_3:float=0.1021
    ConstraintWeights_4:float=0.0107
    ConstraintWeights_5:float=0.0428
    ConstraintWeights_6:float=0.1062
    ConstraintWeights_7:float=0.0745
    ConstraintWeights_8:float=0.0549
    ConstraintWeights_9:float=0.3575
    ConstraintWeights_10:float=0.004
    ConstraintWeights_11:float=0.0014
   ) -> (Pose) {
    alphaBoolBlend {blendOption=Linear}
    boneToModify {boneName=dyn_body_leaf_04_MSTR}
    constraintSetup [
     {targetBone {boneName=head}}
     {targetBone {boneName=clavicle_l}}
     {targetBone {boneName=upperarm_twist_01_r}}
     {targetBone {boneName=upperarm_twist_02_r}}
     {targetBone {boneName=spine_01}}
     {targetBone {boneName=spine_02}}
     {targetBone {boneName=spine_03}}
     {targetBone {boneName=spine_04}}
     {targetBone {boneName=spine_05}}
     {targetBone {boneName=clavicle_r}}
     {targetBone {boneName=neck_02}}
     {targetBone {boneName=neck_01}}
    ]
    constraintWeights [1,1,1,1,1,1,1,1,1,1,1,1]
   }
   n76 = "Constraint - dyn_body_leaf_071_MSTR (PPPPPPPPPPPPPPP)"!(
    ComponentPose=@n75
    Alpha=1
    ConstraintWeights_0=0.2142
    ConstraintWeights_1:float=0.0933
    ConstraintWeights_2:float=0.019
    ConstraintWeights_3:float=0.0198
    ConstraintWeights_4:float=0.0096
    ConstraintWeights_5:float=0.0084
    ConstraintWeights_6:float=0.0668
    ConstraintWeights_7:float=0.1397
    ConstraintWeights_8:float=0.1487
    ConstraintWeights_9:float=0.1323
    ConstraintWeights_10:float=0.0185
    ConstraintWeights_11:float=0.0299
    ConstraintWeights_12:float=0.0855
    ConstraintWeights_13:float=0.004
    ConstraintWeights_14:float=0.0105
   ) -> (Pose) {
    alphaBoolBlend {blendOption=Linear}
    boneToModify {boneName=dyn_body_leaf_071_MSTR}
    constraintSetup [
     {targetBone {boneName=head}}
     {targetBone {boneName=clavicle_l}}
     {targetBone {boneName=spine_01}}
     {targetBone {boneName=upperarm_twist_01_r}}
     {targetBone {boneName=upperarm_twist_02_r}}
     {targetBone {boneName=upperarm_twist_02_l}}
     {targetBone {boneName=spine_02}}
     {targetBone {boneName=spine_03}}
     {targetBone {boneName=spine_04}}
     {targetBone {boneName=spine_05}}
     {targetBone {boneName=upperarm_twist_01_l}}
     {targetBone {boneName=neck_02}}
     {targetBone {boneName=clavicle_r}}
     {targetBone {boneName=pelvis}}
     {targetBone {boneName=neck_01}}
    ]
    constraintWeights [1,1,1,1,1,1,1,1,1,1,1,1,1,1,1]
   }
   n77 = "Constraint - dyn_body_leaf_08_MSTR (PPPPPPPPPPPPPPP)"!(
    ComponentPose=@n76
    Alpha=1
    ConstraintWeights_0=0.1621
    ConstraintWeights_1:float=0.0329
    ConstraintWeights_2:float=0.0169
    ConstraintWeights_3:float=0.0651
    ConstraintWeights_4:float=0.0288
    ConstraintWeights_5:float=0.0021
    ConstraintWeights_6:float=0.0622
    ConstraintWeights_7:float=0.1367
    ConstraintWeights_8:float=0.1386
    ConstraintWeights_9:float=0.1255
    ConstraintWeights_10:float=0.0044
    ConstraintWeights_11:float=0.028
    ConstraintWeights_12:float=0.1845
    ConstraintWeights_13:float=0.0023
    ConstraintWeights_14:float=0.01
   ) -> (Pose) {
    alphaBoolBlend {blendOption=Linear}
    boneToModify {boneName=dyn_body_leaf_08_MSTR}
    constraintSetup [
     {targetBone {boneName=head}}
     {targetBone {boneName=clavicle_l}}
     {targetBone {boneName=spine_01}}
     {targetBone {boneName=upperarm_twist_01_r}}
     {targetBone {boneName=upperarm_twist_02_r}}
     {targetBone {boneName=upperarm_twist_02_l}}
     {targetBone {boneName=spine_02}}
     {targetBone {boneName=spine_03}}
     {targetBone {boneName=spine_04}}
     {targetBone {boneName=spine_05}}
     {targetBone {boneName=upperarm_twist_01_l}}
     {targetBone {boneName=neck_02}}
     {targetBone {boneName=clavicle_r}}
     {targetBone {boneName=pelvis}}
     {targetBone {boneName=neck_01}}
    ]
    constraintWeights [1,1,1,1,1,1,1,1,1,1,1,1,1,1,1]
   }
   n78 = "Constraint - dyn_body_leaf_044_MSTR (PPPPPPPPPPP)"!(
    ComponentPose=@n77
    Alpha=1
    ConstraintWeights_0=0.0012
    ConstraintWeights_1:float=0.0001
    ConstraintWeights_2:float=0.2354
    ConstraintWeights_3:float=0.0857
    ConstraintWeights_4:float=0.0122
    ConstraintWeights_5:float=0.0454
    ConstraintWeights_6:float=0.1332
    ConstraintWeights_7:float=0.0745
    ConstraintWeights_8:float=0.0315
    ConstraintWeights_9:float=0.3798
    ConstraintWeights_10:float=0.0011
   ) -> (Pose) {
    alphaBoolBlend {blendOption=Linear}
    boneToModify {boneName=dyn_body_leaf_044_MSTR}
    constraintSetup [
     {targetBone {boneName=head}}
     {targetBone {boneName=clavicle_l}}
     {targetBone {boneName=upperarm_twist_01_r}}
     {targetBone {boneName=upperarm_twist_02_r}}
     {targetBone {boneName=spine_01}}
     {targetBone {boneName=spine_02}}
     {targetBone {boneName=spine_03}}
     {targetBone {boneName=spine_04}}
     {targetBone {boneName=spine_05}}
     {targetBone {boneName=clavicle_r}}
     {targetBone {boneName=pelvis}}
    ]
    constraintWeights [1,1,1,1,1,1,1,1,1,1,1]
   }
   n79 = "Constraint - dyn_body_leaf_020_MSTR (PPPPPPPPPPP)"!(
    ComponentPose=@n78
    Alpha=1
    ConstraintWeights_0=0.0044
    ConstraintWeights_1:float=0.147
    ConstraintWeights_2:float=0.0367
    ConstraintWeights_3:float=0.0141
    ConstraintWeights_4:float=0.0507
    ConstraintWeights_5:float=0.1899
    ConstraintWeights_6:float=0.1271
    ConstraintWeights_7:float=0.0288
    ConstraintWeights_8:float=0.0015
    ConstraintWeights_9:float=0.3973
    ConstraintWeights_10:float=0.0024
   ) -> (Pose) {
    alphaBoolBlend {blendOption=Linear}
    boneToModify {boneName=dyn_body_leaf_020_MSTR}
    constraintSetup [
     {targetBone {boneName=clavicle_l}}
     {targetBone {boneName=upperarm_twist_01_r}}
     {targetBone {boneName=upperarm_twist_02_r}}
     {targetBone {boneName=spine_01}}
     {targetBone {boneName=spine_02}}
     {targetBone {boneName=spine_03}}
     {targetBone {boneName=spine_04}}
     {targetBone {boneName=spine_05}}
     {targetBone {boneName=neck_02}}
     {targetBone {boneName=clavicle_r}}
     {targetBone {boneName=pelvis}}
    ]
    constraintWeights [1,1,1,1,1,1,1,1,1,1,1]
   }
   n80 = "Constraint - dyn_body_leaf_031_MSTR (PPPPPPPPPPPPPPPPPP)"!(
    ComponentPose=@n79
    Alpha=1
    ConstraintWeights_0=0.0083
    ConstraintWeights_1:float=0.0003
    ConstraintWeights_2:float=0.0062
    ConstraintWeights_3:float=0.0004
    ConstraintWeights_4:float=0.4843
    ConstraintWeights_5:float=0.0006
    ConstraintWeights_6:float=0.0068
    ConstraintWeights_7:float=0.0077
    ConstraintWeights_8:float=0.054
    ConstraintWeights_9:float=0.0694
    ConstraintWeights_10:float=0.0838
    ConstraintWeights_11:float=0.0207
    ConstraintWeights_12:float=0.0004
    ConstraintWeights_13:float=0.0009
    ConstraintWeights_14:float=0.0766
    ConstraintWeights_15:float=0.0001
    ConstraintWeights_16:float=0.0233
    ConstraintWeights_17:float=0.1564
   ) -> (Pose) {
    alphaBoolBlend {blendOption=Linear}
    boneToModify {boneName=dyn_body_leaf_031_MSTR}
    constraintSetup [
     {targetBone {boneName=dyn_skirt_fr_01_r}}
     {targetBone {boneName=dyn_skirt_fr_02_l}}
     {targetBone {boneName=clavicle_l}}
     {targetBone {boneName=dyn_skirt_side_01_r}}
     {targetBone {boneName=dyn_skirt_fr_01}}
     {targetBone {boneName=dyn_skirt_fr_02}}
     {targetBone {boneName=thigh_twist_01_r}}
     {targetBone {boneName=dyn_skirt_side_01_l}}
     {targetBone {boneName=spine_01}}
     {targetBone {boneName=spine_02}}
     {targetBone {boneName=spine_03}}
     {targetBone {boneName=spine_04}}
     {targetBone {boneName=spine_05}}
     {targetBone {boneName=clavicle_r}}
     {targetBone {boneName=pelvis}}
     {targetBone {boneName=dyn_skirt_bk_01_l}}
     {targetBone {boneName=thigh_twist_01_l}}
     {targetBone {boneName=dyn_skirt_fr_01_l}}
    ]
    constraintWeights [1,1,1,1,1,1,1,1,1,1,1,1,1,1,1,1,1,1]
   }
   n81 = "Constraint - dyn_body_leaf_027_MSTR (PPPPPPPPPPP)"!(
    ComponentPose=@n80
    Alpha=1
    ConstraintWeights_0=0.0139
    ConstraintWeights_1:float=0.006
    ConstraintWeights_2:float=0.0084
    ConstraintWeights_3:float=0.1598
    ConstraintWeights_4:float=0.2624
    ConstraintWeights_5:float=0.3203
    ConstraintWeights_6:float=0.0849
    ConstraintWeights_7:float=0.0027
    ConstraintWeights_8:float=0.0091
    ConstraintWeights_9:float=0.1237
    ConstraintWeights_10:float=0.0087
   ) -> (Pose) {
    alphaBoolBlend {blendOption=Linear}
    boneToModify {boneName=dyn_body_leaf_027_MSTR}
    constraintSetup [
     {targetBone {boneName=clavicle_l}}
     {targetBone {boneName=dyn_skirt_fr_01}}
     {targetBone {boneName=thigh_twist_01_r}}
     {targetBone {boneName=spine_01}}
     {targetBone {boneName=spine_02}}
     {targetBone {boneName=spine_03}}
     {targetBone {boneName=spine_04}}
     {targetBone {boneName=spine_05}}
     {targetBone {boneName=clavicle_r}}
     {targetBone {boneName=pelvis}}
     {targetBone {boneName=thigh_twist_01_l}}
    ]
    constraintWeights [1,1,1,1,1,1,1,1,1,1,1]
   }
   n82 = "Constraint - dyn_body_leaf_012_MSTR (PPPPPPPPPPP)"!(
    ComponentPose=@n81
    Alpha=1
    ConstraintWeights_0=0.002
    ConstraintWeights_1:float=0.1461
    ConstraintWeights_2:float=0.053
    ConstraintWeights_3:float=0.1373
    ConstraintWeights_4:float=0.1773
    ConstraintWeights_5:float=0.2111
    ConstraintWeights_6:float=0.0518
    ConstraintWeights_7:float=0.0011
    ConstraintWeights_8:float=0.0141
    ConstraintWeights_9:float=0.1891
    ConstraintWeights_10:float=0.0169
   ) -> (Pose) {
    alphaBoolBlend {blendOption=Linear}
    boneToModify {boneName=dyn_body_leaf_012_MSTR}
    constraintSetup [
     {targetBone {boneName=clavicle_l}}
     {targetBone {boneName=dyn_skirt_fr_01}}
     {targetBone {boneName=thigh_twist_01_r}}
     {targetBone {boneName=spine_01}}
     {targetBone {boneName=spine_02}}
     {targetBone {boneName=spine_03}}
     {targetBone {boneName=spine_04}}
     {targetBone {boneName=spine_05}}
     {targetBone {boneName=clavicle_r}}
     {targetBone {boneName=pelvis}}
     {targetBone {boneName=thigh_twist_01_l}}
    ]
    constraintWeights [1,1,1,1,1,1,1,1,1,1,1]
   }
   n83 = "Constraint - dyn_body_leaf_064_MSTR (PPPPPPPPPP)"!(
    ComponentPose=@n82
    Alpha=1
    ConstraintWeights_0=0.0044
    ConstraintWeights_1:float=0.0231
    ConstraintWeights_2:float=0.0001
    ConstraintWeights_3:float=0.123
    ConstraintWeights_4:float=0.007
    ConstraintWeights_5:float=0.0051
    ConstraintWeights_6:float=0.0003
    ConstraintWeights_7:float=0.2583
    ConstraintWeights_8:float=0.0028
    ConstraintWeights_9:float=0.5758
   ) -> (Pose) {
    alphaBoolBlend {blendOption=Linear}
    boneToModify {boneName=dyn_body_leaf_064_MSTR}
    constraintSetup [
     {targetBone {boneName=dyn_skirt_side_03_l}}
     {targetBone {boneName=dyn_skirt_bk_03_l}}
     {targetBone {boneName=dyn_skirt_fr_02_l}}
     {targetBone {boneName=dyn_skirt_side_01_l}}
     {targetBone {boneName=dyn_skirt_bk_01_l}}
     {targetBone {boneName=thigh_twist_01_l}}
     {targetBone {boneName=thigh_l}}
     {targetBone {boneName=dyn_skirt_side_02_l}}
     {targetBone {boneName=pelvis}}
     {targetBone {boneName=dyn_skirt_bk_02_l}}
    ]
    constraintWeights [1,1,1,1,1,1,1,1,1,1]
   }
   n84 = "Constraint - dyn_body_leaf_060_MSTR (PPPPPPPPPP)"!(
    ComponentPose=@n83
    Alpha=1
    ConstraintWeights_0=0.0011
    ConstraintWeights_1:float=0.0002
    ConstraintWeights_2:float=0.0001
    ConstraintWeights_3:float=0.0009
    ConstraintWeights_4:float=0.2784
    ConstraintWeights_5:float=0.0004
    ConstraintWeights_6:float=0.7163
    ConstraintWeights_7:float=0.0002
    ConstraintWeights_8:float=0.0013
    ConstraintWeights_9:float=0.0011
   ) -> (Pose) {
    alphaBoolBlend {blendOption=Linear}
    boneToModify {boneName=dyn_body_leaf_060_MSTR}
    constraintSetup [
     {targetBone {boneName=dyn_skirt_side_03_l}}
     {targetBone {boneName=dyn_skirt_fr_03_l}}
     {targetBone {boneName=dyn_skirt_bk_03_l}}
     {targetBone {boneName=dyn_skirt_fr_02_l}}
     {targetBone {boneName=dyn_skirt_side_01_l}}
     {targetBone {boneName=thigh_twist_01_l}}
     {targetBone {boneName=dyn_skirt_side_02_l}}
     {targetBone {boneName=pelvis}}
     {targetBone {boneName=dyn_skirt_fr_01_l}}
     {targetBone {boneName=dyn_skirt_bk_02_l}}
    ]
    constraintWeights [1,1,1,1,1,1,1,1,1,1]
   }
   n85 = "Constraint - dyn_body_leaf_061_MSTR (PPPPPPPPPPPP)"!(
    ComponentPose=@n84
    Alpha=1
    ConstraintWeights_0=0.0284
    ConstraintWeights_1:float=0.0007
    ConstraintWeights_2:float=0.091
    ConstraintWeights_3:float=0.0009
    ConstraintWeights_4:float=0.0679
    ConstraintWeights_5:float=0.0005
    ConstraintWeights_6:float=0.0013
    ConstraintWeights_7:float=0.0001
    ConstraintWeights_8:float=0.5033
    ConstraintWeights_9:float=0.0006
    ConstraintWeights_10:float=0.0003
    ConstraintWeights_11:float=0.3049
   ) -> (Pose) {
    alphaBoolBlend {blendOption=Linear}
    boneToModify {boneName=dyn_body_leaf_061_MSTR}
    constraintSetup [
     {targetBone {boneName=dyn_skirt_side_03_l}}
     {targetBone {boneName=dyn_skirt_fr_03_l}}
     {targetBone {boneName=dyn_skirt_bk_03_l}}
     {targetBone {boneName=dyn_skirt_fr_02_l}}
     {targetBone {boneName=dyn_skirt_side_01_l}}
     {targetBone {boneName=dyn_skirt_bk_01_l}}
     {targetBone {boneName=thigh_twist_01_l}}
     {targetBone {boneName=thigh_l}}
     {targetBone {boneName=dyn_skirt_side_02_l}}
     {targetBone {boneName=pelvis}}
     {targetBone {boneName=dyn_skirt_fr_01_l}}
     {targetBone {boneName=dyn_skirt_bk_02_l}}
    ]
    constraintWeights [1,1,1,1,1,1,1,1,1,1,1,1]
   }
   n86 = "Constraint - dyn_body_leaf_049_MSTR (PPPPPPPPP)"!(
    ComponentPose=@n85
    Alpha=1
    ConstraintWeights_0=0.0005
    ConstraintWeights_1:float=0.9608
    ConstraintWeights_2:float=0.0002
    ConstraintWeights_3:float=0.0002
    ConstraintWeights_4:float=0.0002
    ConstraintWeights_5:float=0.0263
    ConstraintWeights_6:float=0.0009
    ConstraintWeights_7:float=0.0095
    ConstraintWeights_8:float=0.0013
   ) -> (Pose) {
    alphaBoolBlend {blendOption=Linear}
    boneToModify {boneName=dyn_body_leaf_049_MSTR}
    constraintSetup [
     {targetBone {boneName=dyn_skirt_bk_01_l}}
     {targetBone {boneName=dyn_skirt_side_01_l}}
     {targetBone {boneName=spine_01}}
     {targetBone {boneName=spine_02}}
     {targetBone {boneName=spine_03}}
     {targetBone {boneName=dyn_skirt_side_02_l}}
     {targetBone {boneName=pelvis}}
     {targetBone {boneName=dyn_skirt_fr_01_l}}
     {targetBone {boneName=thigh_twist_01_l}}
    ]
    constraintWeights [1,1,1,1,1,1,1,1,1]
   }
   n87 = "Constraint - dyn_body_leaf_033_MSTR (PPPPPPPPPPP)"!(
    ComponentPose=@n86
    Alpha=1
    ConstraintWeights_0=0.048
    ConstraintWeights_1:float=0.1345
    ConstraintWeights_2:float=0.0029
    ConstraintWeights_3:float=0.1638
    ConstraintWeights_4:float=0.1948
    ConstraintWeights_5:float=0.067
    ConstraintWeights_6:float=0.0016
    ConstraintWeights_7:float=0.0043
    ConstraintWeights_8:float=0.2626
    ConstraintWeights_9:float=0.1181
    ConstraintWeights_10:float=0.0024
   ) -> (Pose) {
    alphaBoolBlend {blendOption=Linear}
    boneToModify {boneName=dyn_body_leaf_033_MSTR}
    constraintSetup [
     {targetBone {boneName=clavicle_l}}
     {targetBone {boneName=spine_01}}
     {targetBone {boneName=upperarm_twist_02_l}}
     {targetBone {boneName=spine_02}}
     {targetBone {boneName=spine_03}}
     {targetBone {boneName=spine_04}}
     {targetBone {boneName=spine_05}}
     {targetBone {boneName=upperarm_twist_01_l}}
     {targetBone {boneName=pelvis}}
     {targetBone {boneName=thigh_twist_01_l}}
     {targetBone {boneName=upperarm_l}}
    ]
    constraintWeights [1,1,1,1,1,1,1,1,1,1,1]
   }
   n88 = "Constraint - dyn_body_leaf_029_MSTR (PPPPPPPPPPPPPPPP)"!(
    ComponentPose=@n87
    Alpha=1
    ConstraintWeights_0=0.0357
    ConstraintWeights_1:float=0.0126
    ConstraintWeights_2:float=0.056
    ConstraintWeights_3:float=0.0006
    ConstraintWeights_4:float=0.0009
    ConstraintWeights_5:float=0.0023
    ConstraintWeights_6:float=0.0011
    ConstraintWeights_7:float=0.0001
    ConstraintWeights_8:float=0.2386
    ConstraintWeights_9:float=0.0004
    ConstraintWeights_10:float=0.0123
    ConstraintWeights_11:float=0.0011
    ConstraintWeights_12:float=0.5895
    ConstraintWeights_13:float=0.0051
    ConstraintWeights_14:float=0.0406
    ConstraintWeights_15:float=0.0031
   ) -> (Pose) {
    alphaBoolBlend {blendOption=Linear}
    boneToModify {boneName=dyn_body_leaf_029_MSTR}
    constraintSetup [
     {targetBone {boneName=dyn_skirt_side_03_l}}
     {targetBone {boneName=dyn_skirt_fr_03_l}}
     {targetBone {boneName=dyn_skirt_fr_02_l}}
     {targetBone {boneName=dyn_skirt_bk_03_l}}
     {targetBone {boneName=dyn_skirt_fr_01}}
     {targetBone {boneName=dyn_skirt_fr_02}}
     {targetBone {boneName=dyn_skirt_fr_03}}
     {targetBone {boneName=thigh_twist_01_r}}
     {targetBone {boneName=dyn_skirt_side_01_l}}
     {targetBone {boneName=spine_01}}
     {targetBone {boneName=thigh_twist_01_l}}
     {targetBone {boneName=thigh_l}}
     {targetBone {boneName=dyn_skirt_side_02_l}}
     {targetBone {boneName=pelvis}}
     {targetBone {boneName=dyn_skirt_fr_01_l}}
     {targetBone {boneName=dyn_skirt_bk_02_l}}
    ]
    constraintWeights [1,1,1,1,1,1,1,1,1,1,1,1,1,1,1,1]
   }
   n89 = "Constraint - dyn_body_leaf_053_MSTR (PPPPPPPPPPPPPPPPPP)"!(
    ComponentPose=@n88
    Alpha=1
    ConstraintWeights_0=0.0024
    ConstraintWeights_1:float=0.0028
    ConstraintWeights_2:float=0.0002
    ConstraintWeights_3:float=0.0047
    ConstraintWeights_4:float=0.0101
    ConstraintWeights_5:float=0.5208
    ConstraintWeights_6:float=0.0002
    ConstraintWeights_7:float=0.0124
    ConstraintWeights_8:float=0.015
    ConstraintWeights_9:float=0.0062
    ConstraintWeights_10:float=0.0002
    ConstraintWeights_11:float=0.0003
    ConstraintWeights_12:float=0.0002
    ConstraintWeights_13:float=0.0298
    ConstraintWeights_14:float=0.3819
    ConstraintWeights_15:float=0.0073
    ConstraintWeights_16:float=0.0036
    ConstraintWeights_17:float=0.0019
   ) -> (Pose) {
    alphaBoolBlend {blendOption=Linear}
    boneToModify {boneName=dyn_body_leaf_053_MSTR}
    constraintSetup [
     {targetBone {boneName=dyn_skirt_side_02_l}}
     {targetBone {boneName=dyn_skirt_bk_01}}
     {targetBone {boneName=upperarm_l}}
     {targetBone {boneName=clavicle_l}}
     {targetBone {boneName=spine_01}}
     {targetBone {boneName=dyn_skirt_side_01_l}}
     {targetBone {boneName=upperarm_twist_02_l}}
     {targetBone {boneName=spine_02}}
     {targetBone {boneName=spine_03}}
     {targetBone {boneName=spine_04}}
     {targetBone {boneName=spine_05}}
     {targetBone {boneName=upperarm_twist_01_l}}
     {targetBone {boneName=clavicle_r}}
     {targetBone {boneName=pelvis}}
     {targetBone {boneName=dyn_skirt_bk_01_l}}
     {targetBone {boneName=thigh_twist_01_l}}
     {targetBone {boneName=dyn_skirt_fr_01_l}}
     {targetBone {boneName=dyn_skirt_bk_02_l}}
    ]
    constraintWeights [1,1,1,1,1,1,1,1,1,1,1,1,1,1,1,1,1,1]
   }
   n90 = "Constraint - dyn_body_leaf_067_MSTR (PPPPPPPPPPPPPPPPPP)"!(
    ComponentPose=@n89
    Alpha=1
    ConstraintWeights_0=0.0288
    ConstraintWeights_1:float=0.0002
    ConstraintWeights_2:float=0.1006
    ConstraintWeights_3:float=0.6598
    ConstraintWeights_4:float=0.0003
    ConstraintWeights_5:float=0.0121
    ConstraintWeights_6:float=0.0004
    ConstraintWeights_7:float=0.0013
    ConstraintWeights_8:float=0.0045
    ConstraintWeights_9:float=0.0044
    ConstraintWeights_10:float=0.0012
    ConstraintWeights_11:float=0.1378
    ConstraintWeights_12:float=0.0003
    ConstraintWeights_13:float=0.0151
    ConstraintWeights_14:float=0.0002
    ConstraintWeights_15:float=0.0026
    ConstraintWeights_16:float=0.0298
    ConstraintWeights_17:float=0.0003
   ) -> (Pose) {
    alphaBoolBlend {blendOption=Linear}
    boneToModify {boneName=dyn_body_leaf_067_MSTR}
    constraintSetup [
     {targetBone {boneName=dyn_skirt_bk_02_r}}
     {targetBone {boneName=dyn_skirt_side_03_l}}
     {targetBone {boneName=dyn_skirt_bk_03}}
     {targetBone {boneName=dyn_skirt_bk_02}}
     {targetBone {boneName=dyn_skirt_side_02_r}}
     {targetBone {boneName=dyn_skirt_bk_03_l}}
     {targetBone {boneName=thigh_r}}
     {targetBone {boneName=thigh_twist_01_r}}
     {targetBone {boneName=dyn_skirt_bk_01_r}}
     {targetBone {boneName=dyn_skirt_bk_01_l}}
     {targetBone {boneName=thigh_twist_01_l}}
     {targetBone {boneName=dyn_skirt_bk_01}}
     {targetBone {boneName=thigh_l}}
     {targetBone {boneName=dyn_skirt_bk_03_r}}
     {targetBone {boneName=dyn_skirt_side_02_l}}
     {targetBone {boneName=pelvis}}
     {targetBone {boneName=dyn_skirt_bk_02_l}}
     {targetBone {boneName=dyn_skirt_side_03_r}}
    ]
    constraintWeights [1,1,1,1,1,1,1,1,1,1,1,1,1,1,1,1,1,1]
   }
   n91 = "Constraint - dyn_body_leaf_050_MSTR (PPPPPPPPPPPPPPPPPP)"!(
    ComponentPose=@n90
    Alpha=1
    ConstraintWeights_0=0.1247
    ConstraintWeights_1:float=0.0008
    ConstraintWeights_2:float=0.0216
    ConstraintWeights_3:float=0.4188
    ConstraintWeights_4:float=0.0012
    ConstraintWeights_5:float=0.0007
    ConstraintWeights_6:float=0.0005
    ConstraintWeights_7:float=0.0029
    ConstraintWeights_8:float=0.0811
    ConstraintWeights_9:float=0.0001
    ConstraintWeights_10:float=0.0006
    ConstraintWeights_11:float=0.3121
    ConstraintWeights_12:float=0.0001
    ConstraintWeights_13:float=0.0236
    ConstraintWeights_14:float=0.0059
    ConstraintWeights_15:float=0.0012
    ConstraintWeights_16:float=0.0037
    ConstraintWeights_17:float=0.0004
   ) -> (Pose) {
    alphaBoolBlend {blendOption=Linear}
    boneToModify {boneName=dyn_body_leaf_050_MSTR}
    constraintSetup [
     {targetBone {boneName=dyn_skirt_bk_02_r}}
     {targetBone {boneName=dyn_skirt_side_01_r}}
     {targetBone {boneName=dyn_skirt_bk_03}}
     {targetBone {boneName=dyn_skirt_bk_02}}
     {targetBone {boneName=dyn_skirt_side_02_r}}
     {targetBone {boneName=dyn_skirt_bk_03_l}}
     {targetBone {boneName=thigh_r}}
     {targetBone {boneName=thigh_twist_01_r}}
     {targetBone {boneName=dyn_skirt_bk_01_r}}
     {targetBone {boneName=spine_01}}
     {targetBone {boneName=thigh_twist_01_l}}
     {targetBone {boneName=dyn_skirt_bk_01}}
     {targetBone {boneName=thigh_l}}
     {targetBone {boneName=dyn_skirt_bk_03_r}}
     {targetBone {boneName=pelvis}}
     {targetBone {boneName=dyn_skirt_bk_01_l}}
     {targetBone {boneName=dyn_skirt_bk_02_l}}
     {targetBone {boneName=dyn_skirt_side_03_r}}
    ]
    constraintWeights [1,1,1,1,1,1,1,1,1,1,1,1,1,1,1,1,1,1]
   }
   n92 = "Constraint - dyn_body_leaf_028_MSTR (PPPPPPPPPP)"!(
    ComponentPose=@n91
    Alpha=1
    ConstraintWeights_0=0.1736
    ConstraintWeights_1:float=0.0688
    ConstraintWeights_2:float=0.0193
    ConstraintWeights_3:float=0.1761
    ConstraintWeights_4:float=0.2998
    ConstraintWeights_5:float=0.1963
    ConstraintWeights_6:float=0.0039
    ConstraintWeights_7:float=0.0354
    ConstraintWeights_8:float=0.0193
    ConstraintWeights_9:float=0.0076
   ) -> (Pose) {
    alphaBoolBlend {blendOption=Linear}
    boneToModify {boneName=dyn_body_leaf_028_MSTR}
    constraintSetup [
     {targetBone {boneName=clavicle_l}}
     {targetBone {boneName=spine_01}}
     {targetBone {boneName=upperarm_twist_02_l}}
     {targetBone {boneName=spine_02}}
     {targetBone {boneName=spine_03}}
     {targetBone {boneName=spine_04}}
     {targetBone {boneName=spine_05}}
     {targetBone {boneName=upperarm_twist_01_l}}
     {targetBone {boneName=pelvis}}
     {targetBone {boneName=upperarm_l}}
    ]
    constraintWeights [1,1,1,1,1,1,1,1,1,1]
   }
   n93 = "Constraint - dyn_body_leaf_068_MSTR (PPPPPPPPPPPP)"!(
    ComponentPose=@n92
    Alpha=1
    ConstraintWeights_0=0.0003
    ConstraintWeights_1:float=0.0011
    ConstraintWeights_2:float=0.0715
    ConstraintWeights_3:float=0.8274
    ConstraintWeights_4:float=0.001
    ConstraintWeights_5:float=0.0002
    ConstraintWeights_6:float=0.0002
    ConstraintWeights_7:float=0.0002
    ConstraintWeights_8:float=0.0013
    ConstraintWeights_9:float=0.0057
    ConstraintWeights_10:float=0.0727
    ConstraintWeights_11:float=0.0185
   ) -> (Pose) {
    alphaBoolBlend {blendOption=Linear}
    boneToModify {boneName=dyn_body_leaf_068_MSTR}
    constraintSetup [
     {targetBone {boneName=dyn_skirt_bk_02_r}}
     {targetBone {boneName=dyn_skirt_bk_03}}
     {targetBone {boneName=dyn_skirt_bk_02}}
     {targetBone {boneName=dyn_skirt_bk_01}}
     {targetBone {boneName=dyn_skirt_bk_03_l}}
     {targetBone {boneName=thigh_twist_01_r}}
     {targetBone {boneName=dyn_skirt_bk_01_r}}
     {targetBone {boneName=spine_01}}
     {targetBone {boneName=thigh_twist_01_l}}
     {targetBone {boneName=pelvis}}
     {targetBone {boneName=dyn_skirt_bk_01_l}}
     {targetBone {boneName=dyn_skirt_bk_02_l}}
    ]
    constraintWeights [1,1,1,1,1,1,1,1,1,1,1,1]
   }
   n94 = "Constraint - dyn_body_leaf_041_MSTR (PPPPPPPPPPPPPPPP)"!(
    ComponentPose=@n93
    Alpha=1
    ConstraintWeights_0=0.0001
    ConstraintWeights_1:float=0.713
    ConstraintWeights_2:float=0.0009
    ConstraintWeights_3:float=0.0016
    ConstraintWeights_4:float=0.0012
    ConstraintWeights_5:float=0.0033
    ConstraintWeights_6:float=0.1812
    ConstraintWeights_7:float=0.0117
    ConstraintWeights_8:float=0.011
    ConstraintWeights_9:float=0.0146
    ConstraintWeights_10:float=0.0072
    ConstraintWeights_11:float=0.0007
    ConstraintWeights_12:float=0.0042
    ConstraintWeights_13:float=0.0459
    ConstraintWeights_14:float=0.0023
    ConstraintWeights_15:float=0.0009
   ) -> (Pose) {
    alphaBoolBlend {blendOption=Linear}
    boneToModify {boneName=dyn_body_leaf_041_MSTR}
    constraintSetup [
     {targetBone {boneName=dyn_skirt_bk_02_r}}
     {targetBone {boneName=dyn_skirt_bk_01}}
     {targetBone {boneName=dyn_skirt_bk_02}}
     {targetBone {boneName=clavicle_l}}
     {targetBone {boneName=dyn_skirt_side_01_r}}
     {targetBone {boneName=thigh_twist_01_r}}
     {targetBone {boneName=dyn_skirt_bk_01_r}}
     {targetBone {boneName=spine_01}}
     {targetBone {boneName=spine_02}}
     {targetBone {boneName=spine_03}}
     {targetBone {boneName=spine_04}}
     {targetBone {boneName=spine_05}}
     {targetBone {boneName=clavicle_r}}
     {targetBone {boneName=pelvis}}
     {targetBone {boneName=dyn_skirt_bk_01_l}}
     {targetBone {boneName=thigh_twist_01_l}}
    ]
    constraintWeights [1,1,1,1,1,1,1,1,1,1,1,1,1,1,1,1]
   }
   n95 = "Constraint - dyn_body_leaf_039_MSTR (PPPPPPPPPPP)"!(
    ComponentPose=@n94
    Alpha=1
    ConstraintWeights_0=0.027
    ConstraintWeights_1:float=0.0014
    ConstraintWeights_2:float=0.0147
    ConstraintWeights_3:float=0.113
    ConstraintWeights_4:float=0.1267
    ConstraintWeights_5:float=0.2075
    ConstraintWeights_6:float=0.1215
    ConstraintWeights_7:float=0.0161
    ConstraintWeights_8:float=0.0746
    ConstraintWeights_9:float=0.2938
    ConstraintWeights_10:float=0.0038
   ) -> (Pose) {
    alphaBoolBlend {blendOption=Linear}
    boneToModify {boneName=dyn_body_leaf_039_MSTR}
    constraintSetup [
     {targetBone {boneName=clavicle_l}}
     {targetBone {boneName=upperarm_twist_01_r}}
     {targetBone {boneName=thigh_twist_01_r}}
     {targetBone {boneName=spine_01}}
     {targetBone {boneName=spine_02}}
     {targetBone {boneName=spine_03}}
     {targetBone {boneName=spine_04}}
     {targetBone {boneName=spine_05}}
     {targetBone {boneName=clavicle_r}}
     {targetBone {boneName=pelvis}}
     {targetBone {boneName=thigh_twist_01_l}}
    ]
    constraintWeights [1,1,1,1,1,1,1,1,1,1,1]
   }
   n96 = "Constraint - dyn_body_leaf_046_MSTR (PPPPPPPPPPPPPPPP)"!(
    ComponentPose=@n95
    Alpha=1
    ConstraintWeights_0=0.0493
    ConstraintWeights_1:float=0.0433
    ConstraintWeights_2:float=0.2588
    ConstraintWeights_3:float=0.0009
    ConstraintWeights_4:float=0.0048
    ConstraintWeights_5:float=0.0083
    ConstraintWeights_6:float=0.003
    ConstraintWeights_7:float=0.0004
    ConstraintWeights_8:float=0.1421
    ConstraintWeights_9:float=0.0012
    ConstraintWeights_10:float=0.0363
    ConstraintWeights_11:float=0.003
    ConstraintWeights_12:float=0.2807
    ConstraintWeights_13:float=0.0154
    ConstraintWeights_14:float=0.149
    ConstraintWeights_15:float=0.0036
   ) -> (Pose) {
    alphaBoolBlend {blendOption=Linear}
    boneToModify {boneName=dyn_body_leaf_046_MSTR}
    constraintSetup [
     {targetBone {boneName=dyn_skirt_side_03_l}}
     {targetBone {boneName=dyn_skirt_fr_03_l}}
     {targetBone {boneName=dyn_skirt_fr_02_l}}
     {targetBone {boneName=dyn_skirt_bk_03_l}}
     {targetBone {boneName=dyn_skirt_fr_01}}
     {targetBone {boneName=dyn_skirt_fr_02}}
     {targetBone {boneName=dyn_skirt_fr_03}}
     {targetBone {boneName=thigh_twist_01_r}}
     {targetBone {boneName=dyn_skirt_side_01_l}}
     {targetBone {boneName=spine_01}}
     {targetBone {boneName=thigh_twist_01_l}}
     {targetBone {boneName=thigh_l}}
     {targetBone {boneName=dyn_skirt_side_02_l}}
     {targetBone {boneName=pelvis}}
     {targetBone {boneName=dyn_skirt_fr_01_l}}
     {targetBone {boneName=dyn_skirt_bk_02_l}}
    ]
    constraintWeights [1,1,1,1,1,1,1,1,1,1,1,1,1,1,1,1]
   }
   n97 = "Constraint - dyn_body_leaf_030_MSTR (PPPPPPPPPPPPP)"!(
    ComponentPose=@n96
    Alpha=1
    ConstraintWeights_0=0.0001
    ConstraintWeights_1:float=0.0575
    ConstraintWeights_2:float=0.0003
    ConstraintWeights_3:float=0.022
    ConstraintWeights_4:float=0.0215
    ConstraintWeights_5:float=0.0007
    ConstraintWeights_6:float=0.0007
    ConstraintWeights_7:float=0.0008
    ConstraintWeights_8:float=0.0002
    ConstraintWeights_9:float=0.0003
    ConstraintWeights_10:float=0.003
    ConstraintWeights_11:float=0.8892
    ConstraintWeights_12:float=0.0035
   ) -> (Pose) {
    alphaBoolBlend {blendOption=Linear}
    boneToModify {boneName=dyn_body_leaf_030_MSTR}
    constraintSetup [
     {targetBone {boneName=clavicle_l}}
     {targetBone {boneName=dyn_skirt_fr_01}}
     {targetBone {boneName=dyn_skirt_fr_02}}
     {targetBone {boneName=dyn_skirt_fr_02_l}}
     {targetBone {boneName=dyn_skirt_side_01_l}}
     {targetBone {boneName=spine_01}}
     {targetBone {boneName=spine_02}}
     {targetBone {boneName=spine_03}}
     {targetBone {boneName=spine_04}}
     {targetBone {boneName=dyn_skirt_side_02_l}}
     {targetBone {boneName=pelvis}}
     {targetBone {boneName=dyn_skirt_fr_01_l}}
     {targetBone {boneName=thigh_twist_01_l}}
    ]
    constraintWeights [1,1,1,1,1,1,1,1,1,1,1,1,1]
   }
   n98 = "Constraint - dyn_body_leaf_032_MSTR (PPPPPPPPPPPPPPP)"!(
    ComponentPose=@n97
    Alpha=1
    ConstraintWeights_0=0.0006
    ConstraintWeights_1:float=0.0839
    ConstraintWeights_2:float=0.0936
    ConstraintWeights_3:float=0.0049
    ConstraintWeights_4:float=0.2114
    ConstraintWeights_5:float=0.3782
    ConstraintWeights_6:float=0.1449
    ConstraintWeights_7:float=0.0074
    ConstraintWeights_8:float=0.0088
    ConstraintWeights_9:float=0.0003
    ConstraintWeights_10:float=0.0038
    ConstraintWeights_11:float=0.0513
    ConstraintWeights_12:float=0.0003
    ConstraintWeights_13:float=0.0101
    ConstraintWeights_14:float=0.0004
   ) -> (Pose) {
    alphaBoolBlend {blendOption=Linear}
    boneToModify {boneName=dyn_body_leaf_032_MSTR}
    constraintSetup [
     {targetBone {boneName=head}}
     {targetBone {boneName=clavicle_l}}
     {targetBone {boneName=spine_01}}
     {targetBone {boneName=upperarm_twist_02_l}}
     {targetBone {boneName=spine_02}}
     {targetBone {boneName=spine_03}}
     {targetBone {boneName=spine_04}}
     {targetBone {boneName=spine_05}}
     {targetBone {boneName=upperarm_twist_01_l}}
     {targetBone {boneName=clavicle_r}}
     {targetBone {boneName=dyn_leafbody3_01_l}}
     {targetBone {boneName=pelvis}}
     {targetBone {boneName=dyn_skirt_fr_01_l}}
     {targetBone {boneName=thigh_twist_01_l}}
     {targetBone {boneName=upperarm_l}}
    ]
    constraintWeights [1,1,1,1,1,1,1,1,1,1,1,1,1,1,1]
   }
   n99 = "Constraint - dyn_body_leaf_011_MSTR (PPPPPPPPPPPPPPPP)"!(
    ComponentPose=@n98
    Alpha=1
    ConstraintWeights_0=0.0162
    ConstraintWeights_1:float=0.0505
    ConstraintWeights_2:float=0.6007
    ConstraintWeights_3:float=0.0003
    ConstraintWeights_4:float=0.0218
    ConstraintWeights_5:float=0.0193
    ConstraintWeights_6:float=0.0016
    ConstraintWeights_7:float=0.0003
    ConstraintWeights_8:float=0.0481
    ConstraintWeights_9:float=0.0007
    ConstraintWeights_10:float=0.0181
    ConstraintWeights_11:float=0.0012
    ConstraintWeights_12:float=0.0662
    ConstraintWeights_13:float=0.0089
    ConstraintWeights_14:float=0.1447
    ConstraintWeights_15:float=0.0013
   ) -> (Pose) {
    alphaBoolBlend {blendOption=Linear}
    boneToModify {boneName=dyn_body_leaf_011_MSTR}
    constraintSetup [
     {targetBone {boneName=dyn_skirt_side_03_l}}
     {targetBone {boneName=dyn_skirt_fr_03_l}}
     {targetBone {boneName=dyn_skirt_fr_02_l}}
     {targetBone {boneName=dyn_skirt_bk_03_l}}
     {targetBone {boneName=dyn_skirt_fr_01}}
     {targetBone {boneName=dyn_skirt_fr_02}}
     {targetBone {boneName=dyn_skirt_fr_03}}
     {targetBone {boneName=thigh_twist_01_r}}
     {targetBone {boneName=dyn_skirt_side_01_l}}
     {targetBone {boneName=spine_01}}
     {targetBone {boneName=thigh_twist_01_l}}
     {targetBone {boneName=thigh_l}}
     {targetBone {boneName=dyn_skirt_side_02_l}}
     {targetBone {boneName=pelvis}}
     {targetBone {boneName=dyn_skirt_fr_01_l}}
     {targetBone {boneName=dyn_skirt_bk_02_l}}
    ]
    constraintWeights [1,1,1,1,1,1,1,1,1,1,1,1,1,1,1,1]
   }
   n100 = "Constraint - dyn_body_leaf_035_MSTR (PPPPPPPPPPP)"!(
    ComponentPose=@n99
    Alpha=1
    ConstraintWeights_0=0.0686
    ConstraintWeights_1:float=0.1231
    ConstraintWeights_2:float=0.0051
    ConstraintWeights_3:float=0.2039
    ConstraintWeights_4:float=0.2936
    ConstraintWeights_5:float=0.1044
    ConstraintWeights_6:float=0.0036
    ConstraintWeights_7:float=0.0084
    ConstraintWeights_8:float=0.1383
    ConstraintWeights_9:float=0.0488
    ConstraintWeights_10:float=0.0022
   ) -> (Pose) {
    alphaBoolBlend {blendOption=Linear}
    boneToModify {boneName=dyn_body_leaf_035_MSTR}
    constraintSetup [
     {targetBone {boneName=clavicle_l}}
     {targetBone {boneName=spine_01}}
     {targetBone {boneName=upperarm_twist_02_l}}
     {targetBone {boneName=spine_02}}
     {targetBone {boneName=spine_03}}
     {targetBone {boneName=spine_04}}
     {targetBone {boneName=spine_05}}
     {targetBone {boneName=upperarm_twist_01_l}}
     {targetBone {boneName=pelvis}}
     {targetBone {boneName=thigh_twist_01_l}}
     {targetBone {boneName=upperarm_l}}
    ]
    constraintWeights [1,1,1,1,1,1,1,1,1,1,1]
   }
   n101 = "Constraint - dyn_body_leaf_034_MSTR (PPPPPPPP)"!(
    ComponentPose=@n100
    Alpha=1
    ConstraintWeights_0=0.024
    ConstraintWeights_1:float=0.0104
    ConstraintWeights_2:float=0.1453
    ConstraintWeights_3:float=0.1876
    ConstraintWeights_4:float=0.2499
    ConstraintWeights_5:float=0.0622
    ConstraintWeights_6:float=0.2282
    ConstraintWeights_7:float=0.0925
   ) -> (Pose) {
    alphaBoolBlend {blendOption=Linear}
    boneToModify {boneName=dyn_body_leaf_034_MSTR}
    constraintSetup [
     {targetBone {boneName=clavicle_l}}
     {targetBone {boneName=thigh_twist_01_r}}
     {targetBone {boneName=spine_01}}
     {targetBone {boneName=spine_02}}
     {targetBone {boneName=spine_03}}
     {targetBone {boneName=spine_04}}
     {targetBone {boneName=pelvis}}
     {targetBone {boneName=thigh_twist_01_l}}
    ]
    constraintWeights [1,1,1,1,1,1,1,1]
   }
   n102 = "Constraint - dyn_body_leaf_02_MSTR (PPPPPPPPPP)"!(
    ComponentPose=@n101
    Alpha=1
    ConstraintWeights_0=0.0358
    ConstraintWeights_1:float=0.0002
    ConstraintWeights_2:float=0.0827
    ConstraintWeights_3:float=0.1944
    ConstraintWeights_4:float=0.332
    ConstraintWeights_5:float=0.0845
    ConstraintWeights_6:float=0.0007
    ConstraintWeights_7:float=0.2254
    ConstraintWeights_8:float=0.0406
    ConstraintWeights_9:float=0.0038
   ) -> (Pose) {
    alphaBoolBlend {blendOption=Linear}
    boneToModify {boneName=dyn_body_leaf_02_MSTR}
    constraintSetup [
     {targetBone {boneName=clavicle_l}}
     {targetBone {boneName=thigh_twist_01_r}}
     {targetBone {boneName=spine_01}}
     {targetBone {boneName=spine_02}}
     {targetBone {boneName=spine_03}}
     {targetBone {boneName=spine_04}}
     {targetBone {boneName=spine_05}}
     {targetBone {boneName=dyn_leafbody3_01_l}}
     {targetBone {boneName=pelvis}}
     {targetBone {boneName=thigh_twist_01_l}}
    ]
    constraintWeights [1,1,1,1,1,1,1,1,1,1]
   }
   n103 = "Constraint - dyn_body_leaf_063_MSTR (PPPPPPPPPPPPPPPPP)"!(
    ComponentPose=@n102
    Alpha=1
    ConstraintWeights_0=0.0012
    ConstraintWeights_1:float=0.0018
    ConstraintWeights_2:float=0.0624
    ConstraintWeights_3:float=0.1824
    ConstraintWeights_4:float=0.0212
    ConstraintWeights_5:float=0.1536
    ConstraintWeights_6:float=0.0004
    ConstraintWeights_7:float=0.0013
    ConstraintWeights_8:float=0.0001
    ConstraintWeights_9:float=0.0001
    ConstraintWeights_10:float=0.0075
    ConstraintWeights_11:float=0.0018
    ConstraintWeights_12:float=0.001
    ConstraintWeights_13:float=0.0022
    ConstraintWeights_14:float=0.0093
    ConstraintWeights_15:float=0.1042
    ConstraintWeights_16:float=0.4495
   ) -> (Pose) {
    alphaBoolBlend {blendOption=Linear}
    boneToModify {boneName=dyn_body_leaf_063_MSTR}
    constraintSetup [
     {targetBone {boneName=dyn_skirt_bk_02_r}}
     {targetBone {boneName=dyn_skirt_side_03_l}}
     {targetBone {boneName=dyn_skirt_bk_03}}
     {targetBone {boneName=dyn_skirt_bk_02}}
     {targetBone {boneName=dyn_skirt_bk_01}}
     {targetBone {boneName=dyn_skirt_bk_03_l}}
     {targetBone {boneName=thigh_r}}
     {targetBone {boneName=thigh_twist_01_r}}
     {targetBone {boneName=dyn_skirt_bk_01_r}}
     {targetBone {boneName=spine_01}}
     {targetBone {boneName=thigh_twist_01_l}}
     {targetBone {boneName=thigh_l}}
     {targetBone {boneName=dyn_skirt_bk_03_r}}
     {targetBone {boneName=dyn_skirt_side_02_l}}
     {targetBone {boneName=pelvis}}
     {targetBone {boneName=dyn_skirt_bk_01_l}}
     {targetBone {boneName=dyn_skirt_bk_02_l}}
    ]
    constraintWeights [1,1,1,1,1,1,1,1,1,1,1,1,1,1,1,1,1]
   }
   n104 = "Constraint - dyn_body_leaf_017_MSTR (PPPPPPPPPPP)"!(
    ComponentPose=@n103
    Alpha=1
    ConstraintWeights_0=0.0321
    ConstraintWeights_1:float=0.0043
    ConstraintWeights_2:float=0.132
    ConstraintWeights_3:float=0.1645
    ConstraintWeights_4:float=0.2266
    ConstraintWeights_5:float=0.0621
    ConstraintWeights_6:float=0.0013
    ConstraintWeights_7:float=0.0016
    ConstraintWeights_8:float=0.2274
    ConstraintWeights_9:float=0.0353
    ConstraintWeights_10:float=0.1128
   ) -> (Pose) {
    alphaBoolBlend {blendOption=Linear}
    boneToModify {boneName=dyn_body_leaf_017_MSTR}
    constraintSetup [
     {targetBone {boneName=clavicle_l}}
     {targetBone {boneName=thigh_twist_01_r}}
     {targetBone {boneName=spine_01}}
     {targetBone {boneName=spine_02}}
     {targetBone {boneName=spine_03}}
     {targetBone {boneName=spine_04}}
     {targetBone {boneName=spine_05}}
     {targetBone {boneName=upperarm_twist_01_l}}
     {targetBone {boneName=pelvis}}
     {targetBone {boneName=dyn_skirt_fr_01_l}}
     {targetBone {boneName=thigh_twist_01_l}}
    ]
    constraintWeights [1,1,1,1,1,1,1,1,1,1,1]
   }
   n105 = "Constraint - dyn_body_leaf_022_MSTR (PPPPPPPPPPPP)"!(
    ComponentPose=@n104
    Alpha=1
    ConstraintWeights_0=0.1555
    ConstraintWeights_1:float=0.0772
    ConstraintWeights_2:float=0.0041
    ConstraintWeights_3:float=0.1346
    ConstraintWeights_4:float=0.2667
    ConstraintWeights_5:float=0.1861
    ConstraintWeights_6:float=0.0132
    ConstraintWeights_7:float=0.009
    ConstraintWeights_8:float=0.0127
    ConstraintWeights_9:float=0.1286
    ConstraintWeights_10:float=0.0093
    ConstraintWeights_11:float=0.003
   ) -> (Pose) {
    alphaBoolBlend {blendOption=Linear}
    boneToModify {boneName=dyn_body_leaf_022_MSTR}
    constraintSetup [
     {targetBone {boneName=clavicle_l}}
     {targetBone {boneName=spine_01}}
     {targetBone {boneName=upperarm_twist_02_l}}
     {targetBone {boneName=spine_02}}
     {targetBone {boneName=spine_03}}
     {targetBone {boneName=spine_04}}
     {targetBone {boneName=spine_05}}
     {targetBone {boneName=upperarm_twist_01_l}}
     {targetBone {boneName=clavicle_r}}
     {targetBone {boneName=pelvis}}
     {targetBone {boneName=thigh_twist_01_l}}
     {targetBone {boneName=upperarm_l}}
    ]
    constraintWeights [1,1,1,1,1,1,1,1,1,1,1,1]
   }
   n106 = "Constraint - dyn_body_leaf_036_MSTR (PPPPPPPPPPPPP)"!(
    ComponentPose=@n105
    Alpha=1
    ConstraintWeights_0=0.9072
    ConstraintWeights_1:float=0.017
    ConstraintWeights_2:float=0.0054
    ConstraintWeights_3:float=0.0014
    ConstraintWeights_4:float=0.0143
    ConstraintWeights_5:float=0.0287
    ConstraintWeights_6:float=0.0202
    ConstraintWeights_7:float=0.0003
    ConstraintWeights_8:float=0.0019
    ConstraintWeights_9:float=0.0002
    ConstraintWeights_10:float=0.0022
    ConstraintWeights_11:float=0.0001
    ConstraintWeights_12:float=0.0011
   ) -> (Pose) {
    alphaBoolBlend {blendOption=Linear}
    boneToModify {boneName=dyn_body_leaf_036_MSTR}
    constraintSetup [
     {targetBone {boneName=dyn_leafbody2_01_l}}
     {targetBone {boneName=clavicle_l}}
     {targetBone {boneName=spine_01}}
     {targetBone {boneName=upperarm_twist_02_l}}
     {targetBone {boneName=spine_02}}
     {targetBone {boneName=spine_03}}
     {targetBone {boneName=spine_04}}
     {targetBone {boneName=spine_05}}
     {targetBone {boneName=upperarm_twist_01_l}}
     {targetBone {boneName=clavicle_r}}
     {targetBone {boneName=pelvis}}
     {targetBone {boneName=thigh_twist_01_l}}
     {targetBone {boneName=upperarm_l}}
    ]
    constraintWeights [1,1,1,1,1,1,1,1,1,1,1,1,1]
   }
   n107 = "Constraint - dyn_body_leaf_062_MSTR (PPPPPPPPPPP)"!(
    ComponentPose=@n106
    Alpha=1
    ConstraintWeights_0=0.0013
    ConstraintWeights_1:float=0.0036
    ConstraintWeights_2:float=0.0006
    ConstraintWeights_3:float=0.0212
    ConstraintWeights_4:float=0.0014
    ConstraintWeights_5:float=0.4416
    ConstraintWeights_6:float=0.0011
    ConstraintWeights_7:float=0.0001
    ConstraintWeights_8:float=0.0018
    ConstraintWeights_9:float=0.001
    ConstraintWeights_10:float=0.5262
   ) -> (Pose) {
    alphaBoolBlend {blendOption=Linear}
    boneToModify {boneName=dyn_body_leaf_062_MSTR}
    constraintSetup [
     {targetBone {boneName=dyn_skirt_bk_03}}
     {targetBone {boneName=dyn_skirt_bk_02}}
     {targetBone {boneName=dyn_skirt_bk_01}}
     {targetBone {boneName=dyn_skirt_bk_03_l}}
     {targetBone {boneName=dyn_skirt_side_01_l}}
     {targetBone {boneName=dyn_skirt_bk_01_l}}
     {targetBone {boneName=thigh_twist_01_l}}
     {targetBone {boneName=thigh_l}}
     {targetBone {boneName=dyn_skirt_side_02_l}}
     {targetBone {boneName=pelvis}}
     {targetBone {boneName=dyn_skirt_bk_02_l}}
    ]
    constraintWeights [1,1,1,1,1,1,1,1,1,1,1]
   }
   n108 = "Constraint - dyn_body_leaf_057_MSTR (PPPPPPPPPPPP)"!(
    ComponentPose=@n107
    Alpha=1
    ConstraintWeights_0=0.1555
    ConstraintWeights_1:float=0.0772
    ConstraintWeights_2:float=0.0041
    ConstraintWeights_3:float=0.1346
    ConstraintWeights_4:float=0.2667
    ConstraintWeights_5:float=0.1861
    ConstraintWeights_6:float=0.0132
    ConstraintWeights_7:float=0.009
    ConstraintWeights_8:float=0.0127
    ConstraintWeights_9:float=0.1286
    ConstraintWeights_10:float=0.0093
    ConstraintWeights_11:float=0.003
   ) -> (Pose) {
    alphaBoolBlend {blendOption=Linear}
    boneToModify {boneName=dyn_body_leaf_057_MSTR}
    constraintSetup [
     {targetBone {boneName=clavicle_l}}
     {targetBone {boneName=spine_01}}
     {targetBone {boneName=upperarm_twist_02_l}}
     {targetBone {boneName=spine_02}}
     {targetBone {boneName=spine_03}}
     {targetBone {boneName=spine_04}}
     {targetBone {boneName=spine_05}}
     {targetBone {boneName=upperarm_twist_01_l}}
     {targetBone {boneName=clavicle_r}}
     {targetBone {boneName=pelvis}}
     {targetBone {boneName=thigh_twist_01_l}}
     {targetBone {boneName=upperarm_l}}
    ]
    constraintWeights [1,1,1,1,1,1,1,1,1,1,1,1]
   }
   n109 = "Constraint - dyn_body_leaf_056_MSTR (PPPPPPPPPPPPP)"!(
    ComponentPose=@n108
    Alpha=1
    ConstraintWeights_0=0.0568
    ConstraintWeights_1:float=0.108
    ConstraintWeights_2:float=0.0021
    ConstraintWeights_3:float=0.001
    ConstraintWeights_4:float=0.1179
    ConstraintWeights_5:float=0.1557
    ConstraintWeights_6:float=0.0779
    ConstraintWeights_7:float=0.0048
    ConstraintWeights_8:float=0.002
    ConstraintWeights_9:float=0.0066
    ConstraintWeights_10:float=0.4183
    ConstraintWeights_11:float=0.0479
    ConstraintWeights_12:float=0.001
   ) -> (Pose) {
    alphaBoolBlend {blendOption=Linear}
    boneToModify {boneName=dyn_body_leaf_056_MSTR}
    constraintSetup [
     {targetBone {boneName=clavicle_l}}
     {targetBone {boneName=spine_01}}
     {targetBone {boneName=thigh_twist_01_r}}
     {targetBone {boneName=upperarm_twist_02_l}}
     {targetBone {boneName=spine_02}}
     {targetBone {boneName=spine_03}}
     {targetBone {boneName=spine_04}}
     {targetBone {boneName=spine_05}}
     {targetBone {boneName=upperarm_twist_01_l}}
     {targetBone {boneName=clavicle_r}}
     {targetBone {boneName=pelvis}}
     {targetBone {boneName=thigh_twist_01_l}}
     {targetBone {boneName=upperarm_l}}
    ]
    constraintWeights [1,1,1,1,1,1,1,1,1,1,1,1,1]
   }
   n110 = "Constraint - dyn_body_leaf_045_MSTR (PPPPPPPPPPPPPPPPP)"!(
    ComponentPose=@n109
    Alpha=1
    ConstraintWeights_0=0.0496
    ConstraintWeights_1:float=0.0148
    ConstraintWeights_2:float=0.0001
    ConstraintWeights_3:float=0.5655
    ConstraintWeights_4:float=0.2882
    ConstraintWeights_5:float=0.0004
    ConstraintWeights_6:float=0.0003
    ConstraintWeights_7:float=0.0008
    ConstraintWeights_8:float=0.0004
    ConstraintWeights_9:float=0.005
    ConstraintWeights_10:float=0.0118
    ConstraintWeights_11:float=0.0002
    ConstraintWeights_12:float=0.0179
    ConstraintWeights_13:float=0.0022
    ConstraintWeights_14:float=0.0103
    ConstraintWeights_15:float=0.0041
    ConstraintWeights_16:float=0.0285
   ) -> (Pose) {
    alphaBoolBlend {blendOption=Linear}
    boneToModify {boneName=dyn_body_leaf_045_MSTR}
    constraintSetup [
     {targetBone {boneName=dyn_skirt_bk_02_r}}
     {targetBone {boneName=dyn_skirt_fr_01_r}}
     {targetBone {boneName=dyn_skirt_bk_02}}
     {targetBone {boneName=dyn_skirt_side_02_r}}
     {targetBone {boneName=dyn_skirt_side_01_r}}
     {targetBone {boneName=thigh_r}}
     {targetBone {boneName=dyn_skirt_fr_01}}
     {targetBone {boneName=dyn_skirt_fr_02}}
     {targetBone {boneName=dyn_skirt_fr_03}}
     {targetBone {boneName=thigh_twist_01_r}}
     {targetBone {boneName=dyn_skirt_bk_01_r}}
     {targetBone {boneName=spine_01}}
     {targetBone {boneName=dyn_skirt_fr_02_r}}
     {targetBone {boneName=pelvis}}
     {targetBone {boneName=dyn_skirt_bk_03_r}}
     {targetBone {boneName=dyn_skirt_fr_03_r}}
     {targetBone {boneName=dyn_skirt_side_03_r}}
    ]
    constraintWeights [1,1,1,1,1,1,1,1,1,1,1,1,1,1,1,1,1]
   }
   n111 = "Constraint - dyn_body_leaf_026_MSTR (PPPPPPPPPPPPPPPPPPP)"!(
    ComponentPose=@n110
    Alpha=1
    ConstraintWeights_0=0.0002
    ConstraintWeights_1:float=0.2348
    ConstraintWeights_2:float=0.0282
    ConstraintWeights_3:float=0.5536
    ConstraintWeights_4:float=0.0003
    ConstraintWeights_5:float=0.0012
    ConstraintWeights_6:float=0.0004
    ConstraintWeights_7:float=0.0163
    ConstraintWeights_8:float=0.0421
    ConstraintWeights_9:float=0.0004
    ConstraintWeights_10:float=0.0025
    ConstraintWeights_11:float=0.0008
    ConstraintWeights_12:float=0.001
    ConstraintWeights_13:float=0.0003
    ConstraintWeights_14:float=0.0941
    ConstraintWeights_15:float=0.0002
    ConstraintWeights_16:float=0.0217
    ConstraintWeights_17:float=0.001
    ConstraintWeights_18:float=0.0009
   ) -> (Pose) {
    alphaBoolBlend {blendOption=Linear}
    boneToModify {boneName=dyn_body_leaf_026_MSTR}
    constraintSetup [
     {targetBone {boneName=dyn_skirt_bk_02_r}}
     {targetBone {boneName=dyn_skirt_fr_01_r}}
     {targetBone {boneName=dyn_skirt_side_02_r}}
     {targetBone {boneName=dyn_skirt_side_01_r}}
     {targetBone {boneName=thigh_r}}
     {targetBone {boneName=dyn_skirt_fr_01}}
     {targetBone {boneName=dyn_skirt_fr_02}}
     {targetBone {boneName=dyn_skirt_fr_02_r}}
     {targetBone {boneName=thigh_twist_01_r}}
     {targetBone {boneName=dyn_skirt_bk_01_r}}
     {targetBone {boneName=spine_01}}
     {targetBone {boneName=spine_02}}
     {targetBone {boneName=spine_03}}
     {targetBone {boneName=spine_04}}
     {targetBone {boneName=dyn_leafbody6_01_r}}
     {targetBone {boneName=clavicle_r}}
     {targetBone {boneName=pelvis}}
     {targetBone {boneName=dyn_skirt_fr_03_r}}
     {targetBone {boneName=dyn_skirt_side_03_r}}
    ]
    constraintWeights [1,1,1,1,1,1,1,1,1,1,1,1,1,1,1,1,1,1,1]
   }
   n112 = "Constraint - dyn_body_leaf_025_MSTR (PPPPPPPPPPPP)"!(
    ComponentPose=@n111
    Alpha=1
    ConstraintWeights_0=0.0692
    ConstraintWeights_1:float=0.3541
    ConstraintWeights_2:float=0.0049
    ConstraintWeights_3:float=0.0656
    ConstraintWeights_4:float=0.024
    ConstraintWeights_5:float=0.073
    ConstraintWeights_6:float=0.0594
    ConstraintWeights_7:float=0.1197
    ConstraintWeights_8:float=0.1911
    ConstraintWeights_9:float=0.0067
    ConstraintWeights_10:float=0.0235
    ConstraintWeights_11:float=0.0089
   ) -> (Pose) {
    alphaBoolBlend {blendOption=Linear}
    boneToModify {boneName=dyn_body_leaf_025_MSTR}
    constraintSetup [
     {targetBone {boneName=head}}
     {targetBone {boneName=clavicle_l}}
     {targetBone {boneName=spine_01}}
     {targetBone {boneName=upperarm_twist_02_l}}
     {targetBone {boneName=spine_02}}
     {targetBone {boneName=spine_03}}
     {targetBone {boneName=spine_04}}
     {targetBone {boneName=spine_05}}
     {targetBone {boneName=upperarm_twist_01_l}}
     {targetBone {boneName=clavicle_r}}
     {targetBone {boneName=neck_02}}
     {targetBone {boneName=neck_01}}
    ]
    constraintWeights [1,1,1,1,1,1,1,1,1,1,1,1]
   }
   n113 = "Constraint - dyn_body_leaf_07_MSTR (PPPPPPPPPPPPP)"!(
    ComponentPose=@n112
    Alpha=1
    ConstraintWeights_0=0.1715
    ConstraintWeights_1:float=0.0194
    ConstraintWeights_2:float=0.1105
    ConstraintWeights_3:float=0.0416
    ConstraintWeights_4:float=0.0069
    ConstraintWeights_5:float=0.0297
    ConstraintWeights_6:float=0.0764
    ConstraintWeights_7:float=0.0789
    ConstraintWeights_8:float=0.1546
    ConstraintWeights_9:float=0.0019
    ConstraintWeights_10:float=0.2523
    ConstraintWeights_11:float=0.0412
    ConstraintWeights_12:float=0.0152
   ) -> (Pose) {
    alphaBoolBlend {blendOption=Linear}
    boneToModify {boneName=dyn_body_leaf_07_MSTR}
    constraintSetup [
     {targetBone {boneName=head}}
     {targetBone {boneName=clavicle_l}}
     {targetBone {boneName=upperarm_twist_01_r}}
     {targetBone {boneName=upperarm_twist_02_r}}
     {targetBone {boneName=spine_01}}
     {targetBone {boneName=spine_02}}
     {targetBone {boneName=spine_03}}
     {targetBone {boneName=spine_04}}
     {targetBone {boneName=spine_05}}
     {targetBone {boneName=upperarm_twist_01_l}}
     {targetBone {boneName=clavicle_r}}
     {targetBone {boneName=neck_02}}
     {targetBone {boneName=neck_01}}
    ]
    constraintWeights [1,1,1,1,1,1,1,1,1,1,1,1,1]
   }
   n114 = "Constraint - dyn_body_leaf_019_MSTR (PPPPPPPPPPP)"!(
    ComponentPose=@n113
    Alpha=1
    ConstraintWeights_0=0.0023
    ConstraintWeights_1:float=0.2362
    ConstraintWeights_2:float=0.0994
    ConstraintWeights_3:float=0.0138
    ConstraintWeights_4:float=0.0501
    ConstraintWeights_5:float=0.1281
    ConstraintWeights_6:float=0.0781
    ConstraintWeights_7:float=0.0332
    ConstraintWeights_8:float=0.0011
    ConstraintWeights_9:float=0.3564
    ConstraintWeights_10:float=0.001
   ) -> (Pose) {
    alphaBoolBlend {blendOption=Linear}
    boneToModify {boneName=dyn_body_leaf_019_MSTR}
    constraintSetup [
     {targetBone {boneName=head}}
     {targetBone {boneName=upperarm_twist_01_r}}
     {targetBone {boneName=upperarm_twist_02_r}}
     {targetBone {boneName=spine_01}}
     {targetBone {boneName=spine_02}}
     {targetBone {boneName=spine_03}}
     {targetBone {boneName=spine_04}}
     {targetBone {boneName=spine_05}}
     {targetBone {boneName=pelvis}}
     {targetBone {boneName=clavicle_r}}
     {targetBone {boneName=neck_02}}
    ]
    constraintWeights [1,1,1,1,1,1,1,1,1,1,1]
   }
   n115 = "Constraint - dyn_body_leaf_010_MSTR (PPPPPPPPPPPPP)"!(
    ComponentPose=@n114
    Alpha=1
    ConstraintWeights_0=0.0195
    ConstraintWeights_1:float=0.0076
    ConstraintWeights_2:float=0.0389
    ConstraintWeights_3:float=0.0201
    ConstraintWeights_4:float=0.0515
    ConstraintWeights_5:float=0.1478
    ConstraintWeights_6:float=0.3058
    ConstraintWeights_7:float=0.2007
    ConstraintWeights_8:float=0.0371
    ConstraintWeights_9:float=0.0037
    ConstraintWeights_10:float=0.156
    ConstraintWeights_11:float=0.0098
    ConstraintWeights_12:float=0.0013
   ) -> (Pose) {
    alphaBoolBlend {blendOption=Linear}
    boneToModify {boneName=dyn_body_leaf_010_MSTR}
    constraintSetup [
     {targetBone {boneName=head}}
     {targetBone {boneName=clavicle_l}}
     {targetBone {boneName=upperarm_twist_01_r}}
     {targetBone {boneName=upperarm_twist_02_r}}
     {targetBone {boneName=spine_01}}
     {targetBone {boneName=spine_02}}
     {targetBone {boneName=spine_03}}
     {targetBone {boneName=spine_04}}
     {targetBone {boneName=spine_05}}
     {targetBone {boneName=neck_02}}
     {targetBone {boneName=clavicle_r}}
     {targetBone {boneName=pelvis}}
     {targetBone {boneName=neck_01}}
    ]
    constraintWeights [1,1,1,1,1,1,1,1,1,1,1,1,1]
   }
   n116 = "Constraint - dyn_body_leaf_09_MSTR (PPPPPPPPPPPPPPP)"!(
    ComponentPose=@n115
    Alpha=1
    ConstraintWeights_0=0.0922
    ConstraintWeights_1:float=0.028
    ConstraintWeights_2:float=0.0299
    ConstraintWeights_3:float=0.0511
    ConstraintWeights_4:float=0.0245
    ConstraintWeights_5:float=0.0018
    ConstraintWeights_6:float=0.099
    ConstraintWeights_7:float=0.2107
    ConstraintWeights_8:float=0.177
    ConstraintWeights_9:float=0.0897
    ConstraintWeights_10:float=0.0036
    ConstraintWeights_11:float=0.0155
    ConstraintWeights_12:float=0.1669
    ConstraintWeights_13:float=0.0046
    ConstraintWeights_14:float=0.0055
   ) -> (Pose) {
    alphaBoolBlend {blendOption=Linear}
    boneToModify {boneName=dyn_body_leaf_09_MSTR}
    constraintSetup [
     {targetBone {boneName=head}}
     {targetBone {boneName=clavicle_l}}
     {targetBone {boneName=spine_01}}
     {targetBone {boneName=upperarm_twist_01_r}}
     {targetBone {boneName=upperarm_twist_02_r}}
     {targetBone {boneName=upperarm_twist_02_l}}
     {targetBone {boneName=spine_02}}
     {targetBone {boneName=spine_03}}
     {targetBone {boneName=spine_04}}
     {targetBone {boneName=spine_05}}
     {targetBone {boneName=upperarm_twist_01_l}}
     {targetBone {boneName=neck_02}}
     {targetBone {boneName=clavicle_r}}
     {targetBone {boneName=pelvis}}
     {targetBone {boneName=neck_01}}
    ]
    constraintWeights [1,1,1,1,1,1,1,1,1,1,1,1,1,1,1]
   }
   n117 = "Constraint - dyn_body_leaf_014_MSTR (PPPPPPPPPPPPPPPP)"!(
    ComponentPose=@n116
    Alpha=1
    ConstraintWeights_0=0.1741
    ConstraintWeights_1:float=0.0023
    ConstraintWeights_2:float=0.0013
    ConstraintWeights_3:float=0.3936
    ConstraintWeights_4:float=0.0002
    ConstraintWeights_5:float=0.0004
    ConstraintWeights_6:float=0.0001
    ConstraintWeights_7:float=0.0024
    ConstraintWeights_8:float=0.0602
    ConstraintWeights_9:float=0.3022
    ConstraintWeights_10:float=0.0028
    ConstraintWeights_11:float=0.0004
    ConstraintWeights_12:float=0.0457
    ConstraintWeights_13:float=0.0015
    ConstraintWeights_14:float=0.0007
    ConstraintWeights_15:float=0.012
   ) -> (Pose) {
    alphaBoolBlend {blendOption=Linear}
    boneToModify {boneName=dyn_body_leaf_014_MSTR}
    constraintSetup [
     {targetBone {boneName=dyn_skirt_bk_02_r}}
     {targetBone {boneName=dyn_skirt_fr_01_r}}
     {targetBone {boneName=dyn_skirt_bk_02}}
     {targetBone {boneName=dyn_skirt_side_02_r}}
     {targetBone {boneName=thigh_r}}
     {targetBone {boneName=dyn_skirt_bk_03}}
     {targetBone {boneName=dyn_skirt_fr_02}}
     {targetBone {boneName=thigh_twist_01_r}}
     {targetBone {boneName=dyn_skirt_bk_01_r}}
     {targetBone {boneName=dyn_skirt_side_01_r}}
     {targetBone {boneName=dyn_skirt_fr_02_r}}
     {targetBone {boneName=dyn_skirt_bk_01}}
     {targetBone {boneName=dyn_skirt_bk_03_r}}
     {targetBone {boneName=pelvis}}
     {targetBone {boneName=dyn_skirt_fr_03_r}}
     {targetBone {boneName=dyn_skirt_side_03_r}}
    ]
    constraintWeights [1,1,1,1,1,1,1,1,1,1,1,1,1,1,1,1]
   }
   n118 = "Constraint - dyn_body_leaf_015_MSTR (PPPPPPPPPPPPPPPPP)"!(
    ComponentPose=@n117
    Alpha=1
    ConstraintWeights_0=0.0109
    ConstraintWeights_1:float=0.0682
    ConstraintWeights_2:float=0.4834
    ConstraintWeights_3:float=0.2541
    ConstraintWeights_4:float=0.0014
    ConstraintWeights_5:float=0.0012
    ConstraintWeights_6:float=0.003
    ConstraintWeights_7:float=0.0013
    ConstraintWeights_8:float=0.0174
    ConstraintWeights_9:float=0.0015
    ConstraintWeights_10:float=0.0006
    ConstraintWeights_11:float=0.0847
    ConstraintWeights_12:float=0.0074
    ConstraintWeights_13:float=0.0019
    ConstraintWeights_14:float=0.0164
    ConstraintWeights_15:float=0.0002
    ConstraintWeights_16:float=0.0463
   ) -> (Pose) {
    alphaBoolBlend {blendOption=Linear}
    boneToModify {boneName=dyn_body_leaf_015_MSTR}
    constraintSetup [
     {targetBone {boneName=dyn_skirt_bk_02_r}}
     {targetBone {boneName=dyn_skirt_fr_01_r}}
     {targetBone {boneName=dyn_skirt_side_02_r}}
     {targetBone {boneName=dyn_skirt_side_01_r}}
     {targetBone {boneName=thigh_r}}
     {targetBone {boneName=dyn_skirt_fr_01}}
     {targetBone {boneName=dyn_skirt_fr_02}}
     {targetBone {boneName=dyn_skirt_fr_03}}
     {targetBone {boneName=thigh_twist_01_r}}
     {targetBone {boneName=dyn_skirt_bk_01_r}}
     {targetBone {boneName=spine_01}}
     {targetBone {boneName=dyn_skirt_fr_02_r}}
     {targetBone {boneName=pelvis}}
     {targetBone {boneName=dyn_skirt_bk_03_r}}
     {targetBone {boneName=dyn_skirt_fr_03_r}}
     {targetBone {boneName=thigh_twist_01_l}}
     {targetBone {boneName=dyn_skirt_side_03_r}}
    ]
    constraintWeights [1,1,1,1,1,1,1,1,1,1,1,1,1,1,1,1,1]
   }
   n119 = "Constraint - dyn_body_leaf_016_MSTR (PPPPPPPPPPPPP)"!(
    ComponentPose=@n118
    Alpha=1
    ConstraintWeights_0=0.0001
    ConstraintWeights_1:float=0.0013
    ConstraintWeights_2:float=0.0001
    ConstraintWeights_3:float=0.0004
    ConstraintWeights_4:float=0.9708
    ConstraintWeights_5:float=0.0017
    ConstraintWeights_6:float=0.0105
    ConstraintWeights_7:float=0.0021
    ConstraintWeights_8:float=0.0026
    ConstraintWeights_9:float=0.003
    ConstraintWeights_10:float=0.0012
    ConstraintWeights_11:float=0.0009
    ConstraintWeights_12:float=0.0054
   ) -> (Pose) {
    alphaBoolBlend {blendOption=Linear}
    boneToModify {boneName=dyn_body_leaf_016_MSTR}
    constraintSetup [
     {targetBone {boneName=dyn_skirt_bk_02_r}}
     {targetBone {boneName=dyn_skirt_fr_01_r}}
     {targetBone {boneName=dyn_skirt_bk_01}}
     {targetBone {boneName=dyn_skirt_side_02_r}}
     {targetBone {boneName=dyn_skirt_side_01_r}}
     {targetBone {boneName=thigh_twist_01_r}}
     {targetBone {boneName=dyn_skirt_bk_01_r}}
     {targetBone {boneName=spine_01}}
     {targetBone {boneName=spine_02}}
     {targetBone {boneName=spine_03}}
     {targetBone {boneName=spine_04}}
     {targetBone {boneName=clavicle_r}}
     {targetBone {boneName=pelvis}}
    ]
    constraintWeights [1,1,1,1,1,1,1,1,1,1,1,1,1]
   }
   n120 = "Constraint - dyn_body_leaf_052_MSTR (PPPPPPPPPPPPPPPPP)"!(
    ComponentPose=@n119
    Alpha=1
    ConstraintWeights_0=0.0002
    ConstraintWeights_1:float=0.2727
    ConstraintWeights_2:float=0.0007
    ConstraintWeights_3:float=0.0017
    ConstraintWeights_4:float=0.005
    ConstraintWeights_5:float=0.0002
    ConstraintWeights_6:float=0.006
    ConstraintWeights_7:float=0.5746
    ConstraintWeights_8:float=0.0163
    ConstraintWeights_9:float=0.0162
    ConstraintWeights_10:float=0.0214
    ConstraintWeights_11:float=0.0106
    ConstraintWeights_12:float=0.0009
    ConstraintWeights_13:float=0.0069
    ConstraintWeights_14:float=0.0651
    ConstraintWeights_15:float=0.0008
    ConstraintWeights_16:float=0.0008
   ) -> (Pose) {
    alphaBoolBlend {blendOption=Linear}
    boneToModify {boneName=dyn_body_leaf_052_MSTR}
    constraintSetup [
     {targetBone {boneName=dyn_skirt_bk_02_r}}
     {targetBone {boneName=dyn_skirt_bk_01}}
     {targetBone {boneName=dyn_skirt_bk_02}}
     {targetBone {boneName=clavicle_l}}
     {targetBone {boneName=dyn_skirt_side_01_r}}
     {targetBone {boneName=upperarm_twist_01_r}}
     {targetBone {boneName=thigh_twist_01_r}}
     {targetBone {boneName=dyn_skirt_bk_01_r}}
     {targetBone {boneName=spine_01}}
     {targetBone {boneName=spine_02}}
     {targetBone {boneName=spine_03}}
     {targetBone {boneName=spine_04}}
     {targetBone {boneName=spine_05}}
     {targetBone {boneName=clavicle_r}}
     {targetBone {boneName=pelvis}}
     {targetBone {boneName=dyn_skirt_bk_01_l}}
     {targetBone {boneName=thigh_twist_01_l}}
    ]
    constraintWeights [1,1,1,1,1,1,1,1,1,1,1,1,1,1,1,1,1]
   }
   n121 = "Constraint - dyn_body_leaf_040_MSTR (PPPPPPPPPPPPPPP)"!(
    ComponentPose=@n120
    Alpha=1
    ConstraintWeights_0=0.0818
    ConstraintWeights_1:float=0.004
    ConstraintWeights_2:float=0.0454
    ConstraintWeights_3:float=0.0018
    ConstraintWeights_4:float=0.0057
    ConstraintWeights_5:float=0.0002
    ConstraintWeights_6:float=0.0017
    ConstraintWeights_7:float=0.6829
    ConstraintWeights_8:float=0.0001
    ConstraintWeights_9:float=0.0003
    ConstraintWeights_10:float=0.1615
    ConstraintWeights_11:float=0.0046
    ConstraintWeights_12:float=0.0098
    ConstraintWeights_13:float=0.0002
    ConstraintWeights_14:float=0.0001
   ) -> (Pose) {
    alphaBoolBlend {blendOption=Linear}
    boneToModify {boneName=dyn_body_leaf_040_MSTR}
    constraintSetup [
     {targetBone {boneName=dyn_skirt_bk_02_r}}
     {targetBone {boneName=dyn_skirt_bk_03}}
     {targetBone {boneName=dyn_skirt_bk_02}}
     {targetBone {boneName=dyn_skirt_side_02_r}}
     {targetBone {boneName=dyn_skirt_side_01_r}}
     {targetBone {boneName=thigh_r}}
     {targetBone {boneName=thigh_twist_01_r}}
     {targetBone {boneName=dyn_skirt_bk_01_r}}
     {targetBone {boneName=spine_01}}
     {targetBone {boneName=thigh_twist_01_l}}
     {targetBone {boneName=dyn_skirt_bk_01}}
     {targetBone {boneName=pelvis}}
     {targetBone {boneName=dyn_skirt_bk_03_r}}
     {targetBone {boneName=dyn_skirt_bk_02_l}}
     {targetBone {boneName=dyn_skirt_side_03_r}}
    ]
    constraintWeights [1,1,1,1,1,1,1,1,1,1,1,1,1,1,1]
   }
   n122 = "Constraint - dyn_body_leaf_069_MSTR (PPPPPPPPPPPPPPPPP)"!(
    ComponentPose=@n121
    Alpha=1
    ConstraintWeights_0=0.4179
    ConstraintWeights_1:float=0.0378
    ConstraintWeights_2:float=0.011
    ConstraintWeights_3:float=0.0547
    ConstraintWeights_4:float=0.038
    ConstraintWeights_5:float=0.0002
    ConstraintWeights_6:float=0.0005
    ConstraintWeights_7:float=0.0035
    ConstraintWeights_8:float=0.2895
    ConstraintWeights_9:float=0.0001
    ConstraintWeights_10:float=0.0004
    ConstraintWeights_11:float=0.0206
    ConstraintWeights_12:float=0.0001
    ConstraintWeights_13:float=0.1197
    ConstraintWeights_14:float=0.0049
    ConstraintWeights_15:float=0.0003
    ConstraintWeights_16:float=0.0009
   ) -> (Pose) {
    alphaBoolBlend {blendOption=Linear}
    boneToModify {boneName=dyn_body_leaf_069_MSTR}
    constraintSetup [
     {targetBone {boneName=dyn_skirt_bk_02_r}}
     {targetBone {boneName=dyn_skirt_side_01_r}}
     {targetBone {boneName=dyn_skirt_bk_03}}
     {targetBone {boneName=dyn_skirt_bk_02}}
     {targetBone {boneName=dyn_skirt_side_02_r}}
     {targetBone {boneName=dyn_skirt_bk_03_l}}
     {targetBone {boneName=thigh_r}}
     {targetBone {boneName=thigh_twist_01_r}}
     {targetBone {boneName=dyn_skirt_bk_01_r}}
     {targetBone {boneName=spine_01}}
     {targetBone {boneName=thigh_twist_01_l}}
     {targetBone {boneName=dyn_skirt_bk_01}}
     {targetBone {boneName=thigh_l}}
     {targetBone {boneName=dyn_skirt_bk_03_r}}
     {targetBone {boneName=pelvis}}
     {targetBone {boneName=dyn_skirt_bk_02_l}}
     {targetBone {boneName=dyn_skirt_side_03_r}}
    ]
    constraintWeights [1,1,1,1,1,1,1,1,1,1,1,1,1,1,1,1,1]
   }
   n123 = "Constraint - dyn_body_leaf_042_MSTR (PPPPPPPPPPPPPPPPPP)"!(
    ComponentPose=@n122
    Alpha=1
    ConstraintWeights_0=0.1584
    ConstraintWeights_1:float=0.0004
    ConstraintWeights_2:float=0.0024
    ConstraintWeights_3:float=0.0639
    ConstraintWeights_4:float=0.2693
    ConstraintWeights_5:float=0.0004
    ConstraintWeights_6:float=0.0001
    ConstraintWeights_7:float=0.0022
    ConstraintWeights_8:float=0.4674
    ConstraintWeights_9:float=0.0006
    ConstraintWeights_10:float=0.0007
    ConstraintWeights_11:float=0.0008
    ConstraintWeights_12:float=0.0004
    ConstraintWeights_13:float=0.0047
    ConstraintWeights_14:float=0.0037
    ConstraintWeights_15:float=0.0003
    ConstraintWeights_16:float=0.0238
    ConstraintWeights_17:float=0.0005
   ) -> (Pose) {
    alphaBoolBlend {blendOption=Linear}
    boneToModify {boneName=dyn_body_leaf_042_MSTR}
    constraintSetup [
     {targetBone {boneName=dyn_skirt_bk_02_r}}
     {targetBone {boneName=dyn_skirt_fr_01_r}}
     {targetBone {boneName=dyn_skirt_bk_02}}
     {targetBone {boneName=dyn_skirt_side_02_r}}
     {targetBone {boneName=dyn_skirt_side_01_r}}
     {targetBone {boneName=dyn_skirt_bk_03}}
     {targetBone {boneName=dyn_skirt_fr_02_r}}
     {targetBone {boneName=thigh_twist_01_r}}
     {targetBone {boneName=dyn_skirt_bk_01_r}}
     {targetBone {boneName=spine_01}}
     {targetBone {boneName=spine_02}}
     {targetBone {boneName=spine_03}}
     {targetBone {boneName=spine_04}}
     {targetBone {boneName=dyn_skirt_bk_01}}
     {targetBone {boneName=pelvis}}
     {targetBone {boneName=clavicle_r}}
     {targetBone {boneName=dyn_skirt_bk_03_r}}
     {targetBone {boneName=dyn_skirt_side_03_r}}
    ]
    constraintWeights [1,1,1,1,1,1,1,1,1,1,1,1,1,1,1,1,1,1]
   }
   n124 = "Constraint - dyn_body_leaf_066_MSTR (PPPPPPPPPPPPPPPP)"!(
    ComponentPose=@n123
    Alpha=1
    ConstraintWeights_0=0.0008
    ConstraintWeights_1:float=0.1442
    ConstraintWeights_2:float=0.0346
    ConstraintWeights_3:float=0.0384
    ConstraintWeights_4:float=0.0008
    ConstraintWeights_5:float=0.0144
    ConstraintWeights_6:float=0.0129
    ConstraintWeights_7:float=0.0011
    ConstraintWeights_8:float=0.0151
    ConstraintWeights_9:float=0.0007
    ConstraintWeights_10:float=0.6929
    ConstraintWeights_11:float=0.0078
    ConstraintWeights_12:float=0.0002
    ConstraintWeights_13:float=0.0275
    ConstraintWeights_14:float=0.0003
    ConstraintWeights_15:float=0.0083
   ) -> (Pose) {
    alphaBoolBlend {blendOption=Linear}
    boneToModify {boneName=dyn_body_leaf_066_MSTR}
    constraintSetup [
     {targetBone {boneName=dyn_skirt_bk_02_r}}
     {targetBone {boneName=dyn_skirt_fr_01_r}}
     {targetBone {boneName=dyn_skirt_side_02_r}}
     {targetBone {boneName=dyn_skirt_side_01_r}}
     {targetBone {boneName=thigh_r}}
     {targetBone {boneName=dyn_skirt_fr_01}}
     {targetBone {boneName=dyn_skirt_fr_02}}
     {targetBone {boneName=dyn_skirt_fr_03}}
     {targetBone {boneName=thigh_twist_01_r}}
     {targetBone {boneName=spine_01}}
     {targetBone {boneName=dyn_skirt_fr_02_r}}
     {targetBone {boneName=pelvis}}
     {targetBone {boneName=dyn_skirt_bk_03_r}}
     {targetBone {boneName=dyn_skirt_fr_03_r}}
     {targetBone {boneName=thigh_twist_01_l}}
     {targetBone {boneName=dyn_skirt_side_03_r}}
    ]
    constraintWeights [1,1,1,1,1,1,1,1,1,1,1,1,1,1,1,1]
   }
   n125 = "Constraint - dyn_body_leaf_03_MSTR (PPPPPPPPPP)"!(
    ComponentPose=@n124
    Alpha=1
    ConstraintWeights_0=0.0002
    ConstraintWeights_1:float=0.0208
    ConstraintWeights_2:float=0.1218
    ConstraintWeights_3:float=0.2429
    ConstraintWeights_4:float=0.3818
    ConstraintWeights_5:float=0.0963
    ConstraintWeights_6:float=0.0014
    ConstraintWeights_7:float=0.0368
    ConstraintWeights_8:float=0.0958
    ConstraintWeights_9:float=0.0024
   ) -> (Pose) {
    alphaBoolBlend {blendOption=Linear}
    boneToModify {boneName=dyn_body_leaf_03_MSTR}
    constraintSetup [
     {targetBone {boneName=clavicle_l}}
     {targetBone {boneName=thigh_twist_01_r}}
     {targetBone {boneName=spine_01}}
     {targetBone {boneName=spine_02}}
     {targetBone {boneName=spine_03}}
     {targetBone {boneName=spine_04}}
     {targetBone {boneName=spine_05}}
     {targetBone {boneName=clavicle_r}}
     {targetBone {boneName=pelvis}}
     {targetBone {boneName=thigh_twist_01_l}}
    ]
    constraintWeights [1,1,1,1,1,1,1,1,1,1]
   }
   n126 = "Constraint - dyn_body_leaf_043_MSTR (PPPPPPPPPPPPPPPPPPPP)"!(
    ComponentPose=@n125
    Alpha=1
    ConstraintWeights_0=0.0009
    ConstraintWeights_1:float=0.3471
    ConstraintWeights_2:float=0.0672
    ConstraintWeights_3:float=0.3934
    ConstraintWeights_4:float=0.0009
    ConstraintWeights_5:float=0.0032
    ConstraintWeights_6:float=0.0026
    ConstraintWeights_7:float=0.0006
    ConstraintWeights_8:float=0.049
    ConstraintWeights_9:float=0.0002
    ConstraintWeights_10:float=0.0023
    ConstraintWeights_11:float=0.0004
    ConstraintWeights_12:float=0.0004
    ConstraintWeights_13:float=0.095
    ConstraintWeights_14:float=0.0243
    ConstraintWeights_15:float=0.0001
    ConstraintWeights_16:float=0.0062
    ConstraintWeights_17:float=0.0002
    ConstraintWeights_18:float=0.0001
    ConstraintWeights_19:float=0.0059
   ) -> (Pose) {
    alphaBoolBlend {blendOption=Linear}
    boneToModify {boneName=dyn_body_leaf_043_MSTR}
    constraintSetup [
     {targetBone {boneName=dyn_skirt_bk_02_r}}
     {targetBone {boneName=dyn_skirt_fr_01_r}}
     {targetBone {boneName=dyn_skirt_side_02_r}}
     {targetBone {boneName=dyn_skirt_side_01_r}}
     {targetBone {boneName=thigh_r}}
     {targetBone {boneName=dyn_skirt_fr_01}}
     {targetBone {boneName=dyn_skirt_fr_02}}
     {targetBone {boneName=dyn_skirt_fr_03}}
     {targetBone {boneName=thigh_twist_01_r}}
     {targetBone {boneName=dyn_skirt_bk_01_r}}
     {targetBone {boneName=spine_01}}
     {targetBone {boneName=spine_02}}
     {targetBone {boneName=spine_03}}
     {targetBone {boneName=dyn_skirt_fr_02_r}}
     {targetBone {boneName=pelvis}}
     {targetBone {boneName=dyn_skirt_bk_03_r}}
     {targetBone {boneName=dyn_skirt_fr_03_r}}
     {targetBone {boneName=thigh_twist_01_l}}
     {targetBone {boneName=spine_04}}
     {targetBone {boneName=dyn_skirt_side_03_r}}
    ]
    constraintWeights [1,1,1,1,1,1,1,1,1,1,1,1,1,1,1,1,1,1,1,1]
   }
   Save cached pose 'leafConstraintOutPose'(Pose=Component To Local(ComponentPose=@n126)) {cacheName=leafConstraintOutPose}
   n127 = "Constraint - dyn_shoulder_01_r_MSTR (R)"(ComponentPose=Local To Component(LocalPose=Copy Pose From Mesh() {bUseAttachedParent=true}), Alpha=1, ConstraintWeights_0=1) {
    alphaBoolBlend {blendOption=Linear}
    boneToModify {boneName=dyn_shoulder_01_r_MSTR}
    constraintSetup [{transformType=Rotation, targetBone {boneName=upperarm_r}}]
    constraintWeights [1]
   }
   n128 = Look At - Bone: dyn_shoulder_01_l_MSTR(ComponentPose=@n127, Alpha=1) {
    alphaBoolBlend {blendOption=Linear}
    boneToModify {boneName=dyn_shoulder_01_l_MSTR}
    lookAtLocation=<0 1>
    lookAtTarget {boneReference {boneName=dyn_shoulder_aim_l}}
    lookAt_Axis {axis=<1 0>}
    lookUp_Axis {axis {x=0, z=1}}
   }
   n129 = Look At - Bone: dyn_shoulder_01_r_MSTR(ComponentPose=@n128, Alpha=1) {
    alphaBoolBlend {blendOption=Linear}
    boneToModify {boneName=dyn_shoulder_01_r_MSTR}
    lookAtLocation=<0 -2 2>
    lookAtTarget {boneReference {boneName=dyn_shoulder_aim_r}}
    lookAt_Axis {axis=<-1 0>}
    lookUp_Axis {axis {x=0, z=1}}
   }
   n130 = "Constraint - dyn_skirt_fr_01_l_MSTR (R)"(ComponentPose=@n129, Alpha=1, ConstraintWeights_0=1) {
    alphaBoolBlend {blendOption=Linear}
    boneToModify {boneName=dyn_skirt_fr_01_l_MSTR}
    constraintSetup [{transformType=Rotation, targetBone {boneName=thigh_l}}]
    constraintWeights [1]
   }
   n131 = "Constraint - dyn_skirt_side_01_l_MSTR (R)"(ComponentPose=@n130, Alpha=0.5, ConstraintWeights_0=1) {
    alphaBoolBlend {blendOption=Linear}
    boneToModify {boneName=dyn_skirt_side_01_l_MSTR}
    constraintSetup [{transformType=Rotation, targetBone {boneName=thigh_l}}]
    constraintWeights [1]
   }
   n132 = "Constraint - dyn_skirt_fr_01_r_MSTR (R)"(ComponentPose=@n131, Alpha=1, ConstraintWeights_0=1) {
    alphaBoolBlend {blendOption=Linear}
    boneToModify {boneName=dyn_skirt_fr_01_r_MSTR}
    constraintSetup [{transformType=Rotation, targetBone {boneName=thigh_r}}]
    constraintWeights [1]
   }
   n133 = "Constraint - dyn_skirt_side_01_r_MSTR (R)"(ComponentPose=@n132, Alpha=0.5, ConstraintWeights_0=1) {
    alphaBoolBlend {blendOption=Linear}
    boneToModify {boneName=dyn_skirt_side_01_r_MSTR}
    constraintSetup [{transformType=Rotation, targetBone {boneName=thigh_r}}]
    constraintWeights [1]
   }
   n134 = "Constraint - dyn_skirt_fr_01_MSTR (RR)"!(ComponentPose=@n133, Alpha=0.5, ConstraintWeights_0=0.5, ConstraintWeights_1:float=0.5) -> (Pose) {
    alphaBoolBlend {blendOption=Linear}
    boneToModify {boneName=dyn_skirt_fr_01_MSTR}
    constraintSetup [
     {transformType=Rotation, targetBone {boneName=dyn_skirt_fr_01_l_MSTR}}
     {transformType=Rotation, targetBone {boneName=dyn_skirt_fr_01_r_MSTR}}
    ]
    constraintWeights [1,1]
   }
   n135 = "Transform (Modify) Bone - Bone: dyn_skirt_side_01_l_MSTR"(ComponentPose=@n134, Alpha=1, Translation="0,0,0", Rotation="20.000,0.000,0.000", Scale="1,1,1") {
    rotationMode=BMM_Additive
    rotationSpace=BCS_BoneSpace
    alphaBoolBlend {blendOption=Linear}
    boneToModify {boneName=dyn_skirt_side_01_l_MSTR}
    rotation {pitch=20}
   }
   n136 = "Transform (Modify) Bone - Bone: dyn_skirt_side_01_r_MSTR"(ComponentPose=@n135, Alpha=1, Translation="0,0,0", Rotation="20,0.000,0.000", Scale="1,1,1") {
    rotationMode=BMM_Additive
    rotationSpace=BCS_BoneSpace
    alphaBoolBlend {blendOption=Linear}
    boneToModify {boneName=dyn_skirt_side_01_r_MSTR}
    rotation {pitch=20}
   }
   n137 = "Constraint - dyn_grenade3_01_MSTR (PPPPPPPPPPPP)"!(
    ComponentPose=@n136
    Alpha=1
    ConstraintWeights_0=0.4389
    ConstraintWeights_1:float=0.0017
    ConstraintWeights_2:float=0
    ConstraintWeights_3:float=0.001
    ConstraintWeights_4:float=0.0036
    ConstraintWeights_5:float=0.009
    ConstraintWeights_6:float=0.5443
    ConstraintWeights_7:float=0.0009
    ConstraintWeights_8:float=0
    ConstraintWeights_9:float=0
    ConstraintWeights_10:float=0.0004
    ConstraintWeights_11:float=0.0003
   ) -> (Pose) {
    alphaBoolBlend {blendOption=Linear}
    boneToModify {boneName=dyn_grenade3_01_MSTR}
    constraintSetup [
     {targetBone {boneName=dyn_belt_NEW_01_r}}
     {targetBone {boneName=clavicle_l}}
     {targetBone {boneName=upperarm_twist_02_l}}
     {targetBone {boneName=spine_01}}
     {targetBone {boneName=spine_02}}
     {targetBone {boneName=spine_03}}
     {targetBone {boneName=spine_04}}
     {targetBone {boneName=spine_05}}
     {targetBone {boneName=upperarm_twist_01_l}}
     {targetBone {boneName=neck_02}}
     {targetBone {boneName=clavicle_r}}
     {targetBone {boneName=pelvis}}
    ]
    constraintWeights [1,1,1,1,1,1,1,1,1,1,1,1]
   }
   n138 = "Constraint - dyn_grenade2_01_MSTR (PPPPPPPPPPPP)"!(
    ComponentPose=@n137
    Alpha=1
    ConstraintWeights_0=0.1187
    ConstraintWeights_1:float=0.0087
    ConstraintWeights_2:float=0
    ConstraintWeights_3:float=0.0028
    ConstraintWeights_4:float=0.0107
    ConstraintWeights_5:float=0.0222
    ConstraintWeights_6:float=0.8303
    ConstraintWeights_7:float=0.0043
    ConstraintWeights_8:float=0.0002
    ConstraintWeights_9:float=0
    ConstraintWeights_10:float=0.0013
    ConstraintWeights_11:float=0.0007
   ) -> (Pose) {
    alphaBoolBlend {blendOption=Linear}
    boneToModify {boneName=dyn_grenade2_01_MSTR}
    constraintSetup [
     {targetBone {boneName=dyn_belt_NEW_01_r}}
     {targetBone {boneName=clavicle_l}}
     {targetBone {boneName=upperarm_twist_02_l}}
     {targetBone {boneName=spine_01}}
     {targetBone {boneName=spine_02}}
     {targetBone {boneName=spine_03}}
     {targetBone {boneName=spine_04}}
     {targetBone {boneName=spine_05}}
     {targetBone {boneName=upperarm_twist_01_l}}
     {targetBone {boneName=neck_02}}
     {targetBone {boneName=clavicle_r}}
     {targetBone {boneName=pelvis}}
    ]
    constraintWeights [1,1,1,1,1,1,1,1,1,1,1,1]
   }
   n139 = "Constraint - dyn_grenade1_01_MSTR (PPPPPPPPPPP)"!(
    ComponentPose=@n138
    Alpha=1
    ConstraintWeights_0=0.0551
    ConstraintWeights_1:float=0.0006
    ConstraintWeights_2:float=0.0057
    ConstraintWeights_3:float=0.0275
    ConstraintWeights_4:float=0.0552
    ConstraintWeights_5:float=0.8237
    ConstraintWeights_6:float=0.0267
    ConstraintWeights_7:float=0.0019
    ConstraintWeights_8:float=0.0002
    ConstraintWeights_9:float=0.003
    ConstraintWeights_10:float=0.0005
   ) -> (Pose) {
    alphaBoolBlend {blendOption=Linear}
    boneToModify {boneName=dyn_grenade1_01_MSTR}
    constraintSetup [
     {targetBone {boneName=clavicle_l}}
     {targetBone {boneName=upperarm_twist_02_l}}
     {targetBone {boneName=spine_01}}
     {targetBone {boneName=spine_02}}
     {targetBone {boneName=spine_03}}
     {targetBone {boneName=spine_04}}
     {targetBone {boneName=spine_05}}
     {targetBone {boneName=upperarm_twist_01_l}}
     {targetBone {boneName=neck_02}}
     {targetBone {boneName=clavicle_r}}
     {targetBone {boneName=pelvis}}
    ]
    constraintWeights [1,1,1,1,1,1,1,1,1,1,1]
   }
   n140 = "Constraint - dyn_belt_NEW_01_r_MSTR (PPPPPPPP)"!(
    ComponentPose=@n139
    Alpha=1
    ConstraintWeights_0=0.0929
    ConstraintWeights_1:float=0.1224
    ConstraintWeights_2:float=0.1021
    ConstraintWeights_3:float=0.1768
    ConstraintWeights_4:float=0.2651
    ConstraintWeights_5:float=0.1345
    ConstraintWeights_6:float=0.0967
    ConstraintWeights_7:float=0.0096
   ) -> (Pose) {
    alphaBoolBlend {blendOption=Linear}
    boneToModify {boneName=dyn_belt_NEW_01_r_MSTR}
    constraintSetup [
     {targetBone {boneName=dyn_leafbody3_01_r}}
     {targetBone {boneName=dyn_leafbody4_01_r}}
     {targetBone {boneName=spine_01}}
     {targetBone {boneName=spine_02}}
     {targetBone {boneName=spine_03}}
     {targetBone {boneName=spine_04}}
     {targetBone {boneName=clavicle_r}}
     {targetBone {boneName=pelvis}}
    ]
    constraintWeights [1,1,1,1,1,1,1,1]
   }
   n141 = Anim Dynamics - Bone: dyn_skirt_fr_01_l(ComponentPose=@n140, Alpha=0.75, ExternalForce="0,0,0") {
    angularDampingOverride=1
    angularSpringConstant=1500
    bAngularSpring=true
    bGravityOverrideInSimSpace=true
    bOverrideAngularDamping=true
    gravityScale=0.35
    simulationSpace=RootRelative
    alphaBoolBlend {blendOption=Linear}
    boundBone {boneName=dyn_skirt_fr_01_l}
    componentLinearAccScale=<0.3 0.3 0.3>
    componentLinearVelScale=<0.3 0.3 0.3>
    physicsBodyDefinitions [{
      sphereCollisionRadius=0
      boundBone {boneName=dyn_skirt_fr_01_l}
      boxExtents=<10 20 20>
      constraintSetup {angularLimitsMax=<0 0 5>, angularLimitsMin=<0 -15 -5>, angularTarget=<0 0 0>}
      localJointOffset=<10 0 0>
     }]
   }
   n142 = Anim Dynamics - Bone: dyn_skirt_fr_02_l(ComponentPose=@n141, Alpha=0.75, ExternalForce="0,0,0") {
    angularDampingOverride=1
    angularSpringConstant=1500
    bAngularSpring=true
    bGravityOverrideInSimSpace=true
    bOverrideAngularDamping=true
    gravityScale=0.35
    simulationSpace=RootRelative
    alphaBoolBlend {blendOption=Linear}
    boundBone {boneName=dyn_skirt_fr_02_l}
    componentLinearAccScale=<0.3 0.3 0.3>
    componentLinearVelScale=<0.3 0.3 0.3>
    physicsBodyDefinitions [{
      sphereCollisionRadius=0
      boundBone {boneName=dyn_skirt_fr_02_l}
      boxExtents=<10 20 20>
      constraintSetup {angularLimitsMax=<0 0 5>, angularLimitsMin=<0 -15 -5>, angularTarget=<0 0 0>}
      localJointOffset=<10 0 0>
     }]
   }
   n143 = Anim Dynamics - Bone: dyn_skirt_fr_03_l(ComponentPose=@n142, Alpha=0.75, ExternalForce="0,0,0") {
    angularDampingOverride=1
    angularSpringConstant=1500
    bAngularSpring=true
    bGravityOverrideInSimSpace=true
    bOverrideAngularDamping=true
    gravityScale=0.35
    simulationSpace=RootRelative
    alphaBoolBlend {blendOption=Linear}
    boundBone {boneName=dyn_skirt_fr_03_l}
    componentLinearAccScale=<0.3 0.3 0.3>
    componentLinearVelScale=<0.3 0.3 0.3>
    physicsBodyDefinitions [{
      sphereCollisionRadius=0
      boundBone {boneName=dyn_skirt_fr_03_l}
      boxExtents=<10 20 20>
      constraintSetup {angularLimitsMax=<0 0 5>, angularLimitsMin=<0 -15 -5>, angularTarget=<0 0 0>}
      localJointOffset=<10 0 0>
     }]
   }
   n144 = Anim Dynamics - Bone: dyn_skirt_side_01_l(ComponentPose=@n143, Alpha=0.75, ExternalForce="0,0,0") {
    angularDampingOverride=1
    angularSpringConstant=1500
    bAngularSpring=true
    bGravityOverrideInSimSpace=true
    bOverrideAngularDamping=true
    gravityScale=0.35
    simulationSpace=RootRelative
    alphaBoolBlend {blendOption=Linear}
    boundBone {boneName=dyn_skirt_side_01_l}
    componentLinearAccScale=<0.3 0.3 0.3>
    componentLinearVelScale=<0.3 0.3 0.3>
    physicsBodyDefinitions [{
      sphereCollisionRadius=0
      boundBone {boneName=dyn_skirt_side_01_l}
      boxExtents=<10 20 20>
      constraintSetup {angularLimitsMax=<0 0 5>, angularLimitsMin=<0 -15 -5>, angularTarget=<0 0 0>}
      localJointOffset=<10 0 0>
     }]
   }
   n145 = Anim Dynamics - Bone: dyn_skirt_side_02_l(ComponentPose=@n144, Alpha=0.75, ExternalForce="0,0,0") {
    angularDampingOverride=1
    angularSpringConstant=1500
    bAngularSpring=true
    bGravityOverrideInSimSpace=true
    bOverrideAngularDamping=true
    gravityScale=0.35
    simulationSpace=RootRelative
    alphaBoolBlend {blendOption=Linear}
    boundBone {boneName=dyn_skirt_side_02_l}
    componentLinearAccScale=<0.3 0.3 0.3>
    componentLinearVelScale=<0.3 0.3 0.3>
    physicsBodyDefinitions [{
      sphereCollisionRadius=0
      boundBone {boneName=dyn_skirt_side_02_l}
      boxExtents=<10 20 20>
      constraintSetup {angularLimitsMax=<0 0 5>, angularLimitsMin=<0 -15 -5>, angularTarget=<0 0 0>}
      localJointOffset=<10 0 0>
     }]
   }
   n146 = Anim Dynamics - Bone: dyn_skirt_side_03_l(ComponentPose=@n145, Alpha=0.75, ExternalForce="0,0,0") {
    angularDampingOverride=1
    angularSpringConstant=1500
    bAngularSpring=true
    bGravityOverrideInSimSpace=true
    bOverrideAngularDamping=true
    gravityScale=0.35
    simulationSpace=RootRelative
    alphaBoolBlend {blendOption=Linear}
    boundBone {boneName=dyn_skirt_side_03_l}
    componentLinearAccScale=<0.3 0.3 0.3>
    componentLinearVelScale=<0.3 0.3 0.3>
    physicsBodyDefinitions [{
      sphereCollisionRadius=0
      boundBone {boneName=dyn_skirt_side_03_l}
      boxExtents=<10 20 20>
      constraintSetup {angularLimitsMax=<0 0 5>, angularLimitsMin=<0 -15 -5>, angularTarget=<0 0 0>}
      localJointOffset=<10 0 0>
     }]
   }
   n147 = Anim Dynamics - Bone: dyn_skirt_bk_01_l(ComponentPose=@n146, Alpha=0.75, ExternalForce="0,0,0") {
    angularDampingOverride=1
    angularSpringConstant=1500
    bAngularSpring=true
    bGravityOverrideInSimSpace=true
    bOverrideAngularDamping=true
    gravityScale=0.35
    simulationSpace=RootRelative
    alphaBoolBlend {blendOption=Linear}
    boundBone {boneName=dyn_skirt_bk_01_l}
    componentLinearAccScale=<0.3 0.3 0.3>
    componentLinearVelScale=<0.3 0.3 0.3>
    physicsBodyDefinitions [{
      sphereCollisionRadius=0
      boundBone {boneName=dyn_skirt_bk_01_l}
      boxExtents=<10 20 20>
      constraintSetup {angularLimitsMax=<0 0 5>, angularLimitsMin=<0 -15 -5>, angularTarget=<0 0 0>}
      localJointOffset=<10 0 0>
     }]
   }
   n148 = Anim Dynamics - Bone: dyn_skirt_bk_02_l(ComponentPose=@n147, Alpha=0.75, ExternalForce="0,0,0") {
    angularDampingOverride=1
    angularSpringConstant=1500
    bAngularSpring=true
    bGravityOverrideInSimSpace=true
    bOverrideAngularDamping=true
    gravityScale=0.35
    simulationSpace=RootRelative
    alphaBoolBlend {blendOption=Linear}
    boundBone {boneName=dyn_skirt_bk_02_l}
    componentLinearAccScale=<0.3 0.3 0.3>
    componentLinearVelScale=<0.3 0.3 0.3>
    physicsBodyDefinitions [{
      sphereCollisionRadius=0
      boundBone {boneName=dyn_skirt_bk_02_l}
      boxExtents=<10 20 20>
      constraintSetup {angularLimitsMax=<0 0 5>, angularLimitsMin=<0 -15 -5>, angularTarget=<0 0 0>}
      localJointOffset=<10 0 0>
     }]
   }
   n149 = Anim Dynamics - Bone: dyn_skirt_bk_03_l(ComponentPose=@n148, Alpha=0.75, ExternalForce="0,0,0") {
    angularDampingOverride=1
    angularSpringConstant=1500
    bAngularSpring=true
    bGravityOverrideInSimSpace=true
    bOverrideAngularDamping=true
    gravityScale=0.35
    simulationSpace=RootRelative
    alphaBoolBlend {blendOption=Linear}
    boundBone {boneName=dyn_skirt_bk_03_l}
    componentLinearAccScale=<0.3 0.3 0.3>
    componentLinearVelScale=<0.3 0.3 0.3>
    physicsBodyDefinitions [{
      sphereCollisionRadius=0
      boundBone {boneName=dyn_skirt_bk_03_l}
      boxExtents=<10 20 20>
      constraintSetup {angularLimitsMax=<0 0 5>, angularLimitsMin=<0 -15 -5>, angularTarget=<0 0 0>}
      localJointOffset=<10 0 0>
     }]
   }
   n150 = Anim Dynamics - Bone: dyn_skirt_fr_01_r(ComponentPose=@n149, Alpha=0.75, ExternalForce="0,0,0") {
    angularDampingOverride=1
    angularSpringConstant=1500
    bAngularSpring=true
    bGravityOverrideInSimSpace=true
    bOverrideAngularDamping=true
    gravityScale=0.35
    simulationSpace=RootRelative
    alphaBoolBlend {blendOption=Linear}
    boundBone {boneName=dyn_skirt_fr_01_r}
    componentLinearAccScale=<0.3 0.3 0.3>
    componentLinearVelScale=<0.3 0.3 0.3>
    physicsBodyDefinitions [{
      sphereCollisionRadius=0
      boundBone {boneName=dyn_skirt_fr_01_r}
      boxExtents=<10 20 20>
      constraintSetup {angularLimitsMax=<0 0 5>, angularLimitsMin=<0 -15 -5>, angularTarget=<0 0 0>}
      localJointOffset=<10 0 0>
     }]
   }
   n151 = Anim Dynamics - Bone: dyn_skirt_fr_02_r(ComponentPose=@n150, Alpha=0.75, ExternalForce="0,0,0") {
    angularDampingOverride=1
    angularSpringConstant=1500
    bAngularSpring=true
    bGravityOverrideInSimSpace=true
    bOverrideAngularDamping=true
    gravityScale=0.35
    simulationSpace=RootRelative
    alphaBoolBlend {blendOption=Linear}
    boundBone {boneName=dyn_skirt_fr_02_r}
    componentLinearAccScale=<0.3 0.3 0.3>
    componentLinearVelScale=<0.3 0.3 0.3>
    physicsBodyDefinitions [{
      sphereCollisionRadius=0
      boundBone {boneName=dyn_skirt_fr_02_r}
      boxExtents=<10 20 20>
      constraintSetup {angularLimitsMax=<0 0 5>, angularLimitsMin=<0 -15 -5>, angularTarget=<0 0 0>}
      localJointOffset=<10 0 0>
     }]
   }
   n152 = Anim Dynamics - Bone: dyn_skirt_fr_03_r(ComponentPose=@n151, Alpha=0.75, ExternalForce="0,0,0") {
    angularDampingOverride=1
    angularSpringConstant=1500
    bAngularSpring=true
    bGravityOverrideInSimSpace=true
    bOverrideAngularDamping=true
    gravityScale=0.35
    simulationSpace=RootRelative
    alphaBoolBlend {blendOption=Linear}
    boundBone {boneName=dyn_skirt_fr_03_r}
    componentLinearAccScale=<0.3 0.3 0.3>
    componentLinearVelScale=<0.3 0.3 0.3>
    physicsBodyDefinitions [{
      sphereCollisionRadius=0
      boundBone {boneName=dyn_skirt_fr_03_r}
      boxExtents=<10 20 20>
      constraintSetup {angularLimitsMax=<0 0 5>, angularLimitsMin=<0 -15 -5>, angularTarget=<0 0 0>}
      localJointOffset=<10 0 0>
     }]
   }
   n153 = Anim Dynamics - Bone: dyn_skirt_side_01_r(ComponentPose=@n152, Alpha=0.75, ExternalForce="0,0,0") {
    angularDampingOverride=1
    angularSpringConstant=1500
    bAngularSpring=true
    bGravityOverrideInSimSpace=true
    bOverrideAngularDamping=true
    gravityScale=0.35
    simulationSpace=RootRelative
    alphaBoolBlend {blendOption=Linear}
    boundBone {boneName=dyn_skirt_side_01_r}
    componentLinearAccScale=<0.3 0.3 0.3>
    componentLinearVelScale=<0.3 0.3 0.3>
    physicsBodyDefinitions [{
      sphereCollisionRadius=0
      boundBone {boneName=dyn_skirt_side_01_r}
      boxExtents=<10 20 20>
      constraintSetup {angularLimitsMax=<0 0 5>, angularLimitsMin=<0 -15 -5>, angularTarget=<0 0 0>}
      localJointOffset=<10 0 0>
     }]
   }
   n154 = Anim Dynamics - Bone: dyn_skirt_side_02_r(ComponentPose=@n153, Alpha=0.75, ExternalForce="0,0,0") {
    angularDampingOverride=1
    angularSpringConstant=1500
    bAngularSpring=true
    bGravityOverrideInSimSpace=true
    bOverrideAngularDamping=true
    gravityScale=0.35
    simulationSpace=RootRelative
    alphaBoolBlend {blendOption=Linear}
    boundBone {boneName=dyn_skirt_side_02_r}
    componentLinearAccScale=<0.3 0.3 0.3>
    componentLinearVelScale=<0.3 0.3 0.3>
    physicsBodyDefinitions [{
      sphereCollisionRadius=0
      boundBone {boneName=dyn_skirt_side_02_r}
      boxExtents=<10 20 20>
      constraintSetup {angularLimitsMax=<0 0 5>, angularLimitsMin=<0 -15 -5>, angularTarget=<0 0 0>}
      localJointOffset=<10 0 0>
     }]
   }
   n155 = Anim Dynamics - Bone: dyn_skirt_side_03_r(ComponentPose=@n154, Alpha=0.75, ExternalForce="0,0,0") {
    angularDampingOverride=1
    angularSpringConstant=1500
    bAngularSpring=true
    bGravityOverrideInSimSpace=true
    bOverrideAngularDamping=true
    gravityScale=0.35
    simulationSpace=RootRelative
    alphaBoolBlend {blendOption=Linear}
    boundBone {boneName=dyn_skirt_side_03_r}
    componentLinearAccScale=<0.3 0.3 0.3>
    componentLinearVelScale=<0.3 0.3 0.3>
    physicsBodyDefinitions [{
      sphereCollisionRadius=0
      boundBone {boneName=dyn_skirt_side_03_r}
      boxExtents=<10 20 20>
      constraintSetup {angularLimitsMax=<0 0 5>, angularLimitsMin=<0 -15 -5>, angularTarget=<0 0 0>}
      localJointOffset=<10 0 0>
     }]
   }
   n156 = Anim Dynamics - Bone: dyn_skirt_bk_01_r(ComponentPose=@n155, Alpha=0.75, ExternalForce="0,0,0") {
    angularDampingOverride=1
    angularSpringConstant=1500
    bAngularSpring=true
    bGravityOverrideInSimSpace=true
    bOverrideAngularDamping=true
    gravityScale=0.35
    simulationSpace=RootRelative
    alphaBoolBlend {blendOption=Linear}
    boundBone {boneName=dyn_skirt_bk_01_r}
    componentLinearAccScale=<0.3 0.3 0.3>
    componentLinearVelScale=<0.3 0.3 0.3>
    physicsBodyDefinitions [{
      sphereCollisionRadius=0
      boundBone {boneName=dyn_skirt_bk_01_r}
      boxExtents=<10 20 20>
      constraintSetup {angularLimitsMax=<0 0 5>, angularLimitsMin=<0 -15 -5>, angularTarget=<0 0 0>}
      localJointOffset=<10 0 0>
     }]
   }
   n157 = Anim Dynamics - Bone: dyn_skirt_bk_02_r(ComponentPose=@n156, Alpha=0.75, ExternalForce="0,0,0") {
    angularDampingOverride=1
    angularSpringConstant=1500
    bAngularSpring=true
    bGravityOverrideInSimSpace=true
    bOverrideAngularDamping=true
    gravityScale=0.35
    simulationSpace=RootRelative
    alphaBoolBlend {blendOption=Linear}
    boundBone {boneName=dyn_skirt_bk_02_r}
    componentLinearAccScale=<0.3 0.3 0.3>
    componentLinearVelScale=<0.3 0.3 0.3>
    physicsBodyDefinitions [{
      sphereCollisionRadius=0
      boundBone {boneName=dyn_skirt_bk_02_r}
      boxExtents=<10 20 20>
      constraintSetup {angularLimitsMax=<0 0 5>, angularLimitsMin=<0 -15 -5>, angularTarget=<0 0 0>}
      localJointOffset=<10 0 0>
     }]
   }
   n158 = Anim Dynamics - Bone: dyn_skirt_bk_03_r(ComponentPose=@n157, Alpha=0.75, ExternalForce="0,0,0") {
    angularDampingOverride=1
    angularSpringConstant=1500
    bAngularSpring=true
    bGravityOverrideInSimSpace=true
    bOverrideAngularDamping=true
    gravityScale=0.35
    simulationSpace=RootRelative
    alphaBoolBlend {blendOption=Linear}
    boundBone {boneName=dyn_skirt_bk_03_r}
    componentLinearAccScale=<0.3 0.3 0.3>
    componentLinearVelScale=<0.3 0.3 0.3>
    physicsBodyDefinitions [{
      sphereCollisionRadius=0
      boundBone {boneName=dyn_skirt_bk_03_r}
      boxExtents=<10 20 20>
      constraintSetup {angularLimitsMax=<0 0 5>, angularLimitsMin=<0 -40 -5>, angularTarget=<0 0 0>}
      localJointOffset=<10 0 0>
     }]
   }
   n159 = "Transform (Modify) Bone - Bone: dyn_leafbody2_01_r_MSTR"!(ComponentPose=@n158, bAlphaBoolEnabled:bool=True, Translation="0,0,2", Rotation="0,0,5", Scale="1,1,1") -> (Pose) {
    alphaInputType=Bool
    rotationMode=BMM_Additive
    rotationSpace=BCS_BoneSpace
    translationMode=BMM_Additive
    translationSpace=BCS_BoneSpace
    alphaBoolBlend {blendOption=Linear}
    boneToModify {boneName=dyn_leafbody2_01_r_MSTR}
    rotation {roll=5}
    translation {z=2}
   }
   n160 = Anim Dynamics - Bone: dyn_leafbody1_01_l(ComponentPose=@n159, Alpha=0.75, ExternalForce="0,0,0") {
    angularDampingOverride=1
    angularSpringConstant=1500
    bAngularSpring=true
    bGravityOverrideInSimSpace=true
    bOverrideAngularDamping=true
    gravityScale=0.35
    simulationSpace=RootRelative
    alphaBoolBlend {blendOption=Linear}
    alphaScaleBias {scale=2}
    boundBone {boneName=dyn_leafbody1_01_l}
    physicsBodyDefinitions [{
      sphereCollisionRadius=0
      boundBone {boneName=dyn_leafbody1_01_l}
      boxExtents=<10 20 20>
      constraintSetup {angularLimitsMax=<0 0 5>, angularLimitsMin=<0 -20 -5>, angularTarget=<0 0 0>}
      localJointOffset=<10 0 0>
     }]
   }
   n161 = Anim Dynamics - Bone: dyn_leafbody2_01_l(ComponentPose=@n160, Alpha=0.75, ExternalForce="0,0,0") {
    angularDampingOverride=1
    angularSpringConstant=1500
    bAngularSpring=true
    bGravityOverrideInSimSpace=true
    bOverrideAngularDamping=true
    gravityScale=0.35
    simulationSpace=RootRelative
    alphaBoolBlend {blendOption=Linear}
    alphaScaleBias {scale=2}
    boundBone {boneName=dyn_leafbody2_01_l}
    physicsBodyDefinitions [{
      sphereCollisionRadius=0
      boundBone {boneName=dyn_leafbody2_01_l}
      boxExtents=<10 20 20>
      constraintSetup {angularLimitsMax=<0 0 5>, angularLimitsMin=<0 -20 -5>, angularTarget=<0 0 0>}
      localJointOffset=<10 0 0>
     }]
   }
   n162 = Anim Dynamics - Bone: dyn_leafbody3_01_l(ComponentPose=@n161, Alpha=0.75, ExternalForce="0,0,0") {
    angularDampingOverride=1
    angularSpringConstant=1500
    bAngularSpring=true
    bGravityOverrideInSimSpace=true
    bOverrideAngularDamping=true
    gravityScale=0.35
    simulationSpace=RootRelative
    alphaBoolBlend {blendOption=Linear}
    alphaScaleBias {scale=2}
    boundBone {boneName=dyn_leafbody3_01_l}
    physicsBodyDefinitions [{
      sphereCollisionRadius=0
      boundBone {boneName=dyn_leafbody3_01_l}
      boxExtents=<10 20 20>
      constraintSetup {angularLimitsMax=<0 0 5>, angularLimitsMin=<0 -20 -5>, angularTarget=<0 0 0>}
      localJointOffset=<10 0 0>
     }]
   }
   n163 = Anim Dynamics - Bone: dyn_leafbody4_01_l(ComponentPose=@n162, Alpha=0.75, ExternalForce="0,0,0") {
    angularDampingOverride=1
    angularSpringConstant=1500
    bAngularSpring=true
    bGravityOverrideInSimSpace=true
    bOverrideAngularDamping=true
    gravityScale=0.35
    simulationSpace=RootRelative
    alphaBoolBlend {blendOption=Linear}
    alphaScaleBias {scale=2}
    boundBone {boneName=dyn_leafbody4_01_l}
    physicsBodyDefinitions [{
      sphereCollisionRadius=0
      boundBone {boneName=dyn_leafbody4_01_l}
      boxExtents=<10 20 20>
      constraintSetup {angularLimitsMax=<0 0 5>, angularLimitsMin=<0 -20 -5>, angularTarget=<0 0 0>}
      localJointOffset=<10 0 0>
     }]
   }
   n164 = Anim Dynamics - Bone: dyn_leafbody5_01_l(ComponentPose=@n163, Alpha=0, ExternalForce="0,0,0") {
    angularDampingOverride=1
    angularSpringConstant=1500
    bAngularSpring=true
    bGravityOverrideInSimSpace=true
    bOverrideAngularDamping=true
    gravityScale=0.35
    simulationSpace=RootRelative
    alphaBoolBlend {blendOption=Linear}
    alphaScaleBias {scale=2}
    boundBone {boneName=dyn_leafbody5_01_l}
    physicsBodyDefinitions [{
      sphereCollisionRadius=0
      boundBone {boneName=dyn_leafbody5_01_l}
      boxExtents=<10 20 20>
      constraintSetup {angularLimitsMax=<0 0 5>, angularLimitsMin=<0 -20 -5>, angularTarget=<0 0 0>}
      localJointOffset=<10 0 0>
     }]
   }
   n165 = Anim Dynamics - Bone: dyn_leafbody4_01_r(ComponentPose=@n164, Alpha=0, ExternalForce="0,0,0") {
    angularDampingOverride=1
    angularSpringConstant=1500
    bAngularSpring=true
    bGravityOverrideInSimSpace=true
    bOverrideAngularDamping=true
    gravityScale=0.35
    simulationSpace=RootRelative
    alphaBoolBlend {blendOption=Linear}
    alphaScaleBias {scale=2}
    boundBone {boneName=dyn_leafbody4_01_r}
    physicsBodyDefinitions [{
      sphereCollisionRadius=0
      boundBone {boneName=dyn_leafbody4_01_r}
      boxExtents=<10 20 20>
      constraintSetup {angularLimitsMax=<0 0 5>, angularLimitsMin=<0 -20 -5>, angularTarget=<0 0 0>}
      localJointOffset=<10 0 0>
     }]
   }
   n166 = Anim Dynamics - Bone: dyn_leafbody3_01_r(ComponentPose=@n165, Alpha=0, ExternalForce="0,0,0") {
    angularDampingOverride=1
    angularSpringConstant=1500
    bAngularSpring=true
    bGravityOverrideInSimSpace=true
    bOverrideAngularDamping=true
    gravityScale=0.35
    simulationSpace=RootRelative
    alphaBoolBlend {blendOption=Linear}
    alphaScaleBias {scale=2}
    boundBone {boneName=dyn_leafbody3_01_r}
    physicsBodyDefinitions [{
      sphereCollisionRadius=0
      boundBone {boneName=dyn_leafbody3_01_r}
      boxExtents=<10 20 20>
      constraintSetup {angularLimitsMax=<0 0 5>, angularLimitsMin=<0 -20 -5>, angularTarget=<0 0 0>}
      localJointOffset=<10 0 0>
     }]
   }
   n167 = Anim Dynamics - Bone: dyn_leafbody2_01_r(ComponentPose=@n166, Alpha=0.75, ExternalForce="0,0,0") {
    angularDampingOverride=1
    angularSpringConstant=1500
    bAngularSpring=true
    bGravityOverrideInSimSpace=true
    bOverrideAngularDamping=true
    gravityScale=0.35
    simulationSpace=RootRelative
    alphaBoolBlend {blendOption=Linear}
    alphaScaleBias {scale=2}
    boundBone {boneName=dyn_leafbody2_01_r}
    physicsBodyDefinitions [{
      sphereCollisionRadius=0
      boundBone {boneName=dyn_leafbody2_01_r}
      boxExtents=<10 20 20>
      constraintSetup {angularLimitsMax=<0 0 5>, angularLimitsMin=<0 -20 -5>, angularTarget=<0 0 0>}
      localJointOffset=<10 0 0>
     }]
   }
   n168 = Anim Dynamics - Bone: dyn_leafbody1_01_r(ComponentPose=@n167, Alpha=0.75, ExternalForce="0,0,0") {
    angularDampingOverride=1
    angularSpringConstant=1500
    bAngularSpring=true
    bGravityOverrideInSimSpace=true
    bOverrideAngularDamping=true
    gravityScale=0.35
    simulationSpace=RootRelative
    alphaBoolBlend {blendOption=Linear}
    alphaScaleBias {scale=2}
    boundBone {boneName=dyn_leafbody1_01_r}
    physicsBodyDefinitions [{
      sphereCollisionRadius=0
      boundBone {boneName=dyn_leafbody1_01_r}
      boxExtents=<10 20 20>
      constraintSetup {angularLimitsMax=<0 0 5>, angularLimitsMin=<0 -20 -5>, angularTarget=<0 0 0>}
      localJointOffset=<10 0 0>
     }]
   }
   n169 = Anim Dynamics - Bone: dyn_leafbody8_01_r(ComponentPose=@n168, Alpha=0.75, ExternalForce="0,0,0") {
    angularDampingOverride=1
    angularSpringConstant=1500
    bAngularSpring=true
    bGravityOverrideInSimSpace=true
    bOverrideAngularDamping=true
    gravityScale=0.35
    simulationSpace=RootRelative
    alphaBoolBlend {blendOption=Linear}
    alphaScaleBias {scale=2}
    boundBone {boneName=dyn_leafbody8_01_r}
    physicsBodyDefinitions [{
      sphereCollisionRadius=0
      boundBone {boneName=dyn_leafbody8_01_r}
      boxExtents=<10 20 20>
      constraintSetup {angularLimitsMax=<0 0 5>, angularLimitsMin=<0 -20 -5>, angularTarget=<0 0 0>}
      localJointOffset=<10 0 0>
     }]
   }
   n170 = Anim Dynamics - Bone: dyn_leafbody7_01_r(ComponentPose=@n169, Alpha=0.75, ExternalForce="0,0,0") {
    angularDampingOverride=1
    angularSpringConstant=1500
    bAngularSpring=true
    bGravityOverrideInSimSpace=true
    bOverrideAngularDamping=true
    gravityScale=0.35
    simulationSpace=RootRelative
    alphaBoolBlend {blendOption=Linear}
    alphaScaleBias {scale=2}
    boundBone {boneName=dyn_leafbody7_01_r}
    physicsBodyDefinitions [{
      sphereCollisionRadius=0
      boundBone {boneName=dyn_leafbody7_01_r}
      boxExtents=<10 20 20>
      constraintSetup {angularLimitsMax=<0 0 5>, angularLimitsMin=<0 -20 -5>, angularTarget=<0 0 0>}
      localJointOffset=<10 0 0>
     }]
   }
   Save cached pose 'preSimOutPose'(Pose=Component To Local(ComponentPose=@n170)) {cacheName=preSimOutPose}
   n171 = RigidBody(ComponentPose=Local To Component(LocalPose=Use cached pose 'preSimOutPose'()), OverrideWorldGravity=@n35) {
    bClampLinearTranslationLimitToRefPose=true
    bOverrideWorldGravity=true
    overridePhysicsAsset=&/M_MED_Weeping_Woods_Fall_belt_Physics
    simulationSpace=BaseBoneSpace
    alphaBoolBlend {blendOption=Linear}
    baseBoneRef {boneName=spine_01}
    componentAppliedLinearAccClamp=<3000 3000 3000>
    componentLinearAccScale=<0.5 0.5 0.5>
    componentLinearVelScale=<0.7 0.7 0.7>
    overrideWorldGravity {z=-980}
   }
   Save cached pose 'simulationPose'(Pose=Component To Local(ComponentPose=@n171)) {cacheName=simulationPose}
   Use cached pose 'leafConstraintOutPose'()
   Get windAmbientVector() -> (windAmbientVector:Vector)
  ]}
 EventGraph {nodes [
   Event BlueprintUpdateAnimation()
   > n1 = Sequence!(execute) -> (then_0, then_1, then_2:Exec, then_3:Exec)
   n2 = Get bIsTargeting()
   n3 = BreakTransform(InTransform=GetDeltaTransformFromRefPose(self=GetOwnerSkeletalMeshComponent(), BoneName=head, BaseName=spine_05))
   > Set headSwingAlpha(headSwingAlpha:double=MapRangeClamped(Value=BreakRotator(InRot=@n3.Rotation).Yaw, InRangeA=0, InRangeB=-45, OutRangeA="To Float (Boolean)"(InBool=@n2))) -> (Output_Get:double)
   n4 = Break Fort Anim Input Skydiving External Force(FortAnimInput_SkydivingExternalForce=Get SkydivingExternalForceInput()) -> (
    bUseSkydivingVectorForce=~
    SkydivingVectorForce=~
    HeadToPelvisDirection=~
    FloatingMultiplier=~
    FloatingAdditive=~
    DivingMultiplier=~
    DivingAdditive=~
    ParachutingMultiplier=~
    ParachutingAdditive=~
   )
   n5 = Set windVector(
    execute=@n1.then_1
    windVector:Vector=SelectVector(A=vector * float(A=@n4.SkydivingVectorForce, B=0.15), B="0, 0,-980", bPickA=Get bIsSkyDiving())
   ) -> (Output_Get:Vector)
   > Set windAmbientVector(windAmbientVector:Vector=vector * float(A=@n4.SkydivingVectorForce, B=10)) -> (Output_Get:Vector)
   n6 = GetOwnerSkeletalMeshComponent()
   n7 = BreakRotator(InRot=BreakTransform(InTransform=GetDeltaTransformFromRefPose(self=@n6, BoneName=pelvis, BaseName=root)).Rotation)
   n8 = SelectFloat(A=980, B=-980, bPickA="float > float"(A="Absolute (Float)"(A=@n7.Yaw), B=90) {is=title K2Node_CallFunction})
   > Set finalGravity(finalGravity:Vector=SelectVector(A=MakeVector(Y=0, Z=@n8), B=@n5, bPickA=Get bIsPlayingEmote())) -> (Output_Get:Vector)
   K2Node_Event_1 = CustomEvent_2() {is=title K2Node_CustomEvent}
   > Do Once()
   n9 = CheckForBackbling()
   > Set isWearingBackblingAlpha(isWearingBackblingAlpha:double="To Float (Boolean)"(InBool=OR Boolean(A=@n9.bIsBackblingLong, B=@n9))) -> (Output_Get:double)
   > Branch(Condition=@n9.bIsBackblingLong)
   n10 = GetPartSkeletalMeshComponent()
   n11 = FindChildCommponentFromAttachSocket!(
    self:Custom Character Part Anim Instance Object Reference
    Parent:Skeletal Mesh Component Object Reference=@n10
    bIncludeAllDescendants:bool=false
    AttachSocket:Name=Cape
   ) -> (ReturnValue:Skeletal Mesh Component Object Reference)
   n12 = Get Distance Between Two Sockets!(
    Component:Skeletal Mesh Component Object Reference=@n11
    SocketOrBoneNameA:Name=backpack_root
    SocketSpaceA:ERelativeTransformSpace Enum=RTS_World
    SocketOrBoneNameB:Name=cape_length
    SocketSpaceB:ERelativeTransformSpace Enum=RTS_World
    bRemapRange:bool=false
    InRangeMin:float=0
    InRangeMax:float=0
    OutRangeMin:float=0
    OutRangeMax:float=0
   ) -> (ReturnValue:float) {is=title K2Node_CallFunction}
   > Set isWearingCapeAlpha(isWearingCapeAlpha:double="To Float (Boolean)"(InBool="float > float"(A=@n12, B=50))) -> (Output_Get:double)
   K2Node_Event_2 = CustomEvent_1() {is=title K2Node_CustomEvent}
   > Set isCrouchingAlpha(isCrouchingAlpha:double=0) -> (Output_Get:double)
   K2Node_Event_3 = CustomEvent() {is=title K2Node_CustomEvent}
   > Set isCrouchingAlpha(isCrouchingAlpha=1)
   "Not Equal (Object)"!(A:Object Reference=@n11, B:Object Reference) -> (ReturnValue:bool)
   TryGetPawnOwner()
   "Not Equal (Object)"(A=FindChildCommponentFromAttachSocket(Parent=@n10, bIncludeAllDescendants=false, AttachSocket=backpack_br))
  ]}
]
variables [
 finalGravity {type=Vector}
 headSwingAlpha {type=number}
 isCrouchingAlpha {type=number}
 isWearingBackblingAlpha {type=number}
 isWearingCapeAlpha {type=number}
 windAmbientVector {type=Vector}
 windVector {type=Vector}
]
